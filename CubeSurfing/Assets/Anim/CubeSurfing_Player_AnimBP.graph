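# Player Animation Blueprint for Cube Surfing, overriding Fortnite's item final-pose layer to switch between on-foot, tac-sprint, activation montage and surfing with inertialization.
# Surfing poses flow from a stance/jump state machine through lean, react and idle additives into a boost state machine.
# On-foot and sprint states rotate weapon_r in world space with spring-controller lag.
# The event graph pulses IsJump for one tick from routed jump events and plays turn sounds.
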AnimBlueprint /CubeSurfing/Assets/Anim/CubeSurfing_Player_AnimBP
target_skeleton=&Fortnite_M_Avg_Player_Skeleton
graphs [
 AnimGraph {nodes [Output Pose()]}
 EventGraph {nodes [
   Event OnCubeSurfingJump() {is=title K2Node_Event}
   > n1 = StartListeningToEvent!(
    execute:Exec
    Target:Gameplay Event Router Component Object Reference=GetGlobalEventRouter()
    EventType:Script Struct Object Reference=/Script/CubeSurfingRuntime.CubeSurfingEvent_Jump
    Context:Object Reference=Get FortPlayerPawn() -> (FortPlayerPawn:Fort Player Pawn Object Reference)
    EventBubblingRule:EEventBubblingRule Enum
   ) -> (
    then:Exec
    AsyncTaskProxy:Async Action Start Listening To Event Object Reference
    OnEventReceived:Exec
    Payload:Cube Surfing Event Jump Structure
   )
   Set IsJump(execute=@n1.OnEventReceived, IsJump:bool=true) -> (Output_Get:bool)
   > DelayUntilNextTick!(execute) -> (then)
   > Set IsJump(IsJump=false)
   Event PlayTurnSound() {is=title K2Node_Event}
   > Play Sound Simple!(
    execute
    self:Sound Library Subsystem Object Reference=Get SoundLibrarySubsystem() -> (ReturnValue:Sound Library Subsystem Object Reference) {is=title K2Node_GetSubsystem}
    OwningActor:Actor Object Reference=Get FortPlayerPawn()
    InSettings:"Simple Sound Settings Structure (by ref)"=Make Simple Sound Settings!(
     EventName:Gameplay Tag Structure=(TagName="SoundLibrary.SpatialTravel.Turn")
     bAttachedSound:bool=true
     bStopWhenAttachedDestroyed:bool=False
     bAutoDestroy:bool=True
     bForceComponentCreation:bool=False
     VolumeMultiplier:float=1
     PitchMultiplier:float=1
     StartTime:float=0
     ConcurrencySettings:Sound Concurrency Object Reference
     AttenuationSettings:Sound Attenuation Object Reference
     AttachName:Name=None
     AttachToComponent:Scene Component Weak object
     LocationType:EAttachLocation Enum=KeepRelativeOffset
     Location:Vector="0,0,0"
    ) -> (SoundLibrarySimpleContextSettings:Simple Sound Settings Structure) {is=title K2Node_MakeStruct}
   ) -> (then, ReturnValue:Sound Library Play Result Structure)
  ]}
 ItemAimOffset {nodes [
   Input Pose!() -> (Pose, AimOffsetAlpha:double, Yaw:double, Pitch:double) {name=InPose_AimOffset}
   Input Pose!() -> (Pose) {name=InPose_PreAimOffsets}
   Output Pose() {layerGroup=Item Override}
  ]}
 ItemFallAdditive {nodes [
   Input Pose() {name=InPoseItemFallAdditive}
   Output Pose() {layerGroup=Item Override}
  ]}
 ItemFallLandAdditive {nodes [
   Input Pose() {name=InPoseItemFallLandAdditive}
   Output Pose() {layerGroup=Item Override}
  ]}
 ItemFallLevel2Override {nodes [
   Input Pose() {name=InPose_FallLevel2}
   Output Pose() {layerGroup=Item Override}
  ]}
 ItemFinalPoseOverride {nodes [
   AnimGraphNode_StateMachine_1 = OnFoot/Activate/Surfing SM() {is=title AnimGraphNode_StateMachine}
   Output Pose(Result=Inertialization!(Source:Pose Link Structure=@AnimGraphNode_StateMachine_1) -> (Pose:Pose Link Structure) {is=title AnimGraphNode_Inertialization}) {layerGroup=Item Override}
   Save cached pose 'FinalInputPose'(Pose=Input Pose() {name=InFinalPose}) {cacheName=FinalInputPose}
   AnimGraphNode_StateMachine_3 = Boost SM() {is=title AnimGraphNode_StateMachine}
   Save cached pose 'Surfing Pose'(Pose=@AnimGraphNode_StateMachine_3) {cacheName=Surfing Pose}
   AnimGraphNode_StateMachine_0 = Stance + Jump SM() {is=title AnimGraphNode_StateMachine}
   Save cached pose 'PreBoost Pose'(Pose=Surf Additives!(InPose:Pose Link Structure=@AnimGraphNode_StateMachine_0) -> (Pose:Pose Link Structure) {is=title AnimGraphNode_LinkedAnimLayer, layer=Surf Additives}) {cacheName=PreBoost Pose}
  ]}
 Stance + Jump SM {
  entry=Stance BlendSpace
  path=ItemFinalPoseOverride.AnimGraphNode_StateMachine_0.Stance + Jump SM
  ids=id
  nodes [
   Not Falling() {is=title AnimStateAliasNode, also {aliases [Jump Loop, Jump PreLand, Jump Start, Stance BlendSpace]}}
   "To Falling (Jump Loop)"() {is=title AnimStateAliasNode, also {aliases [Landing, Stance BlendSpace]}}
   AnimStateNode_0 = Stance BlendSpace() {is=title AnimStateNode}
   AnimStateNode_1 = Jump Start() {is=title AnimStateNode}
   AnimStateNode_2 = Jump Loop() {
    is=title AnimStateNode
    stateEntered {guid=#1}
    stateFullyBlended {guid=#2}
    stateLeft {guid=#3}
   }
   AnimStateNode_3 = Landing() {is=title AnimStateNode}
   AnimStateNode_4 = Jump PreLand() {
    is=title AnimStateNode
    stateEntered {guid=#4}
    stateFullyBlended {guid=#5}
    stateLeft {guid=#6}
   }
   AnimStateTransitionNode_0 = AnimStateTransitionNode() {bAutomaticRuleBasedOnSequencePlayerInState=true, crossfadeDuration=0.25}
   AnimStateTransitionNode_1 = AnimStateTransitionNode() {crossfadeDuration=0.1}
   AnimStateTransitionNode_2 = AnimStateTransitionNode() {
    crossfadeDuration=0.25
    transitionEnd {guid=#7}
    transitionInterrupt {guid=#8}
    transitionStart {guid=#9}
   }
   AnimStateTransitionNode_3 = AnimStateTransitionNode() {crossfadeDuration=0.25}
   AnimStateTransitionNode_4 = AnimStateTransitionNode() {crossfadeDuration=0.15}
   AnimStateTransitionNode_5 = AnimStateTransitionNode() {bAutomaticRuleBasedOnSequencePlayerInState=true, priorityOrder=10}
   AnimStateTransitionNode_6 = AnimStateTransitionNode() {bAutomaticRuleBasedOnSequencePlayerInState=true, crossfadeDuration=0.8}
   AnimStateTransitionNode_9 = AnimStateTransitionNode()
  ]
  links [
   {
    dst=Jump PreLand
    rule=ItemFinalPoseOverride.AnimGraphNode_StateMachine_0.Stance + Jump SM.AnimStateTransitionNode_2.Transition
    src=Jump Loop
    via=AnimStateTransitionNode_2
   }
   {
    dst=Stance BlendSpace
    rule=ItemFinalPoseOverride.AnimGraphNode_StateMachine_0.Stance + Jump SM.AnimStateTransitionNode_5.Transition
    src=Jump PreLand
    via=AnimStateTransitionNode_5
   }
   {
    dst=Jump Loop
    rule=ItemFinalPoseOverride.AnimGraphNode_StateMachine_0.Stance + Jump SM.AnimStateTransitionNode_0.Transition
    src=Jump Start
    via=AnimStateTransitionNode_0
   }
   {
    dst=Jump Start
    rule=ItemFinalPoseOverride.AnimGraphNode_StateMachine_0.Stance + Jump SM.AnimStateTransitionNode_9.Transition
    src=Landing
    via=AnimStateTransitionNode_9
   }
   {
    dst=Stance BlendSpace
    rule=ItemFinalPoseOverride.AnimGraphNode_StateMachine_0.Stance + Jump SM.AnimStateTransitionNode_6.Transition
    src=Landing
    via=AnimStateTransitionNode_6
   }
   {
    dst=Landing
    rule=ItemFinalPoseOverride.AnimGraphNode_StateMachine_0.Stance + Jump SM.AnimStateTransitionNode_4.Transition
    src=Not Falling
    via=AnimStateTransitionNode_4
   }
   {
    dst=Jump Start
    rule=ItemFinalPoseOverride.AnimGraphNode_StateMachine_0.Stance + Jump SM.AnimStateTransitionNode_1.Transition
    src=Stance BlendSpace
    via=AnimStateTransitionNode_1
   }
   {
    dst=Jump Loop
    rule=ItemFinalPoseOverride.AnimGraphNode_StateMachine_0.Stance + Jump SM.AnimStateTransitionNode_3.Transition
    src="To Falling (Jump Loop)"
    via=AnimStateTransitionNode_3
   }
  ]
 }
 Stance BlendSpace {
  path=ItemFinalPoseOverride.AnimGraphNode_StateMachine_0.Stance + Jump SM.AnimStateNode_0.Stance BlendSpace
  nodes [Output Animation Pose(Result=Blendspace Player 'CubeSurfing_Stance_BS'!(X=0) -> (Pose) {blendSpace=&/CubeSurfing_Stance_BS})]
 }
 Jump Start {
  path=ItemFinalPoseOverride.AnimGraphNode_StateMachine_0.Stance + Jump SM.AnimStateNode_1.Jump Start
  nodes [Output Animation Pose(Result=Sequence Player 'CubeSurfing_Jump_Intro'!(PlayRate:float=0.7, StartPosition:float=0.04) -> (Pose) {bLoopAnimation=false, sequence=&/CubeSurfing_Jump_Intro})]
 }
 Jump Loop {
  path=ItemFinalPoseOverride.AnimGraphNode_StateMachine_0.Stance + Jump SM.AnimStateNode_2.Jump Loop
  nodes [Output Animation Pose(Result=Sequence Player 'CubeSurfing_Jump_Loop'() {sequence=&/CubeSurfing_Jump_Loop})]
 }
 Landing {
  path=ItemFinalPoseOverride.AnimGraphNode_StateMachine_0.Stance + Jump SM.AnimStateNode_3.Landing
  nodes [Output Animation Pose(Result=Sequence Player 'CubeSurfing_Jump_Land'!(PlayRate:float=1) -> (Pose) {bLoopAnimation=false, sequence=&/CubeSurfing_Jump_Land})]
 }
 Jump PreLand {
  path=ItemFinalPoseOverride.AnimGraphNode_StateMachine_0.Stance + Jump SM.AnimStateNode_4.Jump PreLand
  nodes [Output Animation Pose(Result=Sequence Player 'CubeSurfing_Jump_Land_Antic'() {bLoopAnimation=false, sequence=&/CubeSurfing_Jump_Land_Antic})]
 }
 Transition {path=ItemFinalPoseOverride.AnimGraphNode_StateMachine_0.Stance + Jump SM.AnimStateTransitionNode_0.Transition, nodes [Result()]}
 Transition {
  path=ItemFinalPoseOverride.AnimGraphNode_StateMachine_0.Stance + Jump SM.AnimStateTransitionNode_1.Transition
  nodes [Result(bCanEnterTransition=Get IsJump() -> (IsJump:bool))]
 }
 Transition {
  path=ItemFinalPoseOverride.AnimGraphNode_StateMachine_0.Stance + Jump SM.AnimStateTransitionNode_2.Transition
  nodes [Result(bCanEnterTransition=Get bIsPreLand() -> (bIsPreLand:bool))]
 }
 Transition {
  path=ItemFinalPoseOverride.AnimGraphNode_StateMachine_0.Stance + Jump SM.AnimStateTransitionNode_3.Transition
  nodes [Result(bCanEnterTransition=Get bIsFall() -> (bIsFall:bool))]
 }
 Transition {
  path=ItemFinalPoseOverride.AnimGraphNode_StateMachine_0.Stance + Jump SM.AnimStateTransitionNode_4.Transition
  nodes [Result(bCanEnterTransition=Get bIsLanding() -> (bIsLanding:bool))]
 }
 Transition {path=ItemFinalPoseOverride.AnimGraphNode_StateMachine_0.Stance + Jump SM.AnimStateTransitionNode_5.Transition, nodes [Result()]}
 Transition {path=ItemFinalPoseOverride.AnimGraphNode_StateMachine_0.Stance + Jump SM.AnimStateTransitionNode_6.Transition, nodes [Result()]}
 Transition {
  path=ItemFinalPoseOverride.AnimGraphNode_StateMachine_0.Stance + Jump SM.AnimStateTransitionNode_9.Transition
  nodes [Result(bCanEnterTransition=Get IsJump())]
 }
 OnFoot/Activate/Surfing SM {
  entry=Passthrough
  path=ItemFinalPoseOverride.AnimGraphNode_StateMachine_1.OnFoot/Activate/Surfing SM
  ids=id
  nodes [
   AnimStateNode_0 = Passthrough() {
    is=title AnimStateNode
    stateEntered {guid=#10}
    stateFullyBlended {guid=#11}
    stateLeft {guid=#12}
   }
   AnimStateNode_1 = Surfing() {
    is=title AnimStateNode
    stateEntered {guid=#13}
    stateFullyBlended {guid=#14}
    stateLeft {guid=#15}
   }
   AnimStateNode_2 = TacSprint() {
    is=title AnimStateNode
    stateEntered {guid=#16}
    stateFullyBlended {guid=#17}
    stateLeft {guid=#18}
   }
   AnimStateNode_3 = Activate Montage Playing() {is=title AnimStateNode}
   AnimStateTransitionNode_0 = AnimStateTransitionNode() {crossfadeDuration=0.1}
   AnimStateTransitionNode_1 = AnimStateTransitionNode() {crossfadeDuration=0}
   AnimStateTransitionNode_2 = AnimStateTransitionNode() {crossfadeDuration=0.4}
   AnimStateTransitionNode_3 = AnimStateTransitionNode() {crossfadeDuration=0}
   AnimStateTransitionNode_4 = AnimStateTransitionNode() {crossfadeDuration=0.4}
   AnimStateTransitionNode_5 = AnimStateTransitionNode()
   AnimStateTransitionNode_7 = AnimStateTransitionNode()
  ]
  links [
   {
    dst=Surfing
    rule=ItemFinalPoseOverride.AnimGraphNode_StateMachine_1.OnFoot/Activate/Surfing SM.AnimStateTransitionNode_4.Transition
    src=Activate Montage Playing
    via=AnimStateTransitionNode_4
   }
   {
    dst=Activate Montage Playing
    rule=ItemFinalPoseOverride.AnimGraphNode_StateMachine_1.OnFoot/Activate/Surfing SM.AnimStateTransitionNode_3.Transition
    src=Passthrough
    via=AnimStateTransitionNode_3
   }
   {
    dst=Surfing
    rule=ItemFinalPoseOverride.AnimGraphNode_StateMachine_1.OnFoot/Activate/Surfing SM.AnimStateTransitionNode_7.Transition
    src=Passthrough
    via=AnimStateTransitionNode_7
   }
   {
    dst=TacSprint
    rule=ItemFinalPoseOverride.AnimGraphNode_StateMachine_1.OnFoot/Activate/Surfing SM.AnimStateTransitionNode_0.Transition
    src=Passthrough
    via=AnimStateTransitionNode_0
   }
   {
    dst=Passthrough
    rule=ItemFinalPoseOverride.AnimGraphNode_StateMachine_1.OnFoot/Activate/Surfing SM.AnimStateTransitionNode_5.Transition
    src=Surfing
    via=AnimStateTransitionNode_5
   }
   {
    dst=Activate Montage Playing
    rule=ItemFinalPoseOverride.AnimGraphNode_StateMachine_1.OnFoot/Activate/Surfing SM.AnimStateTransitionNode_1.Transition
    src=TacSprint
    via=AnimStateTransitionNode_1
   }
   {
    dst=Passthrough
    rule=ItemFinalPoseOverride.AnimGraphNode_StateMachine_1.OnFoot/Activate/Surfing SM.AnimStateTransitionNode_2.Transition
    src=TacSprint
    via=AnimStateTransitionNode_2
   }
  ]
 }
 Passthrough {
  path=ItemFinalPoseOverride.AnimGraphNode_StateMachine_1.OnFoot/Activate/Surfing SM.AnimStateNode_0.Passthrough
  nodes [
   n1 = "Transform (Modify) Bone - Bone: weapon_r"(Alpha=1, Translation=Get ActorLoc() -> (ActorLoc:Vector), Rotation="0,0,0", Scale="1,1,1") {
    rotationSpace=BCS_WorldSpace
    translationMode=BMM_Replace
    translationSpace=BCS_WorldSpace
    alphaBoolBlend {blendOption=Linear}
    boneToModify {boneName=weapon_r}
   }
   n2 = "Transform (Modify) Bone - Bone: weapon_r"(ComponentPose=@n1, Alpha=1, Translation="-45,0,60", Rotation=Get ActorRotInterp() -> (ActorRotInterp:Rotator), Scale="1,1,1") {
    rotationMode=BMM_Replace
    rotationSpace=BCS_WorldSpace
    translationMode=BMM_Additive
    alphaBoolBlend {blendOption=Linear}
    boneToModify {boneName=weapon_r}
    translation {x=-45, z=60}
   }
   "Transform (Modify) Bone - Bone: weapon_r"(ComponentPose=@n2, Alpha=1, Translation=Get ActorLocAdd() -> (ActorLocAdd:Vector), Rotation="0,0,0", Scale="1,1,1") {
    rotationSpace=BCS_WorldSpace
    translationMode=BMM_Additive
    translationSpace=BCS_WorldSpace
    alphaBoolBlend {blendOption=Linear}
    boneToModify {boneName=weapon_r}
   }
   Spring controller - Bone: weapon_r!(ComponentPose:Component Space Pose Link Structure, Alpha:float=1) -> (Pose:Component Space Pose Link Structure) {
    is=title AnimGraphNode_SpringBone
    bLimitDisplacement=true
    bTranslateZ=false
    maxDisplacement=5
    springDamping=200
    springStiffness=1500
    alphaBoolBlend {blendOption=Linear}
    springBone {boneName=weapon_r}
   }
   Spring controller - Bone: weapon_r(Alpha=1) {
    bLimitDisplacement=true
    bTranslateX=false
    bTranslateY=false
    maxDisplacement=8
    springDamping=200
    springStiffness=3000
    alphaBoolBlend {blendOption=Linear}
    springBone {boneName=weapon_r}
   }
   n3 = "Transform (Modify) Bone - Bone: weapon_r"(
    ComponentPose=Local To Component(LocalPose=Use cached pose 'FinalInputPose'())
    Alpha=1
    Translation="0,0,5"
    Rotation=Get ActorRotInterp()
    Scale="1,1,1"
   ) {
    rotationMode=BMM_Replace
    rotationSpace=BCS_WorldSpace
    translationMode=BMM_Additive
    translationSpace=BCS_WorldSpace
    alphaBoolBlend {blendOption=Linear}
    boneToModify {boneName=weapon_r}
    translation {z=5}
   }
   n4 = Spring controller - Bone: weapon_r!(
    ComponentPose=@n3
    Alpha=1
    SpringStiffness:double=Get SpringStiffnessZ() -> (SpringStiffnessZ:float)
    SpringDamping:double=Get SpringDampingZ() -> (SpringDampingZ:float)
   ) -> (Pose) {
    bLimitDisplacement=true
    bTranslateX=false
    bTranslateY=false
    maxDisplacement=150
    springDamping=100
    springStiffness=2500
    alphaBoolBlend {blendOption=Linear}
    springBone {boneName=weapon_r}
   }
   n5 = Spring controller - Bone: weapon_r(
    ComponentPose=@n4
    Alpha=1
    SpringStiffness=Get SpringStiffnessXY() -> (SpringStiffnessXY:float)
    SpringDamping=Get SpringDampingXY() -> (SpringDampingXY:float)
   ) {
    bLimitDisplacement=true
    bTranslateZ=false
    maxDisplacement=75
    springDamping=200
    springStiffness=3000
    alphaBoolBlend {blendOption=Linear}
    springBone {boneName=weapon_r}
   }
   n6 = Component To Local(ComponentPose=@n5)
   n7 = "Transform (Modify) Bone - Bone: weapon_r"(
    ComponentPose=Local To Component(LocalPose=Use cached pose 'FinalInputPose'())
    Alpha=1
    Translation="0,0,5"
    Rotation=Get ActorRotInterp()
    Scale="1,1,1"
   ) {
    rotationMode=BMM_Replace
    rotationSpace=BCS_WorldSpace
    translationMode=BMM_Additive
    translationSpace=BCS_WorldSpace
    alphaBoolBlend {blendOption=Linear}
    boneToModify {boneName=weapon_r}
    translation {z=5}
   }
   n8 = Spring controller - Bone: weapon_r(ComponentPose=@n7, Alpha=1, SpringStiffness=Get SpringStiffnessZ(), SpringDamping=Get SpringDampingZ()) {
    bLimitDisplacement=true
    bTranslateX=false
    bTranslateY=false
    maxDisplacement=150
    springDamping=100
    springStiffness=2500
    alphaBoolBlend {blendOption=Linear}
    springBone {boneName=weapon_r}
   }
   n9 = Component To Local(ComponentPose=@n8)
   n10 = Get bIsWalking() -> (bIsWalking:bool)
   Output Animation Pose(Result=Blend Poses by bool(
     BlendPose_0=Use cached pose 'FinalInputPose'()
     BlendPose_1=Blend Poses by bool(BlendPose_0=@n6, BlendPose_1=@n9, BlendTime_0=0.2, BlendTime_1=0.2, bActiveValue=@n10) {blendType=Linear}
     bActiveValue=Get bActorHiddenInGame() -> (bActorHiddenInGame:bool)
    ) {blendType=Linear})
  ]
 }
 Surfing {
  path=ItemFinalPoseOverride.AnimGraphNode_StateMachine_1.OnFoot/Activate/Surfing SM.AnimStateNode_1.Surfing
  nodes [
   n1 = "Transform (Modify) Bone - Bone: attach"(Alpha=1, Translation=Get AttachLocation() -> (AttachLocation:Vector), Rotation="0,0,0", Scale="1,1,1") {translationMode=BMM_Replace, translationSpace=BCS_WorldSpace, alphaBoolBlend {blendOption=Linear}, boneToModify {boneName=attach}}
   Component To Local(ComponentPose=@n1)
   n2 = "Transform (Modify) Bone - Bone: pelvis"(ComponentPose=Local To Component(LocalPose=Use cached pose 'Surfing Pose'()), Alpha=1, Translation="0,0,30", Rotation="0,0,0", Scale="1,1,1") {translationMode=BMM_Additive, alphaBoolBlend {blendOption=Linear}, boneToModify {boneName=pelvis}, translation {z=30}}
   n3 = "Transform (Modify) Bone - Bone: attach"(ComponentPose=@n2, Alpha=1, Translation="0,0,30", Rotation="0,0,0", Scale="1,1,1") {
    translationMode=BMM_Additive
    translationSpace=BCS_WorldSpace
    alphaBoolBlend {blendOption=Linear}
    boneToModify {boneName=attach}
    translation {z=30}
   }
   Output Animation Pose(Result=Component To Local(ComponentPose=@n3))
  ]
 }
 TacSprint {
  path=ItemFinalPoseOverride.AnimGraphNode_StateMachine_1.OnFoot/Activate/Surfing SM.AnimStateNode_2.TacSprint
  nodes [
   n1 = Spring controller - Bone: weapon_r!(ComponentPose, Alpha=1) -> (Pose) {
    bLimitDisplacement=true
    bTranslateX=false
    bTranslateY=false
    maxDisplacement=8
    springDamping=200
    springStiffness=500
    alphaBoolBlend {blendOption=Linear}
    springBone {boneName=weapon_r}
   }
   Spring controller - Bone: weapon_r(ComponentPose=@n1, Alpha=1) {
    bLimitDisplacement=true
    bTranslateZ=false
    maxDisplacement=7
    springDamping=100
    springStiffness=500
    alphaBoolBlend {blendOption=Linear}
    springBone {boneName=weapon_r}
   }
   n2 = Two Way Blend(
    A=Layered blend per bone(
     BasePose=Use cached pose 'FinalInputPose'()
     BlendPoses_0=Sequence Player 'CubeSurfing_Relaxed_CMM'() {bLoopAnimation=false, sequence=/CubeSurfing/Assets/Anim/Layering/CubeSurfing_Relaxed_CMM}
    ) {bMeshSpaceRotationBlend=true, layerSetup [{branchFilters [{boneName=clavicle_r}]}]}
    B=Layered blend per bone(
     BasePose=Use cached pose 'FinalInputPose'()
     BlendPoses_0=Sequence Player 'CubeSurfing_Relaxed_CMM'() {bLoopAnimation=false, sequence=&CubeSurfing_Relaxed_CMM}
    ) {layerSetup [{branchFilters [{boneName=clavicle_r}]}]}
    Alpha=0.5
   ) {blendNode {alphaBoolBlend {blendOption=Linear}}}
   n3 = Use cached pose 'FinalInputPose'()
   n4 = Get bBypassArmAdditive() -> (bBypassArmAdditive:bool)
   n5 = "Transform (Modify) Bone - Bone: weapon_r"(
    ComponentPose=Local To Component(LocalPose=Blend Poses by bool(BlendPose_0=@n3, BlendPose_1=@n2, BlendTime_1=0.3, bActiveValue=@n4) {blendType=Linear})
    Alpha=1
    Translation="0,0,5"
    Rotation=Get ActorRotInterp()
    Scale="1,1,1"
   ) {
    rotationMode=BMM_Replace
    rotationSpace=BCS_WorldSpace
    translationMode=BMM_Additive
    translationSpace=BCS_WorldSpace
    alphaBoolBlend {blendOption=Linear}
    boneToModify {boneName=weapon_r}
    translation {z=5}
   }
   n6 = Spring controller - Bone: weapon_r!(ComponentPose=@n5, Alpha=1, SpringStiffness:double=Get SpringStiffnessZ(), SpringDamping:double=Get SpringDampingZ()) -> (Pose) {
    bLimitDisplacement=true
    bTranslateX=false
    bTranslateY=false
    maxDisplacement=150
    springDamping=100
    springStiffness=2500
    alphaBoolBlend {blendOption=Linear}
    springBone {boneName=weapon_r}
   }
   n7 = Spring controller - Bone: weapon_r(ComponentPose=@n6, Alpha=1, SpringStiffness=Get SpringStiffnessXY(), SpringDamping=Get SpringDampingXY()) {
    bLimitDisplacement=true
    bTranslateZ=false
    maxDisplacement=75
    springDamping=200
    springStiffness=3000
    alphaBoolBlend {blendOption=Linear}
    springBone {boneName=weapon_r}
   }
   Output Animation Pose(Result=Component To Local(ComponentPose=@n7))
  ]
 }
 Activate Montage Playing {
  path=ItemFinalPoseOverride.AnimGraphNode_StateMachine_1.OnFoot/Activate/Surfing SM.AnimStateNode_3.Activate Montage Playing
  nodes [Output Animation Pose(Result=Use cached pose 'FinalInputPose'())]
 }
 Transition {
  path=ItemFinalPoseOverride.AnimGraphNode_StateMachine_1.OnFoot/Activate/Surfing SM.AnimStateTransitionNode_0.Transition
  nodes [Result(bCanEnterTransition=Get bIsTacSprint() -> (bIsTacSprint:bool))]
 }
 Transition {
  path=ItemFinalPoseOverride.AnimGraphNode_StateMachine_1.OnFoot/Activate/Surfing SM.AnimStateTransitionNode_1.Transition
  nodes [Result(bCanEnterTransition=Get bIsActivateMontage() -> (bIsActivateMontage:bool))]
 }
 Transition {
  path=ItemFinalPoseOverride.AnimGraphNode_StateMachine_1.OnFoot/Activate/Surfing SM.AnimStateTransitionNode_2.Transition
  nodes [Result(bCanEnterTransition=NOT Boolean(A=Get bIsTacSprint()))]
 }
 Transition {
  path=ItemFinalPoseOverride.AnimGraphNode_StateMachine_1.OnFoot/Activate/Surfing SM.AnimStateTransitionNode_3.Transition
  nodes [Result(bCanEnterTransition=Get bIsActivateMontage())]
 }
 Transition {
  path=ItemFinalPoseOverride.AnimGraphNode_StateMachine_1.OnFoot/Activate/Surfing SM.AnimStateTransitionNode_4.Transition
  nodes [Result(bCanEnterTransition=NOT Boolean(A=Get bIsActivateMontage()))]
 }
 Transition {
  path=ItemFinalPoseOverride.AnimGraphNode_StateMachine_1.OnFoot/Activate/Surfing SM.AnimStateTransitionNode_5.Transition
  nodes [Result(bCanEnterTransition=NOT Boolean(A=Get bIsSurfing() -> (bIsSurfing:bool)))]
 }
 Transition {
  path=ItemFinalPoseOverride.AnimGraphNode_StateMachine_1.OnFoot/Activate/Surfing SM.AnimStateTransitionNode_7.Transition
  nodes [Result(bCanEnterTransition=Get bIsSurfing())]
 }
 Boost SM {
  entry=Passthrough
  path=ItemFinalPoseOverride.AnimGraphNode_StateMachine_3.Boost SM
  ids=id
  nodes [
   AnimStateNode_0 = Passthrough()
   AnimStateNode_4 = Boost Intro() {is=title AnimStateNode}
   AnimStateNode_5 = Boost Loop() {is=title AnimStateNode}
   AnimStateNode_6 = Boost Outro() {is=title AnimStateNode}
   AnimStateTransitionNode_0 = AnimStateTransitionNode()
   AnimStateTransitionNode_5 = AnimStateTransitionNode() {bAutomaticRuleBasedOnSequencePlayerInState=true}
   AnimStateTransitionNode_6 = AnimStateTransitionNode() {bAutomaticRuleBasedOnSequencePlayerInState=true}
   AnimStateTransitionNode_7 = AnimStateTransitionNode()
  ]
  links [
   {
    dst=Boost Loop
    rule=ItemFinalPoseOverride.AnimGraphNode_StateMachine_3.Boost SM.AnimStateTransitionNode_5.AnimationTransitionGraph_1
    src=Boost Intro
    via=AnimStateTransitionNode_5
   }
   {
    dst=Boost Outro
    rule=ItemFinalPoseOverride.AnimGraphNode_StateMachine_3.Boost SM.AnimStateTransitionNode_7.AnimationTransitionGraph_1
    src=Boost Loop
    via=AnimStateTransitionNode_7
   }
   {
    dst=Passthrough
    rule=ItemFinalPoseOverride.AnimGraphNode_StateMachine_3.Boost SM.AnimStateTransitionNode_6.Transition
    src=Boost Outro
    via=AnimStateTransitionNode_6
   }
   {
    dst=Boost Intro
    rule=ItemFinalPoseOverride.AnimGraphNode_StateMachine_3.Boost SM.AnimStateTransitionNode_0.Transition
    src=Passthrough
    via=AnimStateTransitionNode_0
   }
  ]
 }
 Passthrough {
  path=ItemFinalPoseOverride.AnimGraphNode_StateMachine_3.Boost SM.AnimStateNode_0.Passthrough
  nodes [Output Animation Pose(Result=Use cached pose 'PreBoost Pose'())]
 }
 Boost Intro {
  path=ItemFinalPoseOverride.AnimGraphNode_StateMachine_3.Boost SM.AnimStateNode_4.Boost Intro
  nodes [Output Animation Pose(Result=Sequence Player 'CubeSurfing_Boost_Intro'!(StartPosition:float=0.4) -> (Pose) {bLoopAnimation=false, sequence=&/CubeSurfing_Boost_Intro})]
 }
 Boost Loop {
  path=ItemFinalPoseOverride.AnimGraphNode_StateMachine_3.Boost SM.AnimStateNode_5.Boost Loop
  nodes [Output Animation Pose(Result=Apply Additive(
     Base=Sequence Player 'CubeSurfing_Boost_Loop'() {sequence=&/CubeSurfing_Boost_Loop}
     Additive=Blendspace Player 'CubeSurfing_Boost_Turn_BS'(X=0, Y=0) {blendSpace=&/CubeSurfing_Boost_Turn_BS}
    ) {alphaBoolBlend {blendOption=Linear}})]
 }
 Boost Outro {
  path=ItemFinalPoseOverride.AnimGraphNode_StateMachine_3.Boost SM.AnimStateNode_6.Boost Outro
  nodes [Output Animation Pose(Result=Sequence Player 'CubeSurfing_Boost_Outro'() {bLoopAnimation=false, sequence=&/CubeSurfing_Boost_Outro})]
 }
 Transition {
  path=ItemFinalPoseOverride.AnimGraphNode_StateMachine_3.Boost SM.AnimStateTransitionNode_0.Transition
  nodes [Result(bCanEnterTransition=Get bIsPreBoost() -> (bIsPreBoost:bool))]
 }
 AnimationTransitionGraph_1 {path=ItemFinalPoseOverride.AnimGraphNode_StateMachine_3.Boost SM.AnimStateTransitionNode_5.AnimationTransitionGraph_1, nodes [Result()]}
 Transition {path=ItemFinalPoseOverride.AnimGraphNode_StateMachine_3.Boost SM.AnimStateTransitionNode_6.Transition, nodes [Result()]}
 AnimationTransitionGraph_1 {
  path=ItemFinalPoseOverride.AnimGraphNode_StateMachine_3.Boost SM.AnimStateTransitionNode_7.AnimationTransitionGraph_1
  nodes [Result(bCanEnterTransition=NOT Boolean(A=Get bIsBoost() -> (bIsBoost:bool)))]
 }
 ItemFullBodyAimOffsets {nodes [
   Input Pose!() -> (Pose, FullBody_AimOffsetAlpha:double, Aim_Yaw:double, Aim_Pitch:double) {name=InPoseFullBodyAimOffsets}
   Input Pose!() -> (Pose) {name=InPosePreFullBodyAimOffsets}
   Output Pose() {layerGroup=Item Override}
  ]}
 ItemFullBodyOverride {nodes [Output Pose(Result=Input Pose() {name=InPoseFullBody}) {layerGroup=Item Override}]}
 ItemIdleOverride {nodes [
   Input Pose() {name=InPose_Idle}
   Output Pose() {layerGroup=Item Override}
  ]}
 ItemJetPackJumpAdditive {nodes [Output Pose() {layerGroup=Item Override}]}
 ItemJetPackStartAdditive {nodes [Output Pose() {layerGroup=Item Override}]}
 ItemJumpLoopAdditive {nodes [
   Input Pose() {name=InPoseItemJumpLoopAdditive}
   Output Pose() {layerGroup=Item Override}
  ]}
 ItemJumpOverride {nodes [
   Input Pose() {name=InPoseJump}
   Output Pose() {layerGroup=Item Override}
  ]}
 ItemJumpUpAdditive {nodes [
   Input Pose() {name=InPoseItemJumpUpAdditive}
   Output Pose() {layerGroup=Item Override}
  ]}
 ItemPreAimOffsets {nodes [
   Input Pose() {name=InPosePreAimOffsets}
   Output Pose() {layerGroup=Item Override}
  ]}
 ItemPreIK {nodes [
   Input Pose() {name=InPose_PreIK}
   Output Pose() {layerGroup=Item Override}
  ]}
 ItemRightHandIK {nodes [
   Input Pose() {name=InPosePreRightHandIK}
   Input Pose() {name=InPosePostRightHandIK}
   Output Pose() {layerGroup=Item Override}
  ]}
 ItemSkeletalControl {nodes [
   Input Pose()
   Output Pose() {layerGroup=Item Override}
  ]}
 ItemSwimJumpFallLoop {nodes [Output Pose() {layerGroup=Item Override}]}
 ItemSwimJumpSurfaceEnd {nodes [Output Pose() {layerGroup=Item Override}]}
 ItemSwimJumpSurfaceLoop {nodes [Output Pose() {layerGroup=Item Override}]}
 ItemUpperBody {nodes [
   Input Pose!() -> (Pose, InputParam:Fort Anim Input Adjusted Aim Structure) {name=InPoseUpperBody}
   Output Pose() {layerGroup=Item Override}
  ]}
 ItemZipLineStartAdditive {nodes [Output Pose() {layerGroup=Item Override}]}
 Surf Additives {nodes [
   n1 = "Transform (Modify) Bone - Bone: pelvis"(ComponentPose=Local To Component(), Alpha=1, Translation="0,-55,0", Rotation="0,0,0", Scale="1,1,1") {
    translationMode=BMM_Additive
    translationSpace=BCS_ParentBoneSpace
    alphaBoolBlend {blendOption=Linear}
    boneToModify {boneName=pelvis}
    translation {y=-55}
   }
   n2 = "Transform (Modify) Bone - Bone: attach"(ComponentPose=@n1, Alpha=1, Translation="0,-55,0", Rotation="0,0,0", Scale="1,1,1") {
    translationMode=BMM_Additive
    translationSpace=BCS_ParentBoneSpace
    alphaBoolBlend {blendOption=Linear}
    boneToModify {boneName=attach}
    translation {y=-55}
   }
   Component To Local(ComponentPose=@n2)
   n3 = Blendspace Player 'CubeSurfing_React_BS'(X=0, Y=0) {blendSpace=&/CubeSurfing_React_BS}
   n4 = Blendspace Player 'CubeSurfing_Lean_BS_Fast'(X=0, Y=0) {blendSpace=&/CubeSurfing_Lean_BS_Fast}
   n5 = Blendspace Player 'CubeSurfing_Lean_BS_Slow'(X=0, Y=0) {blendSpace=&/CubeSurfing_Lean_BS_Slow}
   n6 = Two Way Blend(
    A=Apply Additive(Base=Input Pose!() -> (Pose), Additive=Two Way Blend(A=@n5, B=@n4, Alpha=0) {blendNode {alphaBoolBlend {blendOption=Linear}}}) {alphaBoolBlend {blendOption=Linear}}
    B=Blendspace Player 'CubeSurfing_Lean_BS_Fast_Sharp_Turn'(X=0, Y=0) {blendSpace=&/CubeSurfing_Lean_BS_Fast_Sharp_Turn}
    Alpha=1
   ) {blendNode {alphaBoolBlend {blendOption=Linear}}}
   Output Pose(Result=Apply Additive(
     Base=Apply Additive(Base=@n6, Additive=@n3) {alphaBoolBlend {blendOption=Linear}}
     Additive="Sequence Player (additive) 'CubeSurfing_Idle'"!(PlayRate:float=1) -> (Pose) {sequence=&/CubeSurfing_Idle}
    ) {alphaBoolBlend {blendOption=Linear}})
   Save cached pose 'SlowLeanPose'(Pose=Apply Additive(
     Base=Blendspace Player 'CubeSurfing_Lean_BS_Slow'(X=0, Y=0) {blendSpace=&CubeSurfing_Lean_BS_Slow}
     Additive=Blendspace Player 'CubeSurfing_Lean_BS_Slow'(X=0, Y=0) {blendSpace=&CubeSurfing_Lean_BS_Slow}
    ) {alphaBoolBlend {blendOption=Linear}}) {cacheName=SlowLeanPose}
   Save cached pose 'FastLeanPose'(Pose=Apply Additive(
     Base=Blendspace Player 'CubeSurfing_Lean_BS_Fast'(X=0, Y=0) {blendSpace=&CubeSurfing_Lean_BS_Fast}
     Additive=Blendspace Player 'CubeSurfing_Lean_BS_Fast'(X=0, Y=0) {blendSpace=&CubeSurfing_Lean_BS_Fast}
    ) {alphaBoolBlend {blendOption=Linear}}) {cacheName=FastLeanPose}
   Use cached pose 'SlowLeanPose'()
   Use cached pose 'FastLeanPose'()
  ]}
]
variables [
 ActorLoc {type=Vector}
 ActorLocAdd {type=Vector}
 ActorRotInterp {type=Rotator}
 IsJump {type=boolean}
]
