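; Child Blueprint that adds jump and landing audio triggers to the water movement sound in CubeSurfing.
; OnWatermovement stores the parent's returned audio component in WaterComp when movement starts and clears it when it stops.
; OnJump and OnLand call their parents, then fire JumpStart and JumpStop trigger parameters on that component if valid, so one playing sound reacts to jumps.
; BeginPlay and Tick add no logic.
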(event EventBeginPlay
  (|CallParentFunction) ; This node is disabled and will not be called.\nDrag off pins to build functionality.
)

(event EventTick
  (params (DeltaSeconds :type Float%20%28single-precision%29)))

(fn OnJump
  (|CallParentFunction)
  (bind _watercomp (Variables|Default|GetWaterComp)
    (:then
      (Audio|Parameter|ExecuteTriggerParameter _watercomp "JumpStart"))
    (:else)))

(fn OnLand
  (|CallParentFunction)
  (bind _watercomp (Variables|Default|GetWaterComp)
    (:then
      (Audio|Parameter|ExecuteTriggerParameter _watercomp "JumpStop"))
    (:else)))

(fn OnWatermovement
  (params (Starting :type Boolean))
  (bind _audiocomp (|CallParentFunction Starting))
  (if Starting
    (Variables|Default|SetWaterComp _audiocomp)
    (else
      (Variables|Default|SetWaterComp 0)))
  (return _audiocomp))
