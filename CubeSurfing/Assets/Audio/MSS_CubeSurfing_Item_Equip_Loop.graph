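# Mono looping MetaSound Source for the CubeSurfing item equip loop, routed as weapon foley with attenuation and concurrency.
# On play, it randomizes volume and pitch, then picks a wave using a no-repeat patch.
# A bool selects one of two wave players, either triggering a random start time or not; only the non-random branch receives pitch shift.
# Both players pass through an attack envelope and random gain into a mixer.
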MetaSoundSource /CubeSurfing/Assets/Audio/MSS_CubeSurfing_Item_Equip_Loop
inputs [
 Attack Time {type=Time}
 bUseRandomStartTime {type=Bool}
 Loop {type=Bool}
 Max Pitch {type=Float}
 Max Volume {type=Float}
 Min Pitch {type=Float}
 Min Volume {type=Float}
 Release Time {type=Time}
 Sounds {type=WaveAsset:Array}
 UE.Source.OnPlay {type=Trigger}
]
interfaces [UE.OutputFormat.Mono:1.0, UE.Source:1.1]
outputs ["UE.OutputFormat.Mono.Audio:0" {type=Audio}]
pages [Default {
  input_defaults {
   Attack Time=0.52
   Loop="<MetaSoundBuilderResult.SUCCEEDED: 0>"
   Max Pitch=1
   Max Volume=1
   Min Pitch=1
   Min Volume=0.9
   Release Time=0.2
   UE.Source.OnPlay=null
   bUseRandomStartTime=null
   Sounds [/CubeSurfing/Assets/Audio/Waves/Ability_CubeSurf_Item_Equip_LP]
  }
  nodes [
   ADSREnvelopeADSREnvelope_0 = ADSR Envelope.ADSR Envelope:Audio(Attack Time=^Attack Time, Release Time=^Release Time, Trigger Attack=^UE.Source.OnPlay) {also {defaults {Decay Time=1, Sustain Level=1}}}
   n1 = VariableAccessor.Audio(Variable=VariableMutator.Audio(Value=@ADSREnvelopeADSREnvelope_0.Out Envelope, Variable=InitVariable.Audio().Variable).Variable)
   n2 = UE.RandomFloat(Max=^Max Volume, Min=^Min Volume, Next=^UE.Source.OnPlay)
   n3 = UE.RandomFloat(Max=^Max Pitch, Min=^Min Pitch, Next=@n2.On Next)
   n4 = /FortMetasound/Patches/MSP_GetNoRepeatNumber(WaveArray=^Sounds, InTrigger=@n3.On Next)
   ArrayRandomGet_0 = Array.Random Get:WaveAsset:Array(In Array=^Sounds, No Repeats=@n4.NoRepeatNumber, Next=@n4.OutTrigger) {also {defaults {Enable Shared State="<MetaSoundBuilderResult.SUCCEEDED: 0>"}}}
   n5 = VariableAccessor.WaveAsset(Variable=VariableMutator.WaveAsset(Value=@ArrayRandomGet_0.Value, Variable=InitVariable.WaveAsset().Variable).Variable)
   GetWaveDuration_0 = "UE.Get Wave Duration: "(Wave=@n5.Value)
   n6 = TriggerCompare.Clamp:Bool(A=^bUseRandomStartTime, Compare=UE.Reroute(Value=@ArrayRandomGet_0.On Next).Value) {also {defaults {B="<MetaSoundBuilderResult.SUCCEEDED: 0>"}}}
   n7 = UE.RandomFloat(Max=UE.ConversionTimeToFloat(In=@GetWaveDuration_0.Duration).Out, Next=@n6.True)
   n8 = VariableAccessor.Time(Variable=VariableMutator.Time(Value=UE.ConversionFloatToTime(In=@n7.Value).Out, Variable=InitVariable.Time().Variable).Variable)
   n9 = VariableAccessor.WaveAsset(Variable=@n5.Variable)
   WavePlayer_1 = UE.Wave Player:Mono(Loop=^Loop, Play=@n7.On Next, Start Time=@n8.Value, Wave Asset=@n9.Value)
   n10 = VariableAccessor.Float(Variable=VariableMutator.Float(Variable=InitVariable.Float().Variable, Value=@n2.Value).Variable)
   n11 = UE.Reroute(Value=@n6.False)
   FrequencyMultipliertoSemitone_0 = UE.Frequency Multiplier to Semitone:Float(Frequency Multiplier=UE.Reroute(Value=@n3.Value).Value)
   n12 = VariableMutator.Float(Value=@FrequencyMultipliertoSemitone_0.Semitones, Variable=InitVariable.Float().Variable)
   WavePlayer_0 = UE.Wave Player:Mono(
    Loop=^Loop
    Play=UE.Reroute(Value=UE.Reroute(Value=@n11.Value).Value).Value
    Pitch Shift=VariableAccessor.Float(Variable=VariableAccessor.Float(Variable=@n12.Variable).Variable).Value
    Start Time=VariableAccessor.Time(Variable=@n8.Variable).Value
    Wave Asset=VariableAccessor.WaveAsset(Variable=@n9.Variable).Value
   )
   AudioMixerAudioMixerMono2_0 = "AudioMixer.Audio Mixer (Mono, 2)"(
    In 0=UE.Multiply:Audio by Float(PrimaryOperand=UE.Multiply:Audio(AdditionalOperands=@n1.Value, PrimaryOperand=@WavePlayer_1.Out Mono).Out, AdditionalOperands=@n10.Value).Out
    In 1=UE.Multiply:Audio by Float(PrimaryOperand=UE.Multiply:Audio(AdditionalOperands=@n1.Value, PrimaryOperand=@WavePlayer_0.Out Mono).Out, AdditionalOperands=@n10.Value).Out
   )
   TriggerAnyTriggerAny2_0 = "TriggerAny.Trigger Any (2)"(In 1=@WavePlayer_0.On Finished, In 0=@WavePlayer_1.On Finished)
  ]
  links [{dst="@UE.OutputFormat.Mono.Audio:0", src=AudioMixerAudioMixerMono2_0.Out}]
 }]
properties {
 assetImportData=&self:AssetImportData
 attenuationSettings=/CubeSurfing/Assets/Audio/Settings/ATT_CubeSurfing_Foley
 bLooping=true
 duration=10000
 maxDistance=2250
 sampleRate=48000
 soundClassObject=&Fort_WeaponFoley
 volume=0.2388
 concurrencySet [&CS_Fort_WeaponFoley]
}
