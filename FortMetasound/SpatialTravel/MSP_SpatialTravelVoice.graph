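# MetaSound patch for one stereo spatial-travel voice, mixing a loop layer with start and stop one-shots.
# FadeIn picks a random loop wave, optionally at a random start position, MSP_FadeToValue scales its volume, and a mapped control drives pitch.
# MSP_OneshotGate turns threshold crossings into random start/stop sounds, with optional tying to the loop's fade lifecycle.
# Manual one-shot triggers bypass that gate, and playback events are exposed as outputs.
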MetaSoundPatch /FortMetasound/SpatialTravel/MSP_SpatialTravelVoice
inputs [
 FadeIn {type=Trigger}
 FadeOut {type=Trigger}
 Loop.FadeTime {type=Float}
 Loop.PitchControl {type=Float}
 Loop.PitchhMin {type=Float}
 Loop.PitchMax {type=Float}
 Loop.RandomStartPos {type=Bool}
 Loop.Sounds {type=WaveAsset:Array}
 Loop.VolumeControl {type=Float}
 OneShot.RetriggerDelay {type=Time}
 OneShot.Starts {type=WaveAsset:Array}
 OneShot.Stops {type=WaveAsset:Array}
 OneShot.Threshold {type=Float}
 OneShot.TieWithLoop {type=Bool}
 OneShot.Value {type=Float}
 OneShot.VolumeControl {type=Float}
 TriggerStartOS {type=Trigger}
 TriggerStopOS {type=Trigger}
]
outputs [
 On Loop Finished {type=Trigger}
 On Play {type=Trigger}
 OutL {type=Audio}
 OutR {type=Audio}
 StartOneShotTriggered {type=Trigger}
 StopOneShotTriggered {type=Trigger}
]
pages [Default {
  input_defaults {
   FadeIn=null
   FadeOut=null
   Loop.FadeTime=0
   Loop.PitchControl=null
   Loop.PitchMax=1
   Loop.PitchhMin=0
   Loop.RandomStartPos="<MetaSoundBuilderResult.SUCCEEDED: 0>"
   Loop.VolumeControl=1
   OneShot.RetriggerDelay=1
   OneShot.Threshold=0.5
   OneShot.TieWithLoop="<MetaSoundBuilderResult.SUCCEEDED: 0>"
   OneShot.Value=1
   OneShot.VolumeControl=1
   TriggerStartOS=null
   TriggerStopOS=null
   Loop.Sounds []
   OneShot.Starts []
   OneShot.Stops []
  }
  nodes [
   n1 = &MSP_OneshotGate(RetriggerDelay=^OneShot.RetriggerDelay, OneshotThreshold=^OneShot.Threshold, Value=^OneShot.Value)
   n2 = TriggerCompare.Clamp:Bool(Compare=^FadeIn, A=^OneShot.TieWithLoop) {also {defaults {B="<MetaSoundBuilderResult.SUCCEEDED: 0>"}}}
   ArrayRandomGet_2 = Array.Random Get:WaveAsset:Array(Next=^FadeIn, In Array=^Loop.Sounds) {also {defaults {Weights []}}}
   n3 = UE.Reroute(Value=@ArrayRandomGet_2.Value)
   GetWaveDuration_0 = "UE.Get Wave Duration: "(Wave=@n3.Value)
   n4 = TriggerCompare.Clamp:Bool(A=^Loop.RandomStartPos, Compare=@ArrayRandomGet_2.On Next) {also {defaults {B="<MetaSoundBuilderResult.SUCCEEDED: 0>"}}}
   n5 = UE.RandomTime(Max=@GetWaveDuration_0.Duration, Next=@n4.False)
   TriggerRouteTriggerRouteTime2_0 = "TriggerRoute.Trigger Route (Time, 2):Time"(Set 1=@n5.On Next, Value 1=@n5.Value, Set 0=@n4.True) {also {defaults {Value 0=0}}}
   n6 = /FortMetasound/Patches/MSP_FadeToValue(ReverseFade=^FadeOut, FadeTime=UE.ConversionFloatToTime(In=^Loop.FadeTime).Out, Fade=@TriggerRouteTriggerRouteTime2_0.On Set)
   n7 = UE.Reroute(Value=@n6.OnDoneReverseFade)
   n8 = VariableMutator.Trigger(Variable=InitVariable.Trigger().Variable, Value=@n7.Value)
   n9 = Value.Value:Bool(
    Reset=TriggerCompare.Clamp:Bool(A=^OneShot.TieWithLoop, Compare=VariableAccessor.Trigger(Variable=@n8.Variable).Value) {also {defaults {B="<MetaSoundBuilderResult.SUCCEEDED: 0>"}}}.True
    Set=@n2.True
   )
   TriggerRouteTriggerRouteBool2_0 = "TriggerRoute.Trigger Route (Bool, 2):Bool"(Set 0=Value.Value:Bool(Set=@n2.False).On Set, Set 1=@n9.On Set, Value 1=@n9.Output Value) {also {defaults {Value 0="<MetaSoundBuilderResult.SUCCEEDED: 0>"}}}
   n10 = VariableAccessor.Bool(Variable=VariableMutator.Bool(Variable=InitVariable.Bool().Variable, Value=@TriggerRouteTriggerRouteBool2_0.Value).Variable)
   TriggerAnyTriggerAny2_1 = "TriggerAny.Trigger Any (2)"(
    In 1=^TriggerStopOS
    In 0=TriggerCompare.Clamp:Bool(Compare=@n1.OnGateExit, A=@n10.Value) {also {defaults {B="<MetaSoundBuilderResult.SUCCEEDED: 0>"}}}.True
   )
   ArrayRandomGet_0 = Array.Random Get:WaveAsset:Array(In Array=^OneShot.Stops, Next=@TriggerAnyTriggerAny2_1.Out) {also {defaults {Weights []}}}
   WavePlayer_0 = UE.Wave Player:Stereo(Play=@ArrayRandomGet_0.On Next, Wave Asset=@ArrayRandomGet_0.Value) {also {version [1,1]}}
   TriggerAnyTriggerAny2_0 = "TriggerAny.Trigger Any (2)"(
    In 1=^TriggerStartOS
    In 0=TriggerCompare.Clamp:Bool(Compare=@n1.OnGateEnter, A=@n10.Value) {also {defaults {B="<MetaSoundBuilderResult.SUCCEEDED: 0>"}}}.True
   )
   ArrayRandomGet_1 = Array.Random Get:WaveAsset:Array(In Array=^OneShot.Starts, Next=@TriggerAnyTriggerAny2_0.Out)
   WavePlayer_1 = UE.Wave Player:Stereo(Play=@ArrayRandomGet_1.On Next, Wave Asset=@ArrayRandomGet_1.Value) {also {version [1,1]}}
   AudioMixerAudioMixerStereo2_1 = "AudioMixer.Audio Mixer (Stereo, 2)"(In 1 L=@WavePlayer_0.Out Left, In 1 R=@WavePlayer_0.Out Right, In 0 L=@WavePlayer_1.Out Left, In 0 R=@WavePlayer_1.Out Right)
   n11 = VariableMutator.Float(Variable=InitVariable.Float().Variable, Value=@n6.Out Value)
   n12 = InitVariable.Float()
   n13 = MapRange.MapRange:Float(In=^Loop.PitchControl, Out Range B=^Loop.PitchMax, Out Range A=^Loop.PitchhMin)
   WavePlayer_2 = UE.Wave Player:Stereo(
    Wave Asset=UE.Reroute(Value=UE.Reroute(Value=@n3.Value).Value).Value
    Start Time=UE.Reroute(Value=UE.Reroute(Value=@TriggerRouteTriggerRouteTime2_0.Value).Value).Value
    Stop=@n7.Value
    Play=@TriggerRouteTriggerRouteTime2_0.On Set
    Pitch Shift=VariableAccessor.Float(Variable=VariableMutator.Float(Variable=@n12.Variable, Value=@n13.Out Value).Variable).Value
   ) {also {defaults {Loop="<MetaSoundBuilderResult.SUCCEEDED: 0>"}, version [1,1]}}
   AudioMixerAudioMixerStereo2_0 = "AudioMixer.Audio Mixer (Stereo, 2)"(
    Gain 1=^OneShot.VolumeControl
    In 1 L=@AudioMixerAudioMixerStereo2_1.Out L
    In 1 R=@AudioMixerAudioMixerStereo2_1.Out R
    Gain 0=UE.Multiply:Float(AdditionalOperands=^Loop.VolumeControl, PrimaryOperand=VariableAccessor.Float(Variable=@n11.Variable).Value).Out
    In 0 L=@WavePlayer_2.Out Left
    In 0 R=@WavePlayer_2.Out Right
   )
   InitVariable.Float()
  ]
  links [
   {dst="@OutL", src=AudioMixerAudioMixerStereo2_0.Out L}
   {dst="@OutR", src=AudioMixerAudioMixerStereo2_0.Out R}
   {dst="@StopOneShotTriggered", src=WavePlayer_0.On Play}
   {dst="@StartOneShotTriggered", src=WavePlayer_1.On Play}
   {dst="@On Loop Finished", src=WavePlayer_2.On Finished}
   {dst="@On Play", src=WavePlayer_2.On Play}
  ]
 }]
