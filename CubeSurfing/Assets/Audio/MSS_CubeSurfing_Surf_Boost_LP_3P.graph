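# Stereo MetaSound for the looping third-person CubeSurfing surf boost sound.
# On play it picks a boost wave with random pitch and volume, shaped by an ADSR envelope, with an optional random start time.
# A wind loop layer fades in through a SpatialTravel voice, its volume driven by the absolute normalized movement elevation.
# Both layers mix to stereo under in-air attenuation, with concurrency capped at one per owner.
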MetaSoundSource /CubeSurfing/Assets/Audio/MSS_CubeSurfing_Surf_Boost_LP_3P
inputs [
 AirLoops {type=WaveAsset:Array}
 Attack Time {type=Time}
 bUseRandomStartTime {type=Bool}
 Loop {type=Bool}
 Max Pitch {type=Float}
 Max Volume {type=Float}
 Min Pitch {type=Float}
 Min Volume {type=Float}
 Movement.Speed {type=Float}
 Movement.SpeedAzimuth {type=Float}
 Movement.SpeedElevation {type=Float}
 Release Time {type=Time}
 Sounds {type=WaveAsset:Array}
 UE.Source.OnPlay {type=Trigger}
 VolumeMultiplier {type=Float}
]
interfaces [Fort.Movement:1.0, UE.OutputFormat.Stereo:1.0, UE.Source.OneShot:1.0, UE.Source:1.1]
outputs [
 "UE.OutputFormat.Stereo.Audio:0" {type=Audio}
 "UE.OutputFormat.Stereo.Audio:1" {type=Audio}
 UE.Source.OneShot.OnFinished {type=Trigger}
]
pages [Default {
  input_defaults {
   Attack Time=0.2
   Loop="<MetaSoundBuilderResult.SUCCEEDED: 0>"
   Max Pitch=1
   Max Volume=1
   Min Pitch=1
   Min Volume=0.9
   Movement.Speed=0
   Movement.SpeedAzimuth=0
   Movement.SpeedElevation=0
   Release Time=0.2
   UE.Source.OnPlay=null
   VolumeMultiplier=2
   bUseRandomStartTime=null
   AirLoops [/Game/Athena/Sounds/Wind_Falling/AMB_Wind_Skydive_Bank_Right_Loop_01]
   Sounds [/CubeSurfing/Assets/Audio/Waves/Ability_CubeSurfing_Boost_Loop_01-glued]
  }
  nodes [
   n1 = UE.RandomFloat(Max=^Max Volume, Min=^Min Volume, Next=^UE.Source.OnPlay)
   n2 = UE.RandomFloat(Max=^Max Pitch, Min=^Min Pitch, Next=@n1.On Next)
   n3 = /FortMetasound/Patches/MSP_GetNoRepeatNumber(WaveArray=^Sounds, InTrigger=@n2.On Next)
   ArrayRandomGet_0 = Array.Random Get:WaveAsset:Array(In Array=^Sounds, No Repeats=@n3.NoRepeatNumber, Next=@n3.OutTrigger) {also {defaults {Enable Shared State="<MetaSoundBuilderResult.SUCCEEDED: 0>"}}}
   n4 = VariableAccessor.WaveAsset(Variable=VariableMutator.WaveAsset(Value=@ArrayRandomGet_0.Value, Variable=InitVariable.WaveAsset().Variable).Variable)
   GetWaveDuration_0 = "UE.Get Wave Duration: "(Wave=@n4.Value)
   n5 = TriggerCompare.Clamp:Bool(A=^bUseRandomStartTime, Compare=UE.Reroute(Value=@ArrayRandomGet_0.On Next).Value) {also {defaults {B="<MetaSoundBuilderResult.SUCCEEDED: 0>"}}}
   n6 = UE.RandomFloat(Max=UE.ConversionTimeToFloat(In=@GetWaveDuration_0.Duration).Out, Next=@n5.True)
   FrequencyMultipliertoSemitone_0 = UE.Frequency Multiplier to Semitone:Float(Frequency Multiplier=UE.Reroute(Value=@n2.Value).Value)
   n7 = VariableAccessor.Float(Variable=VariableMutator.Float(Value=@FrequencyMultipliertoSemitone_0.Semitones, Variable=InitVariable.Float().Variable).Variable)
   n8 = VariableAccessor.Time(Variable=VariableMutator.Time(Value=UE.ConversionFloatToTime(In=@n6.Value).Out, Variable=InitVariable.Time().Variable).Variable)
   n9 = VariableAccessor.WaveAsset(Variable=@n4.Variable)
   WavePlayer_1 = UE.Wave Player:Stereo(Loop=^Loop, Play=@n6.On Next, Pitch Shift=@n7.Value, Start Time=@n8.Value, Wave Asset=@n9.Value)
   n10 = UE.Reroute(Value=@WavePlayer_1.Out Left)
   ADSREnvelopeADSREnvelope_0 = ADSR Envelope.ADSR Envelope:Audio(Attack Time=^Attack Time, Release Time=^Release Time, Trigger Attack=^UE.Source.OnPlay) {also {defaults {Decay Time=1, Sustain Level=1}}}
   n11 = VariableAccessor.Audio(Variable=VariableMutator.Audio(Value=@ADSREnvelopeADSREnvelope_0.Out Envelope, Variable=InitVariable.Audio().Variable).Variable)
   n12 = VariableAccessor.Float(Variable=VariableMutator.Float(Variable=InitVariable.Float().Variable, Value=@n1.Value).Variable)
   n13 = UE.Reroute(Value=@n5.False)
   WavePlayer_0 = UE.Wave Player:Stereo(
    Loop=^Loop
    Play=UE.Reroute(Value=UE.Reroute(Value=@n13.Value).Value).Value
    Pitch Shift=VariableAccessor.Float(Variable=@n7.Variable).Value
    Start Time=VariableAccessor.Time(Variable=@n8.Variable).Value
    Wave Asset=VariableAccessor.WaveAsset(Variable=@n9.Variable).Value
   )
   n14 = UE.Reroute(Value=@WavePlayer_0.Out Right)
   n15 = UE.Reroute(Value=@WavePlayer_0.Out Left)
   n16 = UE.Reroute(Value=@WavePlayer_1.Out Right)
   AudioMixerAudioMixerStereo2_0 = "AudioMixer.Audio Mixer (Stereo, 2)"(
    In 0 L=UE.Multiply:Audio by Float(PrimaryOperand=UE.Multiply:Audio(PrimaryOperand=@n10.Value, AdditionalOperands=@n11.Value).Out, AdditionalOperands=@n12.Value).Out
    In 1 R=UE.Multiply:Audio by Float(PrimaryOperand=UE.Multiply:Audio(AdditionalOperands=@n14.Value, PrimaryOperand=@n11.Value).Out, AdditionalOperands=@n12.Value).Out
    In 1 L=UE.Multiply:Audio by Float(PrimaryOperand=UE.Multiply:Audio(PrimaryOperand=@n15.Value, AdditionalOperands=@n11.Value).Out, AdditionalOperands=@n12.Value).Out
    In 0 R=UE.Multiply:Audio by Float(PrimaryOperand=UE.Multiply:Audio(AdditionalOperands=@n16.Value, PrimaryOperand=@n11.Value).Out, AdditionalOperands=@n12.Value).Out
   )
   n17 = InitVariable.Audio()
   n18 = VariableMutator.Float(
    Value=AbsoluteValue.Abs:Float(Input=/FortMetasound/SpatialTravel/MSP_SpatialTravelController(Speed.Movement=^Movement.Speed, Elevation.Movement=^Movement.SpeedElevation).NormalizedElevation).Value
    Variable=InitVariable.Float().Variable
   )
   n19 = /FortMetasound/SpatialTravel/MSP_SpatialTravelVoice(
    Loop.Sounds=^AirLoops
    FadeIn=^UE.Source.OnPlay
    Loop.VolumeControl=UE.Multiply:Float(AdditionalOperands=^VolumeMultiplier, PrimaryOperand=VariableAccessor.Float(Variable=@n18.Variable).Value).Out
   ) {also {defaults {Loop.FadeTime=0.5}}}
   n20 = InitVariable.Audio()
   AudioMixerAudioMixerStereo2_1 = "AudioMixer.Audio Mixer (Stereo, 2)"(
    In 0 L=@AudioMixerAudioMixerStereo2_0.Out L
    In 0 R=@AudioMixerAudioMixerStereo2_0.Out R
    In 1 R=VariableAccessor.Audio(Variable=VariableMutator.Audio(Variable=@n17.Variable, Value=@n19.OutR).Variable).Value
    In 1 L=VariableAccessor.Audio(Variable=VariableMutator.Audio(Variable=@n20.Variable, Value=@n19.OutL).Variable).Value
   )
   InitVariable.Float()
   TriggerAnyTriggerAny2_0 = "TriggerAny.Trigger Any (2)"(In 0=UE.Reroute(Value=@WavePlayer_1.On Finished).Value, In 1=UE.Reroute(Value=@WavePlayer_0.On Finished).Value)
  ]
  links [
   {dst="@UE.OutputFormat.Stereo.Audio:0", src=AudioMixerAudioMixerStereo2_1.Out L}
   {dst="@UE.OutputFormat.Stereo.Audio:1", src=AudioMixerAudioMixerStereo2_1.Out R}
   {dst="@UE.Source.OneShot.OnFinished", src=TriggerAnyTriggerAny2_0.Out}
  ]
 }]
properties {
 assetImportData=&self:AssetImportData
 attenuationSettings=/CubeSurfing/Assets/Audio/Settings/ATT_CubeSurfing_InAir
 bOverrideConcurrency=true
 numChannels=2
 outputFormat=Stereo
 sampleRate=48000
 soundClassObject=&Fort_Gliders_InAir
 volume=0.3572
 concurrencyOverrides {bLimitToOwner=true, maxCount=1}
}
