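# Looping stereo MetaSound source for first-person cube surfing idle audio, driven by movement speed.
# A SpatialTravel controller normalizes speed, which wavetable curves map to main and brake loop levels and main loop pitch, with start/stop one-shots.
# A high-speed gate cuts biquad gain on the main voice and fades a third loop that has no waves wired, leaving AccelLoop and TurningOS unused.
# The Mobile page bypasses the filter and trims start one-shots.
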MetaSoundSource /CubeSurfing/Assets/Audio/MSS_CubeSurfing_Surf_Idle_LP_1P
inputs [
 AccelLoop {type=WaveAsset:Array}
 AzimuthRetriggerDelay {type=Time}
 AzimuthThreshold {type=Float}
 bUseRandomStartTime {type=Bool}
 ElevationChangeFloor {type=Float}
 ElevationChangeFloorMax {type=Float}
 ExtraOsVolume {type=Float}
 InterpSpeed {type=Time}
 LoopVolume {type=Float}
 MainLoop {type=WaveAsset:Array}
 Movement.Speed {type=Float}
 Movement.SpeedAzimuth {type=Float}
 Movement.SpeedElevation {type=Float}
 MovementSpeed {type=Float}
 MovementSpeedFloor {type=Float}
 MovementSpeedMax {type=Float}
 SlowLoop {type=WaveAsset:Array}
 StartOneShots {type=WaveAsset:Array}
 StopOneShots {type=WaveAsset:Array}
 TurningOS {type=WaveAsset:Array}
 TurnVolume {type=Float}
 UE.Source.OnPlay {type=Trigger}
 WaveTable {type=WaveTableBankAsset}
]
interfaces [Fort.Movement:1.0, UE.OutputFormat.Stereo:1.0, UE.Source:1.1]
outputs [
 "UE.OutputFormat.Stereo.Audio:0" {type=Audio}
 "UE.OutputFormat.Stereo.Audio:1" {type=Audio}
]
pages [
 Default {
  is_build_graph=true
  input_defaults {
   AzimuthRetriggerDelay=1
   AzimuthThreshold=0.79
   ElevationChangeFloor=0
   ElevationChangeFloorMax=100
   ExtraOsVolume=1
   InterpSpeed=0.1
   LoopVolume=1
   Movement.Speed=0.588
   Movement.SpeedAzimuth=0
   Movement.SpeedElevation=0
   MovementSpeed=1
   MovementSpeedFloor=600
   MovementSpeedMax=1800
   TurnVolume=0
   UE.Source.OnPlay=null
   WaveTable=/CubeSurfing/Assets/Audio/Curves/WTB_CubeSurfingIdle
   bUseRandomStartTime=null
   AccelLoop [/CubeSurfing/Assets/Audio/Waves/Ability_CubeSurfing_Accel_LP_01]
   MainLoop [/CubeSurfing/Assets/Audio/Waves/Ability_CubeSurfing_Surf_Loop_01]
   SlowLoop [/CubeSurfing/Assets/Audio/Waves/Ability_CubeSurfing_Brake_Loop_01]
   StartOneShots [
    /CubeSurfing/Assets/Audio/Waves/Ability_CubeSurfing_Accel_OS_01
    /CubeSurfing/Assets/Audio/Waves/Ability_CubeSurfing_Accel_OS_02
    /CubeSurfing/Assets/Audio/Waves/Ability_CubeSurfing_Accel_OS_03
    /CubeSurfing/Assets/Audio/Waves/Ability_CubeSurfing_Accel_OS_04
    /CubeSurfing/Assets/Audio/Waves/Ability_CubeSurfing_Accel_OS_05
   ]
   StopOneShots [/CubeSurfing/Assets/Audio/Waves/Ability_CubeSurfing_Decel_OS_01, /CubeSurfing/Assets/Audio/Waves/Ability_CubeSurfing_Decel_OS_02]
   TurningOS [
    /CubeSurfing/Assets/Audio/Waves/Ability_CubeSurfing_Turn_OS_01
    /CubeSurfing/Assets/Audio/Waves/Ability_CubeSurfing_Turn_OS_02
    /CubeSurfing/Assets/Audio/Waves/Ability_CubeSurfing_Turn_OS_03
    /CubeSurfing/Assets/Audio/Waves/Ability_CubeSurfing_Turn_OS_04
    /CubeSurfing/Assets/Audio/Waves/Ability_CubeSurfing_Turn_OS_05
   ]
  }
  nodes [
   n1 = InitVariable.Audio()
   n2 = /FortMetasound/SpatialTravel/MSP_SpatialTravelController(
    Azimuth.RetriggerDelay=^AzimuthRetriggerDelay
    Azimuth.Threshold=^AzimuthThreshold
    InterpSpeed=^InterpSpeed
    Speed.Movement=^Movement.Speed
    Azimuth.Movement=^Movement.SpeedAzimuth
    Speed.SpeedMin=^MovementSpeedFloor
    Speed.SpeedMax=^MovementSpeedMax
   )
   n3 = VariableAccessor.Float(Variable=VariableMutator.Float(Variable=InitVariable.Float().Variable, Value=@n2.NormalizedSpeed).Variable)
   n4 = VariableMutator.Float(
    Variable=InitVariable.Float().Variable
    Value=UE.WaveTableBankEvaluate(WaveTableBank=^WaveTable, Input=@n3.Value) {also {defaults {Index=2}}}.Output
   )
   n5 = VariableAccessor.Float(Variable=VariableMutator.Float(
     Variable=InitVariable.Float().Variable
     Value=UE.WaveTableBankEvaluate(WaveTableBank=^WaveTable, Input=@n3.Value) {also {defaults {Index=1}}}.Output
    ).Variable)
   n6 = /FortMetasound/SpatialTravel/MSP_SpatialTravelVoice(
    Loop.Sounds=^MainLoop
    OneShot.Starts=^StartOneShots
    OneShot.Stops=^StopOneShots
    FadeIn=^UE.Source.OnPlay
    Loop.PitchControl=UE.InterpTo:Audio(Target=VariableAccessor.Float(Variable=@n4.Variable).Value) {also {defaults {Interp Time=0.3}}}.Value
    Loop.VolumeControl=@n5.Value
    OneShot.Value=@n5.Value
   ) {also {defaults {Loop.FadeTime=0.5, Loop.PitchMax=10, Loop.PitchhMin=-3, OneShot.Threshold=0.35}}}
   n7 = InitVariable.Float()
   n8 = VariableAccessor.Float(Variable=@n3.Variable)
   n9 = &MSP_OneshotGate(Value=VariableAccessor.Float(Variable=@n8.Variable).Value) {also {defaults {OneshotThreshold=0.85}}}
   TriggerRouteTriggerRouteFloat2_0 = "TriggerRoute.Trigger Route (Float, 2):Float"(Set 0=@n9.OnGateEnter, Set 1=@n9.OnGateExit) {also {defaults {Value 0=-8, Value 1=0}}}
   n10 = &StereoBiquadFilter(
    In L=@n6.OutL
    In R=@n6.OutR
    Gain=VariableAccessor.Float(Variable=VariableMutator.Float(Variable=@n7.Variable, Value=@TriggerRouteTriggerRouteFloat2_0.Value).Variable).Value
   ) {also {defaults {Cutoff Frequency=7000, Type=6}}}
   n11 = InitVariable.Audio()
   n12 = /FortMetasound/SpatialTravel/MSP_SpatialTravelLoop(FadeIn=@n9.OnGateEnter, FadeOut=@n9.OnGateExit) {also {defaults {Loop.PitchControl=0}}}
   n13 = InitVariable.Audio()
   n14 = InitVariable.Float()
   n15 = UE.WaveTableBankEvaluate(WaveTableBank=^WaveTable, Input=@n3.Value)
   n16 = &MSP_SpatialTravelLoop(
    Loop.Sounds=^SlowLoop
    FadeIn=^UE.Source.OnPlay
    Loop.PitchControl=@n8.Value
    Loop.VolumeControl=VariableAccessor.Float(Variable=VariableMutator.Float(Variable=@n14.Variable, Value=@n15.Output).Variable).Value
   )
   n17 = InitVariable.Audio()
   n18 = InitVariable.Audio()
   n19 = InitVariable.Audio()
   AudioMixerAudioMixerStereo3_0 = "AudioMixer.Audio Mixer (Stereo, 3)"(
    In 0 R=VariableAccessor.Audio(Variable=VariableMutator.Audio(Variable=@n1.Variable, Value=@n10.Out R).Variable).Value
    In 2 L=VariableAccessor.Audio(Variable=VariableMutator.Audio(Variable=@n11.Variable, Value=@n12.OutL).Variable).Value
    In 1 R=VariableAccessor.Audio(Variable=VariableMutator.Audio(Variable=@n13.Variable, Value=@n16.OutR).Variable).Value
    In 2 R=VariableAccessor.Audio(Variable=VariableMutator.Audio(Variable=@n17.Variable, Value=@n12.OutR).Variable).Value
    In 1 L=VariableAccessor.Audio(Variable=VariableMutator.Audio(Variable=@n18.Variable, Value=@n16.OutL).Variable).Value
    In 0 L=VariableAccessor.Audio(Variable=VariableMutator.Audio(Variable=@n19.Variable, Value=@n10.Out L).Variable).Value
   )
   AudioMixerAudioMixerStereo2_0 = "AudioMixer.Audio Mixer (Stereo, 2)"(Gain 0=^LoopVolume, In 0 L=@AudioMixerAudioMixerStereo3_0.Out L, In 0 R=@AudioMixerAudioMixerStereo3_0.Out R)
   InitVariable.Audio()
   InitVariable.Audio()
   InitVariable.Audio()
   InitVariable.Bool()
   InitVariable.Bool()
   InitVariable.Float()
   InitVariable.Float()
   InitVariable.Float()
   InitVariable.Time()
   InitVariable.Trigger()
   InitVariable.WaveAsset()
   VariableMutator.Float(Variable=InitVariable.Float().Variable, Value=@n2.NormalizedAzimuth)
   VariableMutator.Float(
    Variable=InitVariable.Float().Variable
    Value=UE.WaveTableBankEvaluate(WaveTableBank=^WaveTable, Input=@n3.Value) {also {defaults {Index=3}}}.Output
   )
   VariableMutator.Float(Variable=InitVariable.Float().Variable, Value=@n2.NormalizedElevation)
   VariableMutator.Trigger(Variable=InitVariable.Trigger().Variable, Value=@n2.TurnActivated)
   VariableMutator.Trigger(Variable=InitVariable.Trigger().Variable, Value=@n6.StopOneShotTriggered)
   VariableMutator.Trigger(Variable=InitVariable.Trigger().Variable, Value=@n6.StartOneShotTriggered)
   VariableMutator.Trigger(Variable=InitVariable.Trigger().Variable, Value=@n2.TurnFinished)
  ]
  links [
   {dst="@UE.OutputFormat.Stereo.Audio:0", src=AudioMixerAudioMixerStereo2_0.Out L}
   {dst="@UE.OutputFormat.Stereo.Audio:1", src=AudioMixerAudioMixerStereo2_0.Out R}
  ]
 }
 Mobile {
  input_defaults {StartOneShots [&Ability_CubeSurfing_Accel_OS_01, &Ability_CubeSurfing_Accel_OS_02]}
  nodes [
   n1 = InitVariable.Audio()
   n2 = &MSP_SpatialTravelController(
    Azimuth.RetriggerDelay=^AzimuthRetriggerDelay
    Azimuth.Threshold=^AzimuthThreshold
    InterpSpeed=^InterpSpeed
    Speed.Movement=^Movement.Speed
    Azimuth.Movement=^Movement.SpeedAzimuth
    Speed.SpeedMin=^MovementSpeedFloor
    Speed.SpeedMax=^MovementSpeedMax
   )
   n3 = VariableAccessor.Float(Variable=VariableMutator.Float(Variable=InitVariable.Float().Variable, Value=@n2.NormalizedSpeed).Variable)
   n4 = VariableMutator.Float(
    Variable=InitVariable.Float().Variable
    Value=UE.WaveTableBankEvaluate(WaveTableBank=^WaveTable, Input=@n3.Value) {also {defaults {Index=2}}}.Output
   )
   n5 = VariableAccessor.Float(Variable=VariableMutator.Float(
     Variable=InitVariable.Float().Variable
     Value=UE.WaveTableBankEvaluate(WaveTableBank=^WaveTable, Input=@n3.Value) {also {defaults {Index=1}}}.Output
    ).Variable)
   n6 = &MSP_SpatialTravelVoice(
    Loop.Sounds=^MainLoop
    OneShot.Starts=^StartOneShots
    OneShot.Stops=^StopOneShots
    FadeIn=^UE.Source.OnPlay
    Loop.PitchControl=UE.InterpTo:Audio(Target=VariableAccessor.Float(Variable=@n4.Variable).Value) {also {defaults {Interp Time=0.3}}}.Value
    Loop.VolumeControl=@n5.Value
    OneShot.Value=@n5.Value
   ) {also {defaults {Loop.FadeTime=0.5, Loop.PitchMax=10, Loop.PitchhMin=-3, OneShot.Threshold=0.35}}}
   n7 = InitVariable.Audio()
   n8 = VariableAccessor.Float(Variable=@n3.Variable)
   n9 = &MSP_OneshotGate(Value=VariableAccessor.Float(Variable=@n8.Variable).Value) {also {defaults {OneshotThreshold=0.85}}}
   n10 = &MSP_SpatialTravelLoop(FadeIn=@n9.OnGateEnter, FadeOut=@n9.OnGateExit) {also {defaults {Loop.PitchControl=0}}}
   n11 = InitVariable.Audio()
   n12 = InitVariable.Float()
   n13 = UE.WaveTableBankEvaluate(WaveTableBank=^WaveTable, Input=@n3.Value)
   n14 = &MSP_SpatialTravelLoop(
    Loop.Sounds=^SlowLoop
    FadeIn=^UE.Source.OnPlay
    Loop.PitchControl=@n8.Value
    Loop.VolumeControl=VariableAccessor.Float(Variable=VariableMutator.Float(Variable=@n12.Variable, Value=@n13.Output).Variable).Value
   )
   n15 = InitVariable.Audio()
   n16 = InitVariable.Audio()
   n17 = InitVariable.Audio()
   AudioMixerAudioMixerStereo3_0 = "AudioMixer.Audio Mixer (Stereo, 3)"(
    In 0 R=VariableAccessor.Audio(Variable=VariableMutator.Audio(Variable=@n1.Variable, Value=@n6.OutR).Variable).Value
    In 2 L=VariableAccessor.Audio(Variable=VariableMutator.Audio(Variable=@n7.Variable, Value=@n10.OutL).Variable).Value
    In 1 R=VariableAccessor.Audio(Variable=VariableMutator.Audio(Variable=@n11.Variable, Value=@n14.OutR).Variable).Value
    In 2 R=VariableAccessor.Audio(Variable=VariableMutator.Audio(Variable=@n15.Variable, Value=@n10.OutR).Variable).Value
    In 1 L=VariableAccessor.Audio(Variable=VariableMutator.Audio(Variable=@n16.Variable, Value=@n14.OutL).Variable).Value
    In 0 L=VariableAccessor.Audio(Variable=VariableMutator.Audio(Variable=@n17.Variable, Value=@n6.OutL).Variable).Value
   )
   AudioMixerAudioMixerStereo2_0 = "AudioMixer.Audio Mixer (Stereo, 2)"(Gain 0=^LoopVolume, In 0 L=@AudioMixerAudioMixerStereo3_0.Out L, In 0 R=@AudioMixerAudioMixerStereo3_0.Out R)
   InitVariable.Audio()
   InitVariable.Audio()
   InitVariable.Audio()
   InitVariable.Bool()
   InitVariable.Bool()
   InitVariable.Float()
   InitVariable.Float()
   InitVariable.Float()
   InitVariable.Time()
   InitVariable.Trigger()
   InitVariable.WaveAsset()
   VariableMutator.Float(Variable=InitVariable.Float().Variable, Value=@n2.NormalizedAzimuth)
   VariableMutator.Float(
    Variable=InitVariable.Float().Variable
    Value=UE.WaveTableBankEvaluate(WaveTableBank=^WaveTable, Input=@n3.Value) {also {defaults {Index=3}}}.Output
   )
   VariableMutator.Float(Variable=InitVariable.Float().Variable, Value=@n2.NormalizedElevation)
   TriggerRouteTriggerRouteFloat2_0 = "TriggerRoute.Trigger Route (Float, 2):Float"(Set 0=@n9.OnGateEnter, Set 1=@n9.OnGateExit) {also {defaults {Value 0=-8, Value 1=0}}}
   VariableMutator.Float(Variable=InitVariable.Float().Variable, Value=@TriggerRouteTriggerRouteFloat2_0.Value)
   VariableMutator.Trigger(Variable=InitVariable.Trigger().Variable, Value=@n2.TurnActivated)
   VariableMutator.Trigger(Variable=InitVariable.Trigger().Variable, Value=@n6.StopOneShotTriggered)
   VariableMutator.Trigger(Variable=InitVariable.Trigger().Variable, Value=@n6.StartOneShotTriggered)
   VariableMutator.Trigger(Variable=InitVariable.Trigger().Variable, Value=@n2.TurnFinished)
  ]
  links [
   {dst="@UE.OutputFormat.Stereo.Audio:0", src=AudioMixerAudioMixerStereo2_0.Out L}
   {dst="@UE.OutputFormat.Stereo.Audio:1", src=AudioMixerAudioMixerStereo2_0.Out R}
  ]
 }
]
properties {
 assetImportData=&self:AssetImportData
 bLooping=true
 duration=10000
 numChannels=2
 outputFormat=Stereo
 sampleRate=48000
 soundClassObject=&Fort_Gliders_InAir
 volume=0.2388
}
