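# Stereo MetaSound source for the first-person (1P) water surfing loop in CubeSurfing, limited to one instance per owner.
# A SpatialTravelVoice patch starts on play and loops the wave with a fade time.
# JumpStart routes InAirVolume (default silent) and JumpStop routes full gain through an interpolated value that scales both channels.
# Several inputs, including elevation, pitch and attack/release, are declared but unconnected.
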MetaSoundSource /CubeSurfing/Assets/Audio/MSS_CubeSurfing_Surf_Water_LP_1P
inputs [
 Attack Time {type=Time}
 bUseRandomStartTime {type=Bool}
 ElevationChangeFloor {type=Float}
 ElevationChangeFloorMax {type=Float}
 InAirVolume {type=Float}
 InterpSpeed {type=Time}
 JumpStart {type=Trigger}
 JumpStop {type=Trigger}
 Loop {type=Bool}
 LoopFadeTime {type=Float}
 Max Pitch {type=Float}
 Max Volume {type=Float}
 Min Pitch {type=Float}
 Min Volume {type=Float}
 MovementElevation {type=Float}
 Normalized Elevation {type=Float}
 Release Time {type=Time}
 Sounds {type=WaveAsset:Array}
 UE.Source.OnPlay {type=Trigger}
]
interfaces [UE.OutputFormat.Stereo:1.0, UE.Source.OneShot:1.0, UE.Source:1.1]
outputs [
 "UE.OutputFormat.Stereo.Audio:0" {type=Audio}
 "UE.OutputFormat.Stereo.Audio:1" {type=Audio}
 UE.Source.OneShot.OnFinished {type=Trigger}
]
pages [Default {
  input_defaults {
   Attack Time=0.524
   ElevationChangeFloor=0
   ElevationChangeFloorMax=100
   InAirVolume=0
   InterpSpeed=0.5
   JumpStart=null
   JumpStop=null
   Loop="<MetaSoundBuilderResult.SUCCEEDED: 0>"
   LoopFadeTime=0.516
   Max Pitch=1
   Max Volume=1
   Min Pitch=1
   Min Volume=0.9
   MovementElevation=0
   Normalized Elevation=0
   Release Time=0.912
   UE.Source.OnPlay=null
   bUseRandomStartTime=null
   Sounds [/CubeSurfing/Assets/Audio/Waves/Ability_CubeSurfing_Water_LP_01]
  }
  nodes [
   InitVariable.Audio()
   InitVariable.Float()
   InitVariable.Float()
   InitVariable.Time()
   InitVariable.WaveAsset()
   MSP_SpatialTravelVoice_0 = /FortMetasound/SpatialTravel/MSP_SpatialTravelVoice(Loop.FadeTime=^LoopFadeTime, Loop.Sounds=^Sounds, FadeIn=^UE.Source.OnPlay)
   TriggerRouteTriggerRouteFloat2_0 = "TriggerRoute.Trigger Route (Float, 2):Float"(Value 0=^InAirVolume, Set 0=^JumpStart, Set 1=^JumpStop) {also {defaults {Value 1=1}}}
   n1 = VariableDeferredAccessor.Float(Variable=InitVariable.Float().Variable)
   n2 = Value.Value:Float(Set=@TriggerRouteTriggerRouteFloat2_0.On Set, Target Value=@TriggerRouteTriggerRouteFloat2_0.Value, Init Value=@n1.Value)
   n3 = VariableAccessor.Float(Variable=VariableMutator.Float(Value=UE.InterpTo:Audio(Interp Time=^InterpSpeed, Target=@n2.Output Value).Value, Variable=@n1.Variable).Variable)
   Multiply_0 = UE.Multiply:Audio by Float(PrimaryOperand=@MSP_SpatialTravelVoice_0.OutL, AdditionalOperands=@n3.Value)
   Multiply_1 = UE.Multiply:Audio by Float(PrimaryOperand=@MSP_SpatialTravelVoice_0.OutR, AdditionalOperands=@n3.Value)
  ]
  links [
   {dst="@UE.Source.OneShot.OnFinished", src=MSP_SpatialTravelVoice_0.On Loop Finished}
   {dst="@UE.OutputFormat.Stereo.Audio:0", src=Multiply_0.Out}
   {dst="@UE.OutputFormat.Stereo.Audio:1", src=Multiply_1.Out}
  ]
 }]
properties {
 assetImportData=&self:AssetImportData
 bOverrideConcurrency=true
 numChannels=2
 outputFormat=Stereo
 sampleRate=48000
 soundClassObject=&Fort_Gliders_InAir
 volume=0.4246
 concurrencyOverrides {bLimitToOwner=true, maxCount=1}
}
