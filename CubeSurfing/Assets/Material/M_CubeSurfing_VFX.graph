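# Niagara VFX material for CubeSurfing sprites, ribbons and mesh particles, producing a tinted cube gradient with flicker and fresnel glow.
# Particle color length seeds a time-scrolled sawtooth hash, scaled by DerezSpeed, which modulates the gradient texture.
# Dynamic parameters blend the Color parameter with particle color and set fresnel brightness.
# The HDR-boosted fresnel is tinted, added to the gradient, then divided by EyeAdaptation to offset auto-exposure.
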Material /CubeSurfing/Assets/Material/M_CubeSurfing_VFX
graph {
 properties {
  bAutomaticallySetUsageInEditor=false
  bUsedWithNiagaraMeshParticles=true
  bUsedWithNiagaraRibbons=true
  bUsedWithNiagaraSprites=true
  maxWorldPositionOffsetDisplacement=800
  usedShadingModels=Default Lit
 }
 nodes [
  Constant() {r=0.5}
  Constant() {r=0.2}
  n1 = LinearInterpolate(Alpha=TextureSampleParameter2D() {group=Color_Shading, parameterName=CubeGradient, texture=/CubeSurfing/Assets/Textures/T_CubeGradient}.R) {constA=0.5}
  n2 = ScalarParameter() {defaultValue=0.3, parameterName=HashMin_Cd}
  n3 = ParticleColor()
  n4 = Length(None=@n3.RGB) {is=class MaterialExpressionLength}
  n5 = Multiply(A=@n4) {constB=5}
  n6 = Multiply(A=Time(), B=ScalarParameter() {defaultValue=1, parameterName=DerezSpeed})
  n7 = Power(Base=Frac(None=Add(A=@n5, B=@n6)))
  n8 = DynamicParameter() {defaultValue=<rgba 0 0 1 0>, paramNames [FresnelGlow, UseParticleColor, GridBrightness, CubeLerp]}
  n9 = LinearInterpolate(A=VectorParameter() {parameterName=Color, defaultValue=<rgba 0.3185 0.9823 1 1>}.RGB, B=@n3.RGB, Alpha=@n8.UseParticleColor)
  n10 = Multiply(
   A=NamedRerouteDeclaration(None=LinearInterpolate(A=@n2, Alpha=@n7) {constA=0.2}) {name=SimpleHashNoise_Cd, nodeColor=<rgba 0 1 0.9531 1>}
   B=@n9
  )
  n11 = Constant() {r=2}
  n12 = NamedRerouteDeclaration(None=MaterialFunctionCall(
    LDR Input=MaterialFunctionCall(Brightness=@n8.FresnelGlow, Exponent=@n11) {materialFunction=&MF_Fresnel_FX}.Fresnel
    HDR Dynamic Range=Constant() {r=1}
    HDR Intensity Multiplier=Constant() {r=100}
   ) {materialFunction=&DeriveHDRfromLDR}.RemappedRange) {name=Fresnel, nodeColor=<rgba 1 0 0.6797 1>}
  Divide(A=Add(A=Multiply(A=@n1, B=@n10), B=Multiply(A=@n9, B=@n12)), B=EyeAdaptation().EyeAdaptation)
  Frac(None=@n4)
  Reroute(None=@n3.A)
 ]
}
