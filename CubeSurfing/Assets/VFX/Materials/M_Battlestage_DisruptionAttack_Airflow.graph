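# Unlit, two-sided translucent Niagara ribbon and mesh-particle material for the Battlestage disruption attack's airflow VFX.
# Opacity blends between a panned streak texture with scrolling noise minus a rounded dissolve mask, and an optional normal-distorted smoke trail.
# Particle dynamic parameters drive offset, dissolve, noise and SplineThicken width, while particle color feeds emissive and alpha.
# View-angle fade is skipped at Low/Medium quality, and depth fade on mobile.
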Material /CubeSurfing/Assets/VFX/Materials/M_Battlestage_DisruptionAttack_Airflow
graph {
 properties {
  bUsedWithNiagaraMeshParticles=true
  bUsedWithNiagaraRibbons=true
  blendMode=BLEND_Translucent
  pixelDepthOffsetMode=PDOM_Legacy
  shadingModel=MSM_Unlit
  twoSided=true
  usedShadingModels=Unlit
 }
 nodes [
  n1 = ParticleColor()
  MaterialFunctionCall(Emissive=@n1.RGB) {materialFunction=&MF_FX_EmissiveAdjustments}
  n2 = MaterialFunctionCall("FadeSensitivity (optional)"=ScalarParameter() {defaultValue=2, parameterName=FadeAngle}) {materialFunction=&FadeBasedOnViewAngle}
  n3 = Constant() {r=1}
  n4 = TextureCoordinate()
  n5 = VectorParameter() {parameterName=Tex_UVScale, defaultValue=<rgba 1 1 1 1>}
  n6 = AppendVector(A=@n5.R, B=@n5.R)
  n7 = DynamicParameter() {defaultValue=<rgba 0.5 1 1 10>, paramNames [Offset,Dissolve,Width,Noise_offset]}
  n8 = Constant()
  n9 = ScalarParameter() {defaultValue=1, parameterName=Tex_default_V_panner_speed}
  n10 = TextureSampleParameter2D(UVs=Panner(Coordinate=Multiply(A=@n4, B=@n6), Time=LinearInterpolate(Alpha=@n7.Offset) {constA=-1}, Speed=AppendVector(A=@n8, B=@n9)) {speedY=1}) {
   parameterName=Tex_Default
   samplerSource=SSM_Clamp_WorldGroupSettings
   texture=&/T_BattleStage_DisruptionAttack_Airflow
  }
  n11 = StaticComponentMaskParameter(None=@n10.RGB) {defaultR=true, parameterName=Component_mask}
  n12 = MaterialFunctionCall(XY=TextureCoordinate()) {materialFunction=&Swizzle}
  n13 = VectorParameter() {parameterName=Tex_noise_UVScale, defaultValue=<rgba 0.5 1 1 1>}
  n14 = AppendVector(A=@n13.R, B=@n13.R)
  n15 = DynamicParameter() {defaultValue=<rgba 0.5 1 1 10>, paramNames [Offset,Dissolve,Width,Noise_offset]}
  n16 = Time()
  n17 = ScalarParameter() {defaultValue=1, parameterName=Loop_time}
  n18 = TextureSampleParameter2D(UVs=Panner(Coordinate=Multiply(A=@n12.YX, B=@n14), Time=LinearInterpolate(A=@n15.Noise_offset, B=@n16, Alpha=@n17)) {speedX=1.5}) {parameterName=Tex_noise, samplerType=SAMPLERTYPE_Grayscale, texture=&T_Smoke_Trail_Soft}
  n19 = Multiply(A=Multiply(A=@n18.R, B=@n18.R), B=ScalarParameter() {defaultValue=2, parameterName=Tex_noise_brightness}) {constB=2}
  n20 = Power(
   Base=Multiply(A=Multiply(A=@n11, B=@n19)) {constB=2}
   Exp=ScalarParameter() {defaultValue=1, group=Only_Use_Legs_fresnel, parameterName=power}
  ) {constExponent=1}
  n21 = ScalarParameter() {parameterName=Tex_noise_lerp_dark}
  n22 = VectorParameter() {parameterName=Dissolve, defaultValue=<rgba 0.3 0 1 1>}
  n23 = TextureSampleParameter2D(UVs=Add(A=TextureCoordinate() {vTiling=2}, B=AppendVector(A=@n22.R, B=@n22.R))) {parameterName=Tex_Dissolve, texture=/Game/Global/Textures/Default/Masks_and_Gradients/T_BlasterBeamDistort}
  n24 = Power(Base=Multiply(A=Add(A=@n20) {constB=0}, B=LinearInterpolate(A=@n21, Alpha=@n23.R))) {constExponent=1.5}
  n25 = ScalarParameter() {defaultValue=1, parameterName=Multiply_noise}
  n26 = Sine(None=ComponentMask(None=TextureCoordinate()) {g=true}) {period=2}
  n27 = Multiply(A=Multiply(A=LinearInterpolate(A=@n11, B=@n24, Alpha=@n25), B=Power(Base=@n26)))
  n28 = VectorParameter() {parameterName=Dissolve_UVScale, defaultValue=<rgba 1 1 1 1>}
  n29 = VectorParameter() {parameterName=Dissolve, defaultValue=<rgba 0.3 0 1 1>}
  n30 = Add(A=Multiply(A=TextureCoordinate(), B=AppendVector(A=@n28.R, B=@n28.R)), B=AppendVector(A=@n29.R, B=@n29.R))
  n31 = Power(
   Base=ComponentMask(None=TextureCoordinate()) {r=true}
   Exp=LinearInterpolate(Alpha=DynamicParameter() {defaultValue=<rgba 0 1 1 0>, paramNames [Offset,Dissolve,Width,Param4]}.Dissolve) {constB=5}
  )
  n32 = Round(None=LinearInterpolate(
    A=TextureSampleParameter2D(UVs=@n30) {parameterName=Tex_Dissolve, texture=&T_BlasterBeamDistort}.R
    B=Saturate(None=@n31)
   ))
  n33 = TextureSampleParameter2D(UVs=Panner(Coordinate=TextureCoordinate() {vTiling=0.5}) {speedY=1}) {
   group=Noise
   parameterName=Noise Texture
   samplerType=SAMPLERTYPE_Normal
   texture=&T_Weap_Distort_N_02
  }
  n34 = Multiply(A=ComponentMask(None=@n33.RGB) {g=true, r=true}) {constB=0.2}
  n35 = TextureCoordinate()
  n36 = DynamicParameter() {desc=niagara dynamic 1, parameterIndex=1, defaultValue=<rgba 1 1 0 0>, paramNames [Power_and_UVscale, offset, Noise offset, Param4]}
  n37 = AppendVector(A=Constant() {r=1}, B=LinearInterpolate(Alpha=@n36.Power_and_UVscale) {constA=1, constB=0.7})
  n38 = Panner(
   Coordinate=MaterialFunctionCall(XY=Multiply(A=@n35, B=@n37)) {materialFunction=&Swizzle}.YX
   Speed=ComponentMask(None=VectorParameter() {parameterName=Disturbe speed, defaultValue=<rgba 0.8 0 0 0>}.RGB) {g=true, r=true}
  ) {speedX=0.8}
  n39 = Power(
   Base=TextureSample(UVs=Add(A=@n34, B=@n38)) {samplerType=SAMPLERTYPE_Grayscale, texture=&T_Smoke_Trail_Soft}.R
   Exp=LinearInterpolate(Alpha=@n36.Power_and_UVscale) {constA=1, constB=1.5}
  ) {constExponent=1.5}
  n40 = TextureCoordinate()
  n41 = Power(Base=ComponentMask(None=@n40) {g=true}) {constExponent=1}
  n42 = Multiply(A=OneMinus(None=ComponentMask(None=@n40) {g=true}), B=Saturate(None=Multiply(A=@n41) {constB=10}))
  n43 = Multiply(
   A=@n1.A
   B=LinearInterpolate(A=Subtract(A=@n27, B=@n32), B=Multiply(A=@n39, B=@n42), Alpha=ScalarParameter() {parameterName=Use disturbed trail})
  )
  n44 = DepthFade(Opacity=@n43, FadeDistance=ScalarParameter() {defaultValue=50, parameterName=DepthFade}) {fadeDistanceDefault=50}
  Multiply(
   A=QualitySwitch(Default=@n2.Result, Low=@n3, High=@n2.Result, Medium=@n3, Epic=@n2.Result)
   B=ShadingPathSwitch(Default=@n44, Deferred=@n44, Forward=@n44, Mobile=@n43)
  )
  n45 = Multiply(A=Constant() {r=30}, B=DynamicParameter() {defaultValue=<rgba 0.5 1 1 10>, paramNames [Offset,Dissolve,Width,Noise_offset]}.Width)
  Reroute(None=LinearInterpolate(
    A=MaterialFunctionCall(WidthBase=@n45, WidthTip=@n45) {materialFunction=&SplineThicken}.WorldPositionOffset
    B=Constant()
    Alpha=ScalarParameter() {parameterName=No_SplineThicken}
   ))
 ]
}
