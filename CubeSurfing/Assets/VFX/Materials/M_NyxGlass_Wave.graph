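# M_NyxGlass_Wave is an unlit, additive, two-sided material for Niagara mesh particle wave effects.
# A turbulence texture pans vertically on dynamic-parameter time and blends a tint toward particle color for MF_FX_EmissiveAdjustments.
# Opacity math combines vertex red, inverted Fresnel that softens silhouettes, and the turbulence, through a depth fade with particle-driven distance.
# "Use Full Particle Color" increases particle color and alpha influence.
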Material /CubeSurfing/Assets/VFX/Materials/M_NyxGlass_Wave
graph {
 properties {
  bUsedWithNiagaraMeshParticles=true
  blendMode=BLEND_Additive
  pixelDepthOffsetMode=PDOM_Legacy
  shadingModel=MSM_Unlit
  translucencyPass=MTP_BeforeDOF
  twoSided=true
  usedShadingModels=Unlit
 }
 nodes [
  n1 = VertexColor()
  n2 = Fresnel() {baseReflectFraction=0, exponent=1}
  n3 = ScalarParameter() {defaultValue=2, parameterName=F2}
  n4 = Power(Base=OneMinus(None=Power(Base=@n2, Exp=@n3)), Exp=ScalarParameter() {defaultValue=4, parameterName=F})
  n5 = DynamicParameter() {paramNames [Time,Random,DepthFade,Ref]}
  n6 = TextureSampleParameter2D(UVs=Panner(Coordinate=TextureCoordinate() {vTiling=0.1}, Time=Add(A=@n5.Time, B=@n5.Time)) {speedY=0.5}) {parameterName=Wave Texture, texture=/Game/Global/Textures/Default/Caustics/T_Turbulence_Seamless}
  n7 = ParticleColor()
  n8 = Multiply(A=@n7.A, B=@n6.R)
  n9 = ScalarParameter() {parameterName=Use Full Particle Color}
  DepthFade(Opacity=Multiply(A=Multiply(A=@n1.R, B=@n4), B=LinearInterpolate(A=@n6.R, B=@n8, Alpha=@n9)), FadeDistance=@n5.DepthFade)
  n10 = Multiply(A=LinearInterpolate(A=VectorParameter() {defaultValue=<rgba 0.4948 0.4948 0.4948 1>}.RGB, B=@n7.RGB, Alpha=@n6.R), B=@n7.A)
  MaterialFunctionCall(Emissive=LinearInterpolate(A=@n10, B=Multiply(A=@n7.RGB, B=@n10), Alpha=ScalarParameter() {parameterName=Use Full Particle Color})) {materialFunction=&MF_FX_EmissiveAdjustments}
  Multiply()
 ]
}
