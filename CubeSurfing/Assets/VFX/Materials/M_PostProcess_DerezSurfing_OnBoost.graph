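# Post-process material for CubeSurfing's boost "derez" screen effect, blended into scene color after depth of field.
# Tiled grid and noise textures, animated by time and per-cell hashing, build pixelated dissolve masks and a two-color emissive through named reroutes.
# An inverted sphere mask confines the effect to screen edges, with a simpler branch for Forward, Mobile and Low quality.
# The result adds to the scene, and BlendWeight lerps toward the untouched image.
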Material /CubeSurfing/Assets/VFX/Materials/M_PostProcess_DerezSurfing_OnBoost
graph {
 properties {
  blendableLocation=BL_SceneColorAfterDOF
  materialDomain=MD_PostProcess
  pixelDepthOffsetMode=PDOM_Legacy
  shadingModel=MSM_Unlit
  usedShadingModels=Unlit
 }
 nodes [
  n1 = ViewProperty() {property=MEVP_ViewSize}
  n2 = ComponentMask(None=@n1.Property) {r=true}
  n3 = ComponentMask(None=@n1.Property) {g=true}
  n4 = AppendVector(A=Min(A=Divide(A=@n2, B=@n3)), B=Min(A=Divide(A=@n3, B=@n2)))
  n5 = Multiply(A=ScreenPosition().ViewportUV, B=@n4)
  n6 = Divide(A=OneMinus(None=@n4))
  n7 = Divide(A=Distance(A=Add(A=@n5, B=@n6), B=Constant2Vector() {g=0.5, r=0.5}) {desc=x}, B=ScalarParameter() {defaultValue=2, parameterName=Scale})
  n8 = VectorParameter() {parameterName=A, defaultValue=<rgba 0.5 0.5 0.5 1>}
  n9 = Multiply(A=@n7, B=VectorParameter() {parameterName=B, defaultValue=<rgba 1.7 1.7 1.7 1>}.RGB)
  n10 = VectorParameter() {parameterName=C, defaultValue=<rgba 2.574 2.6 2.5844 1>}
  Add(A=Subtract(A=Multiply(A=@n7, B=@n8.RGB), B=Power(Base=@n9)), B=Power(Base=Multiply(A=@n7, B=@n10.RGB)) {constExponent=3})
  n11 = Time()
  n12 = TextureCoordinate() {uTiling=4}
  n13 = Multiply(A=Time(), B=Constant2Vector())
  n14 = ScalarParameter() {defaultValue=75, parameterName=Tiling}
  n15 = Frac(None=Multiply(A=Add(A=@n12, B=@n13), B=@n14))
  n16 = TextureSample(UVs=Add(A=@n12, B=Divide(A=@n15, B=@n14))) {
   samplerSource=SSM_Wrap_WorldGroupSettings
   samplerType=SAMPLERTYPE_LinearColor
   texture=/Game/Packages/Fortress_SharedMaterials/Textures/T_Slowfield_GridSquares_01
  }
  ConstantBiasScale(None=Sine(None=Add(A=@n11, B=@n16.R)))
  n17 = Multiply(A=TextureCoordinate(), B=ScalarParameter() {defaultValue=75, parameterName=Tiling})
  n18 = TextureCoordinate() {uTiling=0.5, vTiling=0.5}
  n19 = Ceil(None=Abs(None=@n18))
  n20 = Add(A=Time() {period=1}, B=Add(A=ComponentMask(None=@n19) {r=true}, B=ComponentMask(None=@n19) {g=true}))
  n21 = Add(A=MaterialFunctionCall(Vec2=Ceil(None=@n17)) {materialFunction=&MF_Hash}.Vec2, B=@n20)
  Frac(None=@n21)
  n22 = Multiply(A=NamedRerouteUsage(), B=NamedRerouteUsage())
  n23 = TextureCoordinate()
  n24 = Constant2Vector() {g=0.5, r=0.5}
  n25 = Constant() {r=0.75}
  n26 = Constant()
  n27 = Power(Base=OneMinus(None=SphereMask(A=@n23, B=@n24, Radius=@n25, Hardness=@n26))) {constExponent=2.5}
  n28 = Multiply(A=@n27)
  n29 = ShadingPathSwitch(Default=Multiply(A=Multiply(A=@n22, B=@n27), B=@n27), Forward=@n28, Mobile=@n28)
  n30 = QualitySwitch(Default=@n29, Low=@n28, High=@n29, Medium=@n29, Epic=@n29)
  n31 = ComponentMask(None=SceneTexture() {sceneTextureId=PPI_PostProcessInput0}.Color) {b=true, g=true, r=true}
  LinearInterpolate(A=Add(A=Multiply(A=@n30), B=@n31), B=@n31, Alpha=ScalarParameter() {defaultValue=1, parameterName=BlendWeight})
  Multiply(A=@n22, B=@n27)
  Multiply(A=Time(), B=Constant2Vector() {g=-0.01})
  n32 = Multiply(B=@n16.R) {constA=1}
  NamedRerouteDeclaration(None=LinearInterpolate(A=@n32, Alpha=NamedRerouteUsage())) {name=Derez_Mask, nodeColor=<rgba 0.0076 0.75 1 1>}
  n33 = ScalarParameter() {defaultValue=75, parameterName=Tiling}
  n34 = Floor(None=Multiply(A=TextureCoordinate() {uTiling=4}, B=@n33))
  n35 = Sine(None=@n21) {period=125}
  n36 = TextureSample(UVs=Add(A=Divide(A=@n34, B=@n33), B=@n35)) {
   samplerSource=SSM_Wrap_WorldGroupSettings
   samplerType=SAMPLERTYPE_LinearColor
   texture=/Game/Characters/Player/Male/Medium/Bodies/M_MED_EmeraldGlassGreen/Textures/T_CowlNoise
  }
  n37 = LinearInterpolate(
   A=VectorParameter() {group=Derez, parameterName=Color_01, defaultValue=<rgba 0.0308 0.1138 0 1>}.RGB
   B=VectorParameter() {group=Derez, parameterName=Color_02, defaultValue=<rgba 0 1 0.0295 1>}.RGB
   Alpha=Multiply(A=@n36.RGB, B=Power(Base=@n36.RGB) {constExponent=1})
  )
  NamedRerouteDeclaration(None=LinearInterpolate(A=@n37, B=Multiply(A=@n37) {constB=5}) {constAlpha=0}) {name=Emissive, nodeColor=<rgba 0 0.3438 1 1>}
  n38 = TextureCoordinate() {uTiling=2}
  n39 = ScalarParameter() {defaultValue=75, parameterName=Tiling}
  n40 = LinearInterpolate(Alpha=TextureSample(UVs=Multiply(A=@n38, B=@n39)) {samplerType=SAMPLERTYPE_LinearColor, texture=/Game/Environments/Hexmap/HexSelection/T_Hex_RadarGrid02}.R) {constA=-2, constB=10}
  NamedRerouteDeclaration(None=OneMinus(None=Saturate(None=@n40))) {name=Grid, nodeColor=<rgba 1 0 0.5859 1>}
  n41 = TextureCoordinate()
  n42 = ScalarParameter() {defaultValue=75, parameterName=Tiling}
  n43 = Multiply(A=OneMinus(None=ComponentMask(None=@n41) {g=true, r=true}), B=@n42)
  n44 = Saturate(None=Divide(A=Floor(None=@n43), B=@n42))
  NamedRerouteDeclaration(None=@n44) {name=U_Gradient, nodeColor=<rgba 0.1484 1 0 1>}
  n45 = Multiply(A=@n35, B=@n44)
  NamedRerouteDeclaration(None=ComponentMask(None=LinearInterpolate(A=@n45, Alpha=@n44) {constB=0.5}) {r=true}) {name=Gradient_Dissolve, nodeColor=<rgba 0 1 0.625 1>}
  n46 = Multiply(A=@n36.RGB)
  n47 = NamedRerouteUsage()
  n48 = Power(Base=@n47) {constExponent=5}
  n49 = NamedRerouteUsage()
  NamedRerouteDeclaration(None=Add(A=Add(A=@n46, B=@n48), B=LinearInterpolate(B=@n49, Alpha=@n47))) {name=Derez_Dissolve, nodeColor=<rgba 0.4766 1 0 1>}
  NamedRerouteDeclaration(None=ParticleColor().A) {name=AlphaDissolve, nodeColor=<rgba 1 0 0.5156 1>}
  Saturate(None=Multiply(A=@n32) {constB=1.2})
  ScalarParameter() {defaultValue=5, parameterName=Power Dissolve}
  ScalarParameter() {defaultValue=1, parameterName=A}
  ScalarParameter() {defaultValue=1.4, parameterName=B}
  ScalarParameter() {defaultValue=2.6, parameterName=C}
  n50 = Multiply(A=ScalarParameter() {defaultValue=0.5, parameterName=Sine Speed}, B=@n20)
  n51 = TextureSample(UVs=@n18) {samplerType=SAMPLERTYPE_LinearColor, texture=&T_Slowfield_GridSquares_01}
  Sine(None=Frac(None=Add(A=@n50, B=@n51.R)))
 ]
}
