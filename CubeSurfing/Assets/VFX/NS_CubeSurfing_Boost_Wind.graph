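# Niagara system for a cube-surfing boost wind effect, combining burst cube debris, a flare and a shockwave.
# Blocks002 bursts short-lived meshes tinted between User.CubeColorMin and CubeColorMax, launched from an offset origin and slowed by drag.
# Flare is one shrinking cyan sprite with time-of-day-scaled alpha.
# Shock_wave spawns one expanding, fading particle drawn by two mesh renderers, with separate "02" color, size, position and dynamic-material attributes alongside the standard ones.
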NiagaraSystem /CubeSurfing/Assets/VFX/NS_CubeSurfing_Boost_Wind
dependencies {
 usedDataInterfaces [&NiagaraDataInterfaceColorCurve, &NiagaraDataInterfaceCurve, &NiagaraDataInterfaceVectorCurve]
 usedDynamicInputs [&MergeHelper_VectorSwitch, &SpawnGroup]
 usedModules [&SystemState]
}
emitters {
 Blocks002 {
  emitterUpdateScript [
   &EmitterState(Life Cycle Mode=::NewEnumerator1, Loop Behavior=::NewEnumerator1, Loop Delay=0.1, Loop Duration=1, UseLoopDelay=true)
   &SpawnBurst_Instantaneous(
    Spawn Count=&RandomRangeInt(Minimum=13, Maximum=17)
    Use Spawn Probability=false
    UseLoopCountLimit=false
   )
  ]
  particleSpawnScript [
   &InitializeParticle(
    Color Maximum=$User.CubeColorMax
    Color Minimum=$User.CubeColorMin
    Color Mode=::NewEnumerator2
    Lifetime Max=0.5
    Lifetime Min=0.25
    Lifetime Mode=::NewEnumerator1
    Mesh Scale Mode=::NewEnumerator1
    Mesh Uniform Scale Max=4
    Mesh Uniform Scale Min=2
    Position Mode=::NewEnumerator0
    Sprite Size Mode=::NewEnumerator3
    Uniform Sprite Size=5
    Write Material Random=true
   )
   &SystemLocation(MaskSpawnGroup=false, Offset=<0 0 100>)
   &ShapeLocation(MaskSpawnGroup=false, Non Uniform Scale=<1 1 0.2>, Override Seed=false, UseOwnerScale=true)
   &AddVelocity(
    Module.Use Falloff Distance=false
    Origin Offset=<-500 0 100>
    Override Seed=false
    Velocity Mode=::NewEnumerator1
    Velocity Speed=&RandomRangeFloat(Minimum=1000, Maximum=2000)
   ) {module=AddVelocity001}
  ]
  particleUpdateScript [
   &ParticleState()
   &ScaleMeshSize(Scale Factor=&Lerp_Vector(
     A=&RandomRangeVector(Minimum=<0.1 0.1 0.1>, Maximum=<0.25 0.25 0.25>)
     B=<1 1 1>
     Alpha=&FloatFromCurve(FloatCurve=DataInterface() {propertyValues {Curve {
         defaultValue=1
         keys [
          {arriveTangent=9.0725, interpMode=RCIM_Cubic, leaveTangent=9.0725, tangentMode=RCTM_User, time=0, value=0}
          {arriveTangent=-2.5249, interpMode=RCIM_Cubic, leaveTangent=-2.5249, tangentMode=RCTM_User, time=0.1408, value=0.9847}
          {arriveTangent=-0.0405, interpMode=RCIM_Cubic, leaveTangent=-0.0405, tangentMode=RCTM_User, time=1, value=0}
         ]
        }}})
    ))
   &ScaleColor(
    Scale Alpha=&FloatFromCurve(
     FloatCurve=DataInterface() {propertyValues {Curve {defaultValue=1, keys [<at 0 0>, <at 0.2 1>, {arriveTangent=-1.3333, leaveTangent=-1.3333, time=0.95, value=1}, <at 1 0>]}}}
     CurveIndex=$Particles.Mask
    )
    ScaleA=true
    ScaleRGB=true
   )
   &DynamicMaterialParameters(
    Index 0 Param 1=&FloatFromCurve(FloatCurve=DataInterface() {propertyValues {Curve {
        defaultValue=0
        keys [
         {arriveTangent=-0.1794, interpMode=RCIM_Cubic, leaveTangent=-0.1794, tangentMode=RCTM_User, time=0, value=1}
         {arriveTangent=-1.8995, interpMode=RCIM_Cubic, tangentMode=RCTM_Break, time=0.5, value=0.25}
        ]
       }}})
    Index 0 Param 2=1
    Index 0 Param 3=0.33
    Module.Index0Float0Write=true
    Module.Index0Float1Write=true
    Module.Index0Float2Write=true
    Module.Index0Float3Write=true
   )
   &Color(Color=&RandomRangeLinearColor(Minimum=$User.CubeColorMin, Maximum=$User.CubeColorMax))
   &Drag(Drag=&RandomRangeFloat(Minimum=5, Maximum=7.5), Use Linear Drag=true, Use Rotational Drag=true)
   &SolveForcesAndVelocity(Clamp Velocity=false, Limit Acceleration=false)
  ]
  rendererClasses [&NiagaraMeshRendererProperties]
  renderers [{rendererClass=&NiagaraMeshRendererProperties, rendererIndex=0}]
 }
 Flare {
  emitterUpdateScript [
   &EmitterState(Life Cycle Mode=::NewEnumerator1, Loop Behavior=::NewEnumerator1, Loop Duration=1, UseLoopDelay=true)
   &SpawnBurst_Instantaneous(Spawn Count=1, Use Spawn Probability=false)
  ]
  particleSpawnScript [
   &InitializeParticle(
    Color=<rgba 0 0.8915 1.25 1>
    Lifetime=0.25
    Sprite Rotation Mode=::NewEnumerator2
    Sprite Size Mode=::NewEnumerator3
    Uniform Sprite Size=750
    UsePositionOffset=false
    Write Material Random=true
   )
   &SystemLocation(MaskSpawnGroup=false, Offset=<-100 0 100>)
   &DynamicMaterialParameters(
    Index 0 Param 1=&RandomRangeFloat(Minimum=0.4, Maximum=0.6)
    Index 0 Param 2=1
    Index 0 Param 3=1
    Index 0 Param 4=1
    Module.Param0WriteEnabled=true
    Module.Param1WriteEnabled=true
    Module.Param2WriteEnabled=true
    Module.Param3WriteEnabled=true
   )
  ]
  particleUpdateScript [
   &ParticleState()
   &SolveForcesAndVelocity(Clamp Velocity=false, Limit Acceleration=false)
   &ScaleColor(
    Scale Alpha=&Multiply_Float(
     A=&FNTimeOfDayScalar(Night Scale=0.5, Morning Scale=0.3, Evening Scale=0.4)
     B=&FloatFromCurve(FloatCurve=DataInterface() {propertyValues {Curve {keys [<at 0 0>,<at 0.2 1>,<at 0.8 1>,<at 1 0>]}}})
    )
    ScaleA=true
    ScaleRGB=true
   )
   &ScaleSpriteSize(Scale Factor=&Vector2DFromFloat(Value=&FloatFromCurve(FloatCurve=DataInterface() {propertyValues {Curve {keys [<at 0 1.5 RCIM_Cubic>,<at 1 0>]}}})))
   &CameraOffset(Camera Offset Amount=50)
  ]
  rendererClasses [&NiagaraSpriteRendererProperties]
  renderers [{rendererClass=&NiagaraSpriteRendererProperties}]
 }
 Shock_wave {
  emitterUpdateScript [
   &EmitterState()
   &SpawnBurst_Instantaneous(Spawn Count=1, Use Spawn Probability=false)
  ]
  particleSpawnScript [
   &InitializeParticle(Lifetime=0.45, Mesh Scale Mode=::NewEnumerator3, Mesh Uniform Scale=3, UsePositionOffset=false, Write Material Random=true) {module=InitializeParticle001}
   &SystemLocation(MaskSpawnGroup=false, Offset=<-100 0 100>)
   &InitialMeshOrientation(Rotation=<0 0 0.5>, Use Orientation Vector=true, Use Rotation Vector=true)
   /SetVariables(Particles.Mesh_position02:Vector3f=<-40 0 100>) {isSetParameters=true}
   &CameraOffset(Camera Offset Amount=100)
  ]
  particleUpdateScript [
   /SetVariables(Particles.Mesh_color02:LinearColor=&MakeLinearColorFromVectorAndFloat(
     "Vector (RGB)"=&Multiply_Vector(
      A=&MakeVectorFromLinearColorRGB(Color=<rgba 0 0.4741 0.5 3>)
      B=&MakeVectorFromLinearColorRGB(Color=&ColorFromCurve(ColorCurve=DataInterface() {propertyValues {
          AlphaCurve {keys []}
          BlueCurve {keys [<at 0 30 RCIM_Cubic>,<at 0.9151 0.2 RCIM_Cubic>]}
          GreenCurve {keys [<at 0 20.2222 RCIM_Cubic>,<at 0.9151 0.2 RCIM_Cubic>]}
          RedCurve {keys [<at 0 0.6667 RCIM_Cubic>,<at 0.9151 0.2 RCIM_Cubic>]}
         }}))
     )
     "Float (Alpha)"=&Multiply_Float(
      A=&FloatFromCurve(FloatCurve=DataInterface() {propertyValues {Curve {keys [
           <at 0 0 RCIM_Cubic>
           {interpMode=RCIM_Cubic, leaveTangent=-1.8775, tangentMode=RCTM_Break, time=0.026, value=1}
           {arriveTangent=0.0955, tangentMode=RCTM_User, time=0.8, value=0}
          ]}}})
      B=0.5
     )
    )) {isSetParameters=true}
   /SetVariables(Particles.Mesh_Size02:Vector3f=&Multiply_Vector(
     A=<1.7 1.7 1.7>
     B=&VectorFromCurve(
      VectorCurve=DataInterface() {propertyValues {
        XCurve {keys [
          {arriveTangent=1.6206, interpMode=RCIM_Cubic, leaveTangent=1.6206, tangentMode=RCTM_User, time=0, value=0.1}
          {arriveTangent=0.5903, interpMode=RCIM_Cubic, leaveTangent=0.5903, tangentMode=RCTM_User, time=0.7511, value=1.5857}
         ]}
        YCurve {keys [<at 0 0.25 RCIM_Cubic>, {arriveTangent=2.1842, interpMode=RCIM_Cubic, leaveTangent=2.1842, tangentMode=RCTM_User, time=0.7501, value=2}]}
        ZCurve {keys [<at 0 0.25 RCIM_Cubic>, {arriveTangent=2.1842, interpMode=RCIM_Cubic, leaveTangent=2.1842, tangentMode=RCTM_User, time=0.75, value=2}]}
       }}
      Scale Curve=&Multiply_VectorByFloat(Vector=<1 1 1>, Float=3)
     )
    )) {isSetParameters=true}
   /SetVariables(Particles.Mesh_Dynamic02:Vector4f=&MakeVector4(
     X=&FloatFromCurve(FloatCurve=DataInterface() {propertyValues {Curve {keys [<at 0 0 RCIM_Cubic>,<at 0.8 0.28 RCIM_Cubic>]}}})
     Y=&RandomRangeFloat(Minimum=0, Maximum=1)
     Z=50
     W=0
    )) {isSetParameters=true}
   &ParticleState()
   &SolveForcesAndVelocity(Clamp Velocity=false, Limit Acceleration=false)
   &ScaleColor(
    Scale Alpha=&Multiply_Float(
     A=1
     B=&Multiply_Float(
      A=&FloatFromCurve(FloatCurve=DataInterface() {propertyValues {Curve {keys [<at 0 0 RCIM_Cubic>,<at 0.026 1 RCIM_Cubic>,<at 1 0>]}}})
      B=0.5
     )
    )
    Scale RGB=&MakeVectorFromLinearColorRGB(Color=&ColorFromCurve(ColorCurve=DataInterface() {propertyValues {
        AlphaCurve {keys []}
        BlueCurve {keys [{arriveTangent=2.3578, interpMode=RCIM_Cubic, time=0, value=6.25}, <at 0.6 1 RCIM_Cubic>]}
        GreenCurve {keys [{arriveTangent=-30.4531, interpMode=RCIM_Cubic, time=0, value=8.6458}, <at 0.6 1 RCIM_Cubic>]}
        RedCurve {keys [{arriveTangent=-115.5308, interpMode=RCIM_Cubic, time=0, value=12}, <at 0.6 1 RCIM_Cubic>]}
       }}))
    ScaleA=true
    ScaleRGB=true
   ) {module=ScaleColor001}
   &ScaleMeshSize(Scale Factor=&VectorFromCurve(
     VectorCurve=DataInterface() {propertyValues {
       XCurve {keys [
         {arriveTangent=1.6206, interpMode=RCIM_Cubic, leaveTangent=1.6206, tangentMode=RCTM_User, time=0, value=0.1}
         {arriveTangent=0.5903, interpMode=RCIM_Cubic, leaveTangent=0.5903, tangentMode=RCTM_User, time=0.7511, value=1.5857}
        ]}
       YCurve {keys [<at 0 0.25 RCIM_Cubic>, {arriveTangent=2.1842, interpMode=RCIM_Cubic, leaveTangent=2.1842, tangentMode=RCTM_User, time=0.7501, value=2}]}
       ZCurve {keys [<at 0 0.25 RCIM_Cubic>, {arriveTangent=2.1842, interpMode=RCIM_Cubic, leaveTangent=2.1842, tangentMode=RCTM_User, time=0.75, value=2}]}
      }}
     Scale Curve=&Multiply_VectorByFloat(Vector=<1 1 1>, Float=2)
    ))
   &DynamicMaterialParameters(
    Index 0 Param 1=&FloatFromCurve(FloatCurve=DataInterface() {propertyValues {Curve {keys [<at 0 0 RCIM_Cubic>,<at 1 0.28 RCIM_Cubic>]}}})
    Index 0 Param 2=&RandomRangeFloat(Minimum=0, Maximum=1)
    Index 0 Param 3=50
    Index 0 Param 4=&FloatFromCurve(FloatCurve=DataInterface() {propertyValues {Curve {keys [
         {interpMode=RCIM_Cubic, leaveTangent=0.0377, tangentMode=RCTM_Break, time=0, value=0}
         {arriveTangent=-3.0374, interpMode=RCIM_Cubic, leaveTangent=2.0371, tangentMode=RCTM_Break, time=0.2, value=-1.5}
         {interpMode=RCIM_Cubic, tangentMode=RCTM_Break, time=1, value=0}
        ]}}})
    Module.Param0WriteEnabled=true
    Module.Param1WriteEnabled=true
    Module.Param2WriteEnabled=true
    Module.Param3WriteEnabled=true
   )
  ]
  rendererClasses [&NiagaraMeshRendererProperties]
  renderers [
   {rendererClass=&NiagaraMeshRendererProperties, rendererIndex=0}
   {rendererClass=&NiagaraMeshRendererProperties, rendererIndex=1}
  ]
 }
}
summary {
 systemName=NS_CubeSurfing_Boost_Wind
 emitters [
  {emitterName=Shock_wave, rendererClasses [&NiagaraMeshRendererProperties]}
  {emitterName=Blocks002, rendererClasses [&NiagaraMeshRendererProperties]}
  {emitterName=Flare, rendererClasses [&NiagaraSpriteRendererProperties]}
 ]
 userVariables [
  User.CubeColorMax {type=LinearColor, description="", defaultValue {struct=LinearColor, value=<rgba 0 0.6514 0.7379 1>}}
  User.CubeColorMin {type=LinearColor, description="", defaultValue {struct=LinearColor, value=<rgba 0.0262 0.2016 0.5583 1>}}
 ]
}
