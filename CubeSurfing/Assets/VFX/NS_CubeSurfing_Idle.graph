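# Niagara system for the Cube Surfing idle effect: a blue, pulsing glow built from a single Core emitter.
# One burst particle with looping lifetime feeds a light, three sprite renderers and a mesh renderer.
# Each sprite layer has its own color and size parameters, animated by age-driven curves.
# User.AlphaIntro scales opacity, FNTimeOfDayScalar adjusts brightness by time of day, and system loop age drives the light color curve.
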NiagaraSystem /CubeSurfing/Assets/VFX/NS_CubeSurfing_Idle
dependencies {
 usedDataInterfaces [&NiagaraDataInterfaceColorCurve, &NiagaraDataInterfaceCurve]
 usedDynamicInputs []
 usedModules [&SystemState]
}
emitters {Core {
  emitterUpdateScript [
   &EmitterState(
    Inactive Response=::NewEnumerator1
    Life Cycle Mode=::NewEnumerator1
    Loop Behavior=::NewEnumerator1
    Loop Delay=1
    Loop Duration Mode=::NewEnumerator1
    UseLoopDelay=false
   )
   &SpawnBurst_Instantaneous(Spawn Count=1, Use Spawn Probability=false)
  ]
  particleSpawnScript [
   &InitializeParticle(Lifetime=5, Mesh Scale Mode=::NewEnumerator3, Mesh Uniform Scale=0.1, Position Mode=::NewEnumerator0, Write Material Random=true)
   /SetVariables(
    Particles.Sprite_Color_01:LinearColor=<rgba 0 1.4263 2 1>
    Particles.Sprite_DP_01:Vector4f=&MakeVector4(X=&RandomRangeFloat(Minimum=0.4, Maximum=0.6), Y=1, Z=1, W=1)
    Particles.Sprite_Size_01:Vector2f=&Vector2DFromFloat(Value=90)
   ) {isSetParameters=true}
   /SetVariables(
    Particles.Sprite_Color_02:LinearColor=<rgba 0 0.0654 0.276 0.3>
    Particles.Sprite_DP_02:Vector4f=&MakeVector4(X=&RandomRangeFloat(Minimum=0.4, Maximum=0.6), Y=1, Z=1, W=1)
    Particles.Sprite_Size_02:Vector2f=&Vector2DFromFloat(Value=225)
   ) {isSetParameters=true}
   /SetVariables(
    Particles.Sprite_Color_03:LinearColor=<rgba 0 0.5167 1 1>
    Particles.Sprite_DP_03:Vector4f=<0.6 1 1 1>
    Particles.Sprite_Size_03:Vector2f=&Vector2DFromFloat(Value=100)
   ) {isSetParameters=true}
   &InitialMeshOrientation(Orientation Vector=<0 0 1>, Rotation=<0 0.25 0>, Use Orientation Vector=true, Use Rotation Vector=true) {module=InitialMeshOrientation001}
  ]
  particleUpdateScript [
   &ParticleState(
    Kill Particles When Lifetime Has Elapsed=false
    Let Infinitely Lived Particles Die When Emitter Deactivates=true
    Loop Particles Lifetime=true
   )
   &CameraOffset(Camera Offset Amount=25)
   &ScaleColor(
    Scale Alpha=&Multiply_Float(
     A=&FloatFromCurve(
      FloatCurve=DataInterface() {propertyValues {Curve {
         defaultValue=0
         keys [
          {interpMode=RCIM_Cubic, tangentMode=RCTM_User, time=0, value=0.75}
          {interpMode=RCIM_Cubic, tangentMode=RCTM_User, time=0.5, value=1}
          {tangentMode=RCTM_User, time=1, value=0.75}
         ]
        }}}
      Scale Curve=$User.AlphaIntro
     )
     B=&FNTimeOfDayScalar(Night Scale=0.5, Morning Scale=0.3, Evening Scale=0.4)
    )
    Scale RGB=&MakeVectorFromLinearColorRGB(Color=&ColorFromCurve(ColorCurve=DataInterface() {propertyValues {
        AlphaCurve {keys []}
        BlueCurve {keys [<at 0.2535 1 RCIM_Cubic>]}
        GreenCurve {keys [<at 0.2535 0.763 RCIM_Cubic>]}
        RedCurve {keys [<at 0.2535 0 RCIM_Cubic>]}
       }}))
    ScaleA=true
    ScaleRGB=true
   )
   /SetVariables(
    Particles.Sprite_Color_01:LinearColor=&MultiplyLinearColor(
     Linear Color=$Particles.Initial.Sprite_Color_01
     Scale Factor=&MakeFloat4FromVectorAndFloat(
      "Vector (RGB)"=<1 1 1>
      "Float (Alpha)"=&Multiply_Float(
       A=&FNTimeOfDayScalar(Night Scale=0.5, Morning Scale=0.3, Evening Scale=0.4)
       B=&FloatFromCurve(
        FloatCurve=DataInterface() {propertyValues {Curve {
           defaultValue=0.8
           keys [
            {interpMode=RCIM_Cubic, tangentMode=RCTM_User, time=0, value=0.8}
            {interpMode=RCIM_Cubic, tangentMode=RCTM_User, time=0.3, value=1}
            {arriveTangent=-0.2083, interpMode=RCIM_Cubic, leaveTangent=-0.2083, time=0.5, value=0.8625}
            {arriveTangent=-0.125, interpMode=RCIM_Cubic, leaveTangent=-0.125, time=0.9, value=0.875}
            {tangentMode=RCTM_User, time=1, value=0.8}
           ]
          }}}
        CurveIndex=&Multiply_Float(A=$Particles.NormalizedAge, B=1.75)
        Scale Curve=$User.AlphaIntro
       )
      )
     )
    )
    Particles.Sprite_Size_01:Vector2f=&Multiply_Vector2D(
     A=$Particles.Initial.Sprite_Size_01
     B=&Vector2DFromFloat(Value=&FloatFromCurve(
       FloatCurve=DataInterface() {propertyValues {Curve {
          defaultValue=1.5
          keys [
           {interpMode=RCIM_Cubic, tangentMode=RCTM_User, time=0, value=0.8}
           {interpMode=RCIM_Cubic, tangentMode=RCTM_User, time=0.5, value=1}
           {tangentMode=RCTM_User, time=1, value=0.8}
          ]
         }}}
       CurveIndex=&Multiply_Float(A=$Particles.NormalizedAge, B=1.75)
      ))
    )
   ) {isSetParameters=true}
   /SetVariables(
    Particles.Sprite_Color_02:LinearColor=&MultiplyLinearColor(
     Linear Color=$Particles.Initial.Sprite_Color_02
     Scale Factor=&MakeFloat4FromVectorAndFloat(
      "Vector (RGB)"=&Multiply_VectorByFloat(Vector=<1 1 1>, Float=&FNTimeOfDayScalar(Night Scale=0.5, Morning Scale=0.3, Evening Scale=0.4))
      "Float (Alpha)"=&FloatFromCurve(
       FloatCurve=DataInterface() {propertyValues {Curve {
          defaultValue=0.8
          keys [
           {interpMode=RCIM_Cubic, tangentMode=RCTM_User, time=0, value=0.8}
           {interpMode=RCIM_Cubic, tangentMode=RCTM_User, time=0.3, value=1}
           {arriveTangent=-0.2083, interpMode=RCIM_Cubic, leaveTangent=-0.2083, time=0.5, value=0.8625}
           {arriveTangent=-0.125, interpMode=RCIM_Cubic, leaveTangent=-0.125, time=0.9, value=0.875}
           {tangentMode=RCTM_User, time=1, value=0.8}
          ]
         }}}
       CurveIndex=&Multiply_Float(A=$Particles.NormalizedAge, B=1.75)
       Scale Curve=$User.AlphaIntro
      )
     )
    )
    Particles.Sprite_Size_02:Vector2f=&Multiply_Vector2D(
     A=$Particles.Initial.Sprite_Size_02
     B=&Vector2DFromFloat(Value=&FloatFromCurve(
       FloatCurve=DataInterface() {propertyValues {Curve {
          defaultValue=0.8
          keys [
           {interpMode=RCIM_Cubic, tangentMode=RCTM_User, time=0, value=0.8}
           {interpMode=RCIM_Cubic, tangentMode=RCTM_User, time=0.3, value=1}
           {arriveTangent=-0.2083, interpMode=RCIM_Cubic, leaveTangent=-0.2083, time=0.5, value=0.8625}
           {arriveTangent=-0.125, interpMode=RCIM_Cubic, leaveTangent=-0.125, time=0.9, value=0.875}
           {tangentMode=RCTM_User, time=1, value=0.8}
          ]
         }}}
       CurveIndex=&Multiply_Float(A=$Particles.NormalizedAge, B=1.75)
      ))
    )
   ) {isSetParameters=true}
   /SetVariables(
    Particles.Sprite_Color_03:LinearColor=&MultiplyLinearColor(
     Linear Color=$Particles.Initial.Sprite_Color_03
     Scale Factor=&MakeFloat4FromVectorAndFloat(
      "Vector (RGB)"=<1 1 1>
      "Float (Alpha)"=&Multiply_Float(
       A=&FloatFromCurve(
        FloatCurve=DataInterface() {propertyValues {Curve {
           defaultValue=0.8
           keys [
            {interpMode=RCIM_Cubic, tangentMode=RCTM_User, time=0, value=0.8}
            {interpMode=RCIM_Cubic, tangentMode=RCTM_User, time=0.2, value=0.925}
            {arriveTangent=-0.0721, interpMode=RCIM_Cubic, leaveTangent=-0.0721, time=0.5898, value=0.837}
            {arriveTangent=-0.0901, interpMode=RCIM_Cubic, leaveTangent=-0.0901, time=0.7072, value=0.8884}
            {tangentMode=RCTM_User, time=1, value=0.8}
           ]
          }}}
        CurveIndex=&Multiply_Float(A=$Particles.NormalizedAge, B=1)
        Scale Curve=&Multiply_Float(A=0.05, B=$User.AlphaIntro)
       )
       B=&FNTimeOfDayScalar(Night Scale=1, Morning Scale=0.3, Evening Scale=0.5)
      )
     )
    )
    Particles.Sprite_Size_03:Vector2f=&Multiply_Vector2D(
     A=$Particles.Initial.Sprite_Size_03
     B=&Vector2DFromFloat(Value=&FloatFromCurve(
       FloatCurve=DataInterface() {propertyValues {Curve {
          defaultValue=0.8
          keys [
           {interpMode=RCIM_Cubic, tangentMode=RCTM_User, time=0, value=0.95}
           {interpMode=RCIM_Cubic, tangentMode=RCTM_User, time=0.5, value=1}
           {tangentMode=RCTM_User, time=1, value=0.95}
          ]
         }}}
       CurveIndex=&Multiply_Float(A=$Particles.NormalizedAge, B=1.75)
      ))
    )
   ) {isSetParameters=true}
   &DynamicMaterialParameters(
    Index 0 Param 1=0.1
    Index 0 Param 2=0.2
    Module.Index0Float0Write=true
    Module.Index0Float1Write=true
    Module.Index0Float2Write=true
    Module.Index0Float3Write=true
   )
   &FN_Light_Attributes(
    Color=&MultiplyLinearColor(
     Linear Color=<rgba 10 32.5 100 1>
     Scale Factor=&MakeVector4_FromLinearColor(LinearColor=&ColorFromCurve(
       ColorCurve=DataInterface() {propertyValues {
         AlphaCurve {keys [<at 0.0214 0.5>,<at 0.4872 1>,<at 0.919 0.5>]}
         BlueCurve {keys [<at 0.4799 1>]}
         GreenCurve {keys [<at 0.4799 1>]}
         RedCurve {keys [<at 0.4799 1>]}
        }}
       CurveIndex=&NormalizeFloat(Value=&Multiply_Float(A=$System.NormalizedLoopAge, B=1), Min=0, Max=1)
      ))
    )
    Create Color Attribute=true
    Create Exponent Attribute=true
    Create Radius Attribute=true
    Exponent=5
    Radius=100
   )
  ]
  rendererClasses [
   &NiagaraLightRendererProperties
   &NiagaraSpriteRendererProperties
   &NiagaraMeshRendererProperties
  ]
  renderers [
   {rendererClass=&NiagaraLightRendererProperties, rendererIndex=0}
   {rendererClass=&NiagaraSpriteRendererProperties, rendererIndex=1}
   {rendererClass=&NiagaraSpriteRendererProperties, rendererIndex=2}
   {rendererClass=&NiagaraSpriteRendererProperties, rendererIndex=3}
   {rendererClass=&NiagaraMeshRendererProperties, rendererIndex=4}
  ]
 }}
summary {
 systemName=NS_CubeSurfing_Idle
 emitters [{
   emitterName=Core
   rendererClasses [
    &NiagaraLightRendererProperties
    &NiagaraSpriteRendererProperties
    &NiagaraMeshRendererProperties
   ]
  }]
 userVariables [User.AlphaIntro {type=NiagaraFloat, defaultValue=0, description=""}]
}
