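# Niagara one-shot landing puff for Cube Surfing, layering five burst emitters into a ground impact.
# Mesh blocks tinted by user cube colors fly out and slow under drag, alongside dust sprites, air streaks and an expanding upward-facing ring.
# Wind_Radial uses spawn groups as visibility tags to drive four mesh renderers, each with its own curve-animated size and color.
# System time-of-day color tints streaks, ring and wind; ScaleFX sizes streaks and wind.
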NiagaraSystem /CubeSurfing/Assets/VFX/NS_CubeSurfing_PuffLand
dependencies {
 usedDataInterfaces [
  &NiagaraDataInterfaceCurve
  &NiagaraDataInterfaceVector2DCurve
  &NiagaraDataInterfaceVectorCurve
  &NiagaraDataInterfaceColorCurve
 ]
 usedDynamicInputs [
  &MakeLinearColorFromVectorAndFloat
  &ReturnNormalizedExecIndex
  &MergeHelper_VectorSwitch
  &SpawnGroup
  &SimulationPosition
 ]
 usedModules [
  &self:SystemSpawnScript.SystemScriptSource.SystemScriptGraph.NiagaraNodeAssignment_0.SetVariables_#1
  &SystemState
  &/TimeOfDayGlobal
 ]
}
emitters {
 AirStreaks {
  emitterUpdateScript [
   &EmitterState(Life Cycle Mode=::NewEnumerator1, Loop Behavior=::NewEnumerator1, Loop Duration=1, UseLoopDelay=false)
   &SpawnBurst_Instantaneous(Spawn Count=18, Spawn Time=0.1, Use Spawn Probability=false, UseLoopCountLimit=false)
  ]
  particleSpawnScript [
   /SetVariables(Particles.RandomFloat=&UniformRangedFloat(Minimum=0, Maximum=3)) {isSetParameters=true}
   &InitializeParticle(
    Color Maximum=<rgba 0.25 0.215 0.2057 1>
    Color Minimum=<rgba 0.15 0.1459 0.1409 1>
    Color Mode=::NewEnumerator2
    Lifetime Max=0.3
    Lifetime Min=0.15
    Lifetime Mode=::NewEnumerator1
    Sprite Size Max=<60 200>
    Sprite Size Min=<50 150>
    Sprite Size Mode=::NewEnumerator2
    UsePositionOffset=false
    Write Material Random=true
   ) {module=InitializeParticle001}
   /SetVariables(Particles.SpriteSize=&Multiply_Vector2D_ByFloat(Float=&Multiply_Float(A=$System.ScaleFX, B=1), Vector2D=$Particles.SpriteSize)) {isSetParameters=true}
   &ShapeLocation(
    Handle Radius=15
    Large Radius=50
    MaskSpawnGroup=false
    Non Uniform Scale=&Multiply_VectorByFloat(Vector=<1 1 0>, Float=$System.ScaleFX)
    Offset=<0 0 20>
    Offset Mode=::NewEnumerator0
    Override Seed=false
    Shape Primitive=::NewEnumerator3
    UseOwnerScale=true
   )
   &AddVelocity(Module.Use Falloff Distance=false, Override Seed=false, Velocity Mode=::NewEnumerator1, Velocity Speed=500)
   &ApplyInitialForces()
   &SubUVAnimation(End Frame=3, SubUV Animation Mode:ENiagara_SubUVLookupMode=::NewEnumerator2)
  ]
  particleUpdateScript [
   &ParticleState() {module=UpdateAge}
   &SolveForcesAndVelocity(Clamp Velocity=false, Limit Acceleration=false)
   &ScaleSpriteSize(Scale Factor=&Vector2DFromCurve(
     Vector2Curve=DataInterface() {propertyValues {XCurve {keys [<at 0.0036 1.6897>,<at 0.9967 0.0582>]}, YCurve {keys [<at 0 1>,<at 1.002 3.7227>]}}}
     ScaleCurve=&Multiply_Vector2D_ByFloat(Float=0.65, Vector2D=<1 1>)
    ))
   &ScaleColor(
    Scale Alpha=&FloatFromCurve(
     FloatCurve=DataInterface() {propertyValues {Curve {keys [{interpMode=RCIM_Cubic, tangentMode=RCTM_Break, time=0, value=1}, {interpMode=RCIM_Cubic, tangentMode=RCTM_Break, time=1, value=0}]}}}
     Scale Curve=&RandomRangeFloat(Minimum=0.05, Maximum=0.2)
    )
    Scale RGB=&MakeVectorFromLinearColorRGB(Color=$System.TODColor)
    ScaleA=true
    ScaleRGB=true
   )
   &DynamicMaterialParameters(
    Index 0 Param 2=1
    Index 0 Param 3=1
    Index 0 Param 4=&FloatFromCurve(FloatCurve=DataInterface() {propertyValues {Curve {keys [<at 0.3 1>,<at 1 0.25>]}}})
    Module.Param0WriteEnabled=true
    Module.Param1WriteEnabled=true
    Module.Param2WriteEnabled=true
    Module.Param3WriteEnabled=true
   )
  ]
  rendererClasses [&NiagaraSpriteRendererProperties]
  renderers [{rendererClass=&NiagaraSpriteRendererProperties}]
 }
 Block_Burst {
  emitterUpdateScript [
   &EmitterState(Life Cycle Mode=::NewEnumerator1, Loop Behavior=::NewEnumerator1, Loop Delay=0.05, Loop Duration=1, UseLoopDelay=true)
   &SpawnBurst_Instantaneous(
    Spawn Count=&RandomRangeInt(Minimum=33, Maximum=22)
    Use Spawn Probability=false
    UseLoopCountLimit=false
   )
  ]
  particleSpawnScript [
   &InitializeParticle(
    Color Maximum=$User.CubeColorMax
    Color Minimum=$User.CubeColorMin
    Color Mode=::NewEnumerator2
    Lifetime Max=0.5
    Lifetime Min=0.25
    Lifetime Mode=::NewEnumerator1
    Mesh Scale Mode=::NewEnumerator1
    Mesh Uniform Scale Max=3
    Mesh Uniform Scale Min=2
    Sprite Size Mode=::NewEnumerator3
    Uniform Sprite Size=7.5
    UsePositionOffset=false
    Write Material Random=true
   )
   &ShapeLocation(
    MaskSpawnGroup=false
    Non Uniform Scale=<1 1 0>
    Offset=<0 0 5>
    Offset Mode=::NewEnumerator0
    Override Seed=false
    Ring Radius=50
    Shape Primitive=::NewEnumerator4
    UseOwnerScale=true
   )
   &AddVelocity(
    Module.Use Falloff Distance=false
    Override Seed=false
    Velocity Mode=::NewEnumerator1
    Velocity Speed=&RandomRangeFloat(Minimum=250, Maximum=1500)
   ) {module=AddVelocity001}
  ]
  particleUpdateScript [
   &ParticleState()
   &ScaleMeshSize(Scale Factor=&Lerp_Vector(
     A=&RandomRangeVector(Minimum=<0.1 0.1 0.1>, Maximum=<0.25 0.25 0.25>)
     B=<1 1 1>
     Alpha=&FloatFromCurve(FloatCurve=DataInterface() {propertyValues {Curve {
         defaultValue=1
         keys [
          {arriveTangent=9.0725, interpMode=RCIM_Cubic, leaveTangent=9.0725, tangentMode=RCTM_User, time=0, value=0}
          {arriveTangent=-2.5249, interpMode=RCIM_Cubic, leaveTangent=-2.5249, tangentMode=RCTM_User, time=0.1408, value=0.9847}
          {arriveTangent=-0.0405, interpMode=RCIM_Cubic, leaveTangent=-0.0405, tangentMode=RCTM_User, time=1, value=0}
         ]
        }}})
    ))
   &ScaleColor(
    Scale Alpha=&FloatFromCurve(
     FloatCurve=DataInterface() {propertyValues {Curve {defaultValue=1, keys [<at 0 0>, <at 0.2 1>, {arriveTangent=-1.3333, leaveTangent=-1.3333, time=0.95, value=1}, <at 1 0>]}}}
     CurveIndex=$Particles.Mask
    )
    ScaleA=true
    ScaleRGB=true
   )
   &DynamicMaterialParameters(
    Index 0 Param 1=&FloatFromCurve(
     FloatCurve=DataInterface() {propertyValues {Curve {
        defaultValue=0
        keys [
         {arriveTangent=-0.1794, interpMode=RCIM_Cubic, leaveTangent=-0.1794, tangentMode=RCTM_User, time=0, value=1}
         {arriveTangent=-1.8995, interpMode=RCIM_Cubic, tangentMode=RCTM_Break, time=0.5, value=0.25}
        ]
       }}}
     CurveIndex=&NormalizeDistanceRange(
      Start Position:NiagaraPosition=/Niagara/DynamicInputs/Helpers/SimulationPosition()
      End Position:NiagaraPosition=$Particles.Position
      Distance:NiagaraFloat=150
      Invert Normalized Range:NiagaraBool=false
     )
    )
    Index 0 Param 2=1
    Index 0 Param 3=0.33
    Module.Index0Float0Write=true
    Module.Index0Float1Write=true
    Module.Index0Float2Write=true
    Module.Index0Float3Write=true
   )
   &Color(Color=&RandomRangeLinearColor(Minimum=$User.CubeColorMin, Maximum=$User.CubeColorMax))
   &Drag(Drag=7.5, Use Linear Drag=true, Use Rotational Drag=true)
   &SolveForcesAndVelocity(Clamp Velocity=false, Limit Acceleration=false)
   &AlignSpriteToMeshOrientation(SetAlignment=true, SetFacing=true)
  ]
  rendererClasses [&NiagaraMeshRendererProperties]
  renderers [{rendererClass=&NiagaraMeshRendererProperties, rendererIndex=0}]
 }
 Dust {
  emitterUpdateScript [
   &EmitterState()
   &SpawnBurst_Instantaneous(Spawn Count=&UniformRangedInt(Minimum=25, Maximum=35), Use Spawn Probability=false)
  ]
  particleSpawnScript [
   /SetVariables(Particles.VelCurveRand:NiagaraFloat=&FloatFromCurve(
     FloatCurve=DataInterface() {propertyValues {Curve {keys [
         {interpMode=RCIM_Cubic, leaveTangent=0.1132, tangentMode=RCTM_Break, time=0, value=0}
         {arriveTangent=1.9858, interpMode=RCIM_Cubic, leaveTangent=1.7203, tangentMode=RCTM_Break, time=0.3, value=0.4}
         {interpMode=RCIM_Cubic, tangentMode=RCTM_Break, time=1, value=1}
        ]}}}
     CurveIndex=&RandomRangeFloat(Minimum=0, Maximum=1)
    )) {isSetParameters=true}
   /SetVariables(Particles.InitSpriteRotRate=&RandomRangeFloat(Minimum=-1, Maximum=1)) {isSetParameters=true}
   &InitializeParticle(
    Color Maximum=<rgba 0.3229 0.3187 0.3162 0.25>
    Color Minimum=<rgba 0.3542 0.3433 0.3339 0.5>
    Color Mode=::NewEnumerator2
    Lifetime Max=0.5
    Lifetime Min=0.35
    Lifetime Mode=::NewEnumerator1
    Sprite Rotation Angle=&RandomRangeFloat(Minimum=0, Maximum=360)
    Sprite Rotation Mode=::NewEnumerator0
    Sprite Size Mode=::NewEnumerator3
    Uniform Sprite Size=&Multiply_Float(A=&RandomRangeFloat(Minimum=50, Maximum=150), B=1.5)
    UsePositionOffset=false
    Write Material Random=true
   )
   /SetVariables(Particles.SpriteSize=&Multiply_Vector2D_ByFloat(Float=1.5, Vector2D=$Particles.SpriteSize)) {isSetParameters=true}
   &ShapeLocation(
    Handle Radius=0.1
    Large Radius=5
    MaskSpawnGroup=false
    Offset=<0 0 20>
    Offset Coordinate Space=::World
    Offset Mode=::NewEnumerator0
    Override Seed=false
    Rotation Coordinate Space=::World
    Shape Primitive=::NewEnumerator3
    UseOwnerScale=true
   )
   &AddVelocityFromPoint(
    Module.Use Falloff Distance=false
    Origin Offset=<0 0 20>
    Origin Offset Coordinate Space=::Simulation
    Velocity Strength=&Multiply_Float(A=&Lerp_Float(A=1000, B=2000, Alpha=$Particles.VelCurveRand), B=1)
   )
  ]
  particleUpdateScript [
   &ParticleState()
   &Drag(Drag=8, Use Linear Drag=true, Use Rotational Drag=false)
   &SpriteRotationRate(Rotation Rate=&Multiply_Float(
     A=$Particles.InitSpriteRotRate
     B=&FloatFromCurve(
      FloatCurve=DataInterface() {propertyValues {Curve {keys [
          {arriveTangent=-4.4624, interpMode=RCIM_Cubic, leaveTangent=-4.4624, tangentMode=RCTM_User, time=0, value=1}
          {arriveTangentWeight=0.1718, interpMode=RCIM_Cubic, tangentMode=RCTM_User, time=0.5, value=0.05}
         ]}}}
      Scale Curve=2000
     )
    ))
   &ScaleColor(
    Scale Alpha=&Multiply_Float(
     A=&RandomRangeFloat(Minimum=0.25, Maximum=0.5)
     B=&FloatFromCurve(FloatCurve=DataInterface() {propertyValues {Curve {keys [<at 0.1 1>,<at 0.5 0>]}}}, CurveIndex=$Emitter.Age)
    )
    Scale RGB=&VectorFromFloat(Value=2)
    ScaleA=true
    ScaleRGB=true
   )
   &ScaleSpriteSize(Scale Factor=&Vector2DFromFloat(Value=&FloatFromCurve(FloatCurve=DataInterface() {propertyValues {Curve {keys [
          {arriveTangent=0.0418, interpMode=RCIM_Cubic, leaveTangent=0.0418, tangentMode=RCTM_User, time=0, value=0.5}
          {arriveTangent=0.8771, interpMode=RCIM_Cubic, leaveTangent=0.8771, tangentMode=RCTM_User, time=0.7, value=0.775}
          {arriveTangent=0.7996, arriveTangentWeight=0.3375, interpMode=RCIM_Cubic, leaveTangent=0.7996, tangentMode=RCTM_User, time=1, value=1}
         ]}}})))
   &DynamicMaterialParameters(
    Index 0 Param 1=&Add_Float(
     A=&RandomRangeFloat(Minimum=0, Maximum=1)
     B=&FloatFromCurve(
      FloatCurve=DataInterface() {propertyValues {Curve {keys [
          {arriveTangent=2.7533, interpMode=RCIM_Cubic, leaveTangent=1.8958, tangentMode=RCTM_User, time=0, value=1}
          {arriveTangent=0.8292, interpMode=RCIM_Cubic, leaveTangent=0.8292, tangentMode=RCTM_User, time=0.321, value=1.3453}
          {arriveTangent=0.0529, interpMode=RCIM_Cubic, leaveTangent=0.2021, tangentMode=RCTM_User, time=1, value=1.5}
         ]}}}
      CurveIndex=&Add_Float(A=&Multiply_Float(A=$Particles.Initial.MaterialRandom, B=0.5), B=$Particles.NormalizedAge)
     )
    )
    Index 0 Param 2=&FloatFromCurve(FloatCurve=DataInterface() {propertyValues {Curve {keys [
         {arriveTangent=0.0453, interpMode=RCIM_Cubic, tangentMode=RCTM_Break, time=0, value=-1}
         {arriveTangent=0.0604, interpMode=RCIM_Cubic, leaveTangent=0.0604, tangentMode=RCTM_User, time=1, value=-1}
        ]}}})
    Index 0 Param 3=&FloatFromCurve(
     FloatCurve=DataInterface() {propertyValues {Curve {keys [
         {arriveTangent=0.0632, interpMode=RCIM_Cubic, leaveTangent=0.0632, tangentMode=RCTM_User, time=0, value=1}
         {
          arriveTangent=-1.6858
          arriveTangentWeight=0.0333
          interpMode=RCIM_Cubic
          leaveTangent=-1.6858
          leaveTangentWeight=0.3137
          tangentMode=RCTM_User
          time=0.05
          value=0.1
         }
         {arriveTangentWeight=0.1027, interpMode=RCIM_Cubic, tangentMode=RCTM_Break, time=0.2, value=0}
        ]}}}
     Scale Curve=&RandomRangeFloat(Minimum=0.25, Maximum=0.5)
    )
    Index 0 Param 4=&FloatFromCurve(FloatCurve=DataInterface() {propertyValues {Curve {keys [<at 0.5 1>,<at 1 0>]}}})
    Index 1 Param 1=&FloatFromCurve(
     FloatCurve=DataInterface() {propertyValues {Curve {keys [
         {interpMode=RCIM_Cubic, tangentMode=RCTM_Break, time=0, value=1}
         {arriveTangentWeight=0.0833, interpMode=RCIM_Cubic, tangentMode=RCTM_Break, time=0.25, value=0.3}
        ]}}}
     Scale Curve=&RandomRangeFloat(Minimum=1.5, Maximum=3)
    )
    Index 1 Param 2=&FloatFromCurve(
     FloatCurve=DataInterface() {propertyValues {Curve {keys [
         {interpMode=RCIM_Cubic, leaveTangent=-2.5489, tangentMode=RCTM_Break, time=0, value=1}
         {interpMode=RCIM_Cubic, leaveTangent=0.9066, tangentMode=RCTM_Break, time=1, value=0}
        ]}}}
     Scale Curve=&RandomRangeFloat(Minimum=0.01, Maximum=0.1)
    )
    Index 1 Param 3=&FloatFromCurve(FloatCurve=DataInterface() {propertyValues {Curve {keys [
         {interpMode=RCIM_Cubic, leaveTangent=6.3917, tangentMode=RCTM_Break, time=0, value=0.2}
         {arriveTangent=1.7549, interpMode=RCIM_Cubic, leaveTangent=10.6767, tangentMode=RCTM_User, time=0.25, value=1}
        ]}}})
    Index 1 Param 4=&FloatFromCurve(
     FloatCurve=DataInterface() {propertyValues {Curve {keys [{arriveTangent=-0.136, interpMode=RCIM_Cubic, leaveTangent=-0.136, tangentMode=RCTM_User, time=0, value=1}, <at 1 0.95>]}}}
     Scale Curve=&RandomRangeFloat(Minimum=1.3, Maximum=2)
    )
    Index 2 Param 1=&FloatFromCurve(FloatCurve=DataInterface() {propertyValues {Curve {keys [
         {interpMode=RCIM_Cubic, leaveTangent=4.2485, tangentMode=RCTM_Break, time=0, value=0}
         {arriveTangent=1.0548, interpMode=RCIM_Cubic, leaveTangent=1.0548, tangentMode=RCTM_User, time=0.17, value=0.5715}
         <at 1 1>
        ]}}})
    Index 2 Param 2=&FloatFromCurve(
     FloatCurve=DataInterface() {propertyValues {Curve {keys [
         {arriveTangent=1.9056, interpMode=RCIM_Cubic, leaveTangent=1.9056, tangentMode=RCTM_User, time=0, value=0.2}
         {arriveTangent=0.0259, interpMode=RCIM_Cubic, leaveTangent=0.0259, tangentMode=RCTM_User, time=1, value=1.4}
        ]}}}
     Scale Curve=&RandomRangeFloat(Minimum=0.1, Maximum=0.2)
    )
    Index 2 Param 3=&FloatFromCurve(FloatCurve=DataInterface() {propertyValues {Curve {keys [
         {interpMode=RCIM_Cubic, leaveTangent=2.3064, tangentMode=RCTM_Break, time=0, value=1}
         {interpMode=RCIM_Cubic, leaveTangent=11.1233, tangentMode=RCTM_Break, time=1, value=2}
        ]}}})
    Index 2 Param 4=&FloatFromCurve(
     FloatCurve=DataInterface() {propertyValues {Curve {keys [
         {arriveTangent=-0.6867, interpMode=RCIM_Cubic, leaveTangent=-0.6867, tangentMode=RCTM_User, time=0, value=1}
         {interpMode=RCIM_Cubic, leaveTangent=-1.2956, tangentMode=RCTM_Break, time=1, value=0.75}
        ]}}}
     Scale Curve=1.5
    )
    Module.Param0WriteEnabled=true
    Module.Param0WriteEnabled1=true
    Module.Param0WriteEnabled2=true
    Module.Param1WriteEnabled=false
    Module.Param1WriteEnabled1=true
    Module.Param1WriteEnabled2=true
    Module.Param2WriteEnabled=false
    Module.Param2WriteEnabled1=true
    Module.Param2WriteEnabled2=true
    Module.Param3WriteEnabled=true
    Module.Param3WriteEnabled1=true
    Module.Param3WriteEnabled2=true
    Write Parameter Index 1=true
    Write Parameter Index 2=true
   )
   &SolveForcesAndVelocity(Clamp Velocity=false, Limit Acceleration=false)
  ]
  rendererClasses [&NiagaraSpriteRendererProperties]
  renderers [
   {rendererClass=&NiagaraSpriteRendererProperties}
   {rendererClass=&NiagaraSpriteRendererProperties, rendererIndex=1}
  ]
 }
 ImpactRing {
  emitterUpdateScript [
   &EmitterState()
   &SpawnBurst_Instantaneous(Spawn Count=1, Use Spawn Probability=false)
  ]
  particleSpawnScript [
   &InitializeParticle(
    Color=<rgba 0.6731 0.9603 1 1>
    Lifetime=0.3
    Mass Mode=::NewEnumerator1
    Sprite Rotation Angle=&UniformRangedFloat(Minimum=-180, Maximum=180)
    Sprite Rotation Mode=::NewEnumerator0
    Sprite Size=&Vector2DFromFloat(Value=200)
    Sprite Size Mode=::NewEnumerator4
    UsePositionOffset=false
    Write Material Random=true
   ) {module=InitializeParticle001}
   /SetVariables(Particles.SpriteFacing=<0 0 1>) {isSetParameters=true}
   &CameraOffset(Camera Offset Amount=50)
  ]
  particleUpdateScript [
   &ParticleState()
   &ScaleSpriteSize(Scale Factor=&Vector2DFromFloat(Value=&Lerp_Float(
      A=0.5
      B=4
      Alpha=&FloatFromCurve(
       FloatCurve=DataInterface() {propertyValues {Curve {keys [
           {interpMode=RCIM_Cubic, leaveTangentWeight=0.3857, tangentMode=RCTM_User, time=0, value=0}
           {arriveTangent=1.0287, interpMode=RCIM_Cubic, leaveTangent=1.0287, tangentMode=RCTM_User, time=0.3, value=0.7}
           {interpMode=RCIM_Cubic, tangentMode=RCTM_User, time=1, value=1}
          ]}}}
       Scale Curve=0.75
      )
     )))
   &ScaleColor(
    Scale Alpha=&FloatFromCurve(FloatCurve=DataInterface() {propertyValues {Curve {keys [
         {arriveTangent=3.3842, interpMode=RCIM_Cubic, leaveTangent=3.3842, tangentMode=RCTM_User, time=0, value=0.7}
         {arriveTangentWeight=0.132, interpMode=RCIM_Cubic, leaveTangentWeight=0.5288, tangentMode=RCTM_User, time=0.2, value=1}
         {interpMode=RCIM_Cubic, tangentMode=RCTM_User, time=0.5, value=1}
         {arriveTangentWeight=0.2973, interpMode=RCIM_Cubic, tangentMode=RCTM_User, time=1.0011, value=0.002}
        ]}}})
    Scale RGB=&MakeVectorFromLinearColorRGB(Color=$System.TODColor)
    ScaleA=true
    ScaleRGB=true
   )
  ]
  rendererClasses [&NiagaraSpriteRendererProperties]
  renderers [{rendererClass=&NiagaraSpriteRendererProperties}]
 }
 Wind_Radial {
  emitterUpdateScript [
   &EmitterState(Life Cycle Mode=::NewEnumerator1, Loop Behavior=::NewEnumerator1, Loop Duration=1, UseLoopDelay=false)
   &SpawnBurst_Instantaneous(Use Spawn Probability=false, UseLoopCountLimit=false)
   &SpawnBurst_Instantaneous(Spawn Group=1, Use Spawn Probability=false, UseLoopCountLimit=false) {module=SpawnBurst_Instantaneous001}
   &SpawnBurst_Instantaneous(Spawn Count=2, Spawn Group=2, Spawn Time=0.1, Use Spawn Probability=false, UseLoopCountLimit=false) {module=SpawnBurst_Instantaneous002}
   &SpawnBurst_Instantaneous(Spawn Count=4, Spawn Group=3, Spawn Time=0.1, Use Spawn Probability=false, UseLoopCountLimit=false) {module=SpawnBurst_Instantaneous003}
  ]
  particleSpawnScript [
   /SetVariables(Particles.SpawnGroup=$Emitter.SpawnGroup) {isSetParameters=true}
   /SetVariables(Particles.VisibilityTag=$Particles.SpawnGroup) {isSetParameters=true}
   &InitializeParticle(Color Mode=::NewEnumerator0, UsePositionOffset=false, Write Material Random=true)
   /SetVariables(
    Particles.Mesh_01_CameraOffset=100
    Particles.Mesh_01_Color:LinearColor=<rgba 0.7091 0.5906 0.4849 2>
    Particles.Mesh_01_Orientation:Quat4f=&OrientMeshToVector(Look At Direction=<1 0 0>, Reference Axis Vector=<1 0 0>)
    Particles.Mesh_01_Size:Vector3f=&VectorFromFloat(Value=&Multiply_Float(A=3, B=$System.ScaleFX))
   ) {isSetParameters=true}
   /SetVariables(
    Particles.Mesh_02_Color:LinearColor=<rgba 0.3162 0.3155 0.315 1>
    Particles.Mesh_02_Orientation:Quat4f=&OrientMeshToVector(Look At Direction=<0 1 0>, Reference Axis Vector=<1 0 0>)
    Particles.Mesh_02_Size:Vector3f=&Multiply_VectorByFloat(Vector=<0.6 0.05 0.6>, Float=$System.ScaleFX)
   ) {isSetParameters=true}
   /SetVariables(
    Particles.Mesh_03_Color:LinearColor=<rgba 0.8609 0.7779 0.6695 0.375>
    Particles.Mesh_03_Orientation:Quat4f=&MultiplyQuaternion(
     Quaternion A=&OrientMeshToVector(Look At Direction=<1 0 0>, Reference Axis Vector=<1 0 0>)
     Quaternion B=&MakeQuatFromAxisAngle(Axis=<0 0 1>, AngleInDegrees=&RandomRangeFloat(Minimum=-180, Maximum=180))
    )
    Particles.Mesh_03_Size:Vector3f=&VectorFromFloat(Value=&Multiply_Float(A=1.8, B=$System.ScaleFX))
   ) {isSetParameters=true}
   /SetVariables(
    Particles.Mesh_04_Color:LinearColor=<rgba 0.9143 0.8098 0.7338 0.275>
    Particles.Mesh_04_Orientation:Quat4f=&MultiplyQuaternion(
     Quaternion A=&MultiplyQuaternion(
      Quaternion A=&OrientMeshToVector(Look At Direction=<1 0 0>, Reference Axis Vector=<1 0 0>)
      Quaternion B=&MakeQuatFromAxisAngle(Axis=<0 0 1>, AngleInDegrees=&RandomRangeFloat(Minimum=-180, Maximum=180))
     )
     Quaternion B=&MakeQuatFromAxisAngle(
      Axis=<1 0 0>
      AngleInDegrees=&RandomRangeFloat(Minimum=-20, Maximum=20, Randomness Mode=::NewEnumerator1)
     )
    )
    Particles.Mesh_04_Size:Vector3f=&VectorFromFloat(Value=&Multiply_Float(A=&RandomRangeFloat(Minimum=1.25, Maximum=4), B=$System.ScaleFX))
   ) {isSetParameters=true}
   &SystemLocation(MaskSpawnGroup=false, Offset=<0 0 10>)
   &CameraOffset(Camera Offset Amount=25)
  ]
  particleUpdateScript [
   &ParticleState()
   /SetVariables(
    Particles.Mesh_01_Color:LinearColor=&MultiplyLinearColor(
     Linear Color=$Particles.Initial.Mesh_01_Color
     Scale Factor=&MakeFloat4FromVectorAndFloat(
      "Vector (RGB)"=&VectorFromCurve(
       VectorCurve=DataInterface() {propertyValues {XCurve {keys [<at 0 1>,<at 1 0.526>]}, YCurve {keys [<at 0 1>,<at 1 0.526>]}, ZCurve {keys [<at 0 1>,<at 1 0.526>]}}}
       CurveIndex=&Divide_Float(A=$Particles.NormalizedAge, B=0.7)
       Scale Curve=&MakeVectorFromLinearColorRGB(Color=$System.TODColor)
      )
      "Float (Alpha)"=&FloatFromCurve(
       FloatCurve=DataInterface() {propertyValues {Curve {keys [<at 0 1>,<at 0.1 0.5>,<at 1 0>]}}}
       CurveIndex=&Divide_Float(A=$Particles.NormalizedAge, B=0.7)
       Scale Curve=0.075
      )
     )
    )
    Particles.Mesh_01_DynamicParameter:Vector4f=&MakeVector4(
     X=&FloatFromCurve(
      FloatCurve=DataInterface() {propertyValues {Curve {keys [{arriveTangent=2.5068, interpMode=RCIM_Cubic, leaveTangent=2.5068, tangentMode=RCTM_User, time=0, value=-0.4}, <at 1 0.55>]}}}
      CurveIndex=&Divide_Float(A=$Particles.NormalizedAge, B=0.7)
     )
     Y=1
     Z=0
     W=0
    )
    Particles.Mesh_01_Size:Vector3f=&Multiply_Vector(
     A=$Particles.Initial.Mesh_01_Size
     B=&VectorFromFloat(Value=&FloatFromCurve(
       FloatCurve=DataInterface() {propertyValues {Curve {keys [
           {arriveTangent=3.5711, interpMode=RCIM_Cubic, leaveTangent=3.5711, tangentMode=RCTM_User, time=0, value=0}
           {interpMode=RCIM_Cubic, tangentMode=RCTM_User, time=1, value=1.5}
          ]}}}
       CurveIndex=&Divide_Float(A=$Particles.NormalizedAge, B=0.7)
       Scale Curve=1.25
      ))
    )
   ) {isSetParameters=true}
   /SetVariables(
    Particles.Mesh_02_Color:LinearColor=&MultiplyLinearColor(
     Linear Color=$Particles.Initial.Mesh_02_Color
     Scale Factor=&MakeFloat4FromVectorAndFloat(
      "Vector (RGB)"=&VectorFromCurve(
       VectorCurve=DataInterface() {propertyValues {
         XCurve {keys [
           <at 0 1 RCIM_Cubic>
           {arriveTangent=9.97, interpMode=RCIM_Cubic, leaveTangent=9.97, time=0.1536, value=1}
           <at 0.776 8.7362 RCIM_Cubic>
           <at 1 1 RCIM_Cubic>
          ]}
         YCurve {keys [
           <at 0 1 RCIM_Cubic>
           {arriveTangent=8.4018, interpMode=RCIM_Cubic, leaveTangent=8.4018, time=0.1536, value=0.6932}
           {arriveTangent=0.3625, interpMode=RCIM_Cubic, leaveTangent=0.3625, time=0.7783, value=7.5392}
           <at 1 1 RCIM_Cubic>
          ]}
         ZCurve {keys [
           <at 0 1 RCIM_Cubic>
           {arriveTangent=7.6288, interpMode=RCIM_Cubic, leaveTangent=7.6288, time=0.1536, value=0.1681}
           {arriveTangent=0.9829, interpMode=RCIM_Cubic, leaveTangent=0.9829, time=0.7781, value=6.9361}
           <at 1 1 RCIM_Cubic>
          ]}
        }}
       CurveIndex=&Divide_Float(A=$Particles.NormalizedAge, B=0.3)
       Scale Curve=&Multiply_Vector(A=&MakeVectorFromLinearColorRGB(Color=$System.TODColor), B=<5 5 5>)
      )
      "Float (Alpha)"=&FloatFromCurve(
       FloatCurve=DataInterface() {propertyValues {Curve {keys [
           {interpMode=RCIM_Cubic, leaveTangentWeight=0.3333, tangentMode=RCTM_User, time=0, value=0}
           {
            arriveTangent=-3.9986
            arriveTangentWeight=0.0551
            interpMode=RCIM_Cubic
            leaveTangent=-3.9986
            leaveTangentWeight=0.2783
            tangentMode=RCTM_User
            time=0.1652
            value=1
           }
           {arriveTangentWeight=0.2783, interpMode=RCIM_Cubic, tangentMode=RCTM_User, time=1, value=0}
          ]}}}
       CurveIndex=&Divide_Float(A=$Particles.NormalizedAge, B=0.3)
      )
     )
    )
    Particles.Mesh_02_DynamicParameter:Vector4f=&MakeVector4(
     X=&FloatFromCurve(
      FloatCurve=DataInterface() {propertyValues {Curve {keys [
          {arriveTangent=4.1785, interpMode=RCIM_Cubic, leaveTangent=4.1785, tangentMode=RCTM_User, time=0, value=0}
          {arriveTangent=1.1179, interpMode=RCIM_Cubic, leaveTangent=1.1179, tangentMode=RCTM_User, time=0.2042, value=0.6228}
          {arriveTangent=0.0224, interpMode=RCIM_Cubic, leaveTangent=0.0224, tangentMode=RCTM_User, time=1, value=1}
         ]}}}
      CurveIndex=&Divide_Float(A=$Particles.NormalizedAge, B=0.3)
     )
     Y=0
     Z=0
     W=1
    )
    Particles.Mesh_02_Size:Vector3f=&Multiply_Vector(
     A=$Particles.Initial.Mesh_02_Size
     B=&VectorFromFloat(Value=&FloatFromCurve(
       FloatCurve=DataInterface() {propertyValues {Curve {keys [
           {arriveTangent=4.1785, interpMode=RCIM_Cubic, leaveTangent=4.1785, tangentMode=RCTM_User, time=0, value=0}
           {arriveTangent=0.9995, interpMode=RCIM_Cubic, leaveTangent=0.9995, tangentMode=RCTM_User, time=0.2223, value=0.6825}
           {arriveTangent=-0.0405, interpMode=RCIM_Cubic, leaveTangent=-0.0405, tangentMode=RCTM_User, time=1, value=1}
          ]}}}
       CurveIndex=&Divide_Float(A=$Particles.NormalizedAge, B=0.3)
       Scale Curve=2.5
      ))
    )
   ) {isSetParameters=true}
   /SetVariables(
    Particles.Mesh_03_Color:LinearColor=&MultiplyLinearColor(
     Linear Color=$Particles.Initial.Mesh_03_Color
     Scale Factor=&MakeFloat4FromVectorAndFloat(
      "Vector (RGB)"=&MakeVectorFromLinearColorRGB(Color=$System.TODColor)
      "Float (Alpha)"=&FloatFromCurve(
       FloatCurve=DataInterface() {propertyValues {Curve {keys [
           {arriveTangent=6.2301, interpMode=RCIM_Cubic, leaveTangent=6.2301, tangentMode=RCTM_User, time=0, value=0.5}
           {arriveTangent=-0.7357, interpMode=RCIM_Cubic, leaveTangent=-0.7357, tangentMode=RCTM_User, time=0.1, value=2}
           <at 0.4 0.2>
           {arriveTangent=0.4048, interpMode=RCIM_Cubic, leaveTangent=0.4048, tangentMode=RCTM_User, time=0.7, value=0}
          ]}}}
       CurveIndex=&Divide_Float(A=$Particles.NormalizedAge, B=0.5)
       Scale Curve=0.25
      )
     )
    )
    Particles.Mesh_03_Size:Vector3f=&Multiply_Vector(
     A=$Particles.Initial.Mesh_03_Size
     B=&Multiply_VectorByFloat(
      Vector=&MakeVector(
       X=1
       Y=1
       Z=&FloatFromCurve(
        FloatCurve=DataInterface() {propertyValues {Curve {keys [<at 0 1>,<at 1 0>]}}}
        CurveIndex=&Divide_Float(A=$Particles.NormalizedAge, B=0.5)
       )
      )
      Float=&FloatFromCurve(
       FloatCurve=DataInterface() {propertyValues {Curve {keys [
           {arriveTangent=7.2197, interpMode=RCIM_Cubic, leaveTangent=7.2197, tangentMode=RCTM_User, time=0, value=0}
           {arriveTangent=1.9952, interpMode=RCIM_Cubic, leaveTangent=1.9952, time=0.4917, value=1.9432}
           {arriveTangent=-0.0405, interpMode=RCIM_Cubic, leaveTangent=-0.0405, tangentMode=RCTM_User, time=1.0024, value=2}
          ]}}}
       CurveIndex=&Divide_Float(A=$Particles.NormalizedAge, B=0.5)
      )
     )
    )
   ) {isSetParameters=true}
   /SetVariables(
    Particles.Mesh_04_Color:LinearColor=&MultiplyLinearColor(
     Linear Color=$Particles.Initial.Mesh_04_Color
     Scale Factor=&MakeFloat4FromVectorAndFloat(
      "Vector (RGB)"=&MakeVectorFromLinearColorRGB(Color=$System.TODColor)
      "Float (Alpha)"=&FloatFromCurve(
       FloatCurve=DataInterface() {propertyValues {Curve {keys [
           {arriveTangent=8.2356, interpMode=RCIM_Cubic, leaveTangent=8.2356, tangentMode=RCTM_User, time=0, value=0}
           {arriveTangent=0.1969, interpMode=RCIM_Cubic, leaveTangent=0.1969, tangentMode=RCTM_User, time=0.2223, value=1}
           {arriveTangent=-3.0486, interpMode=RCIM_Cubic, leaveTangent=-3.0486, tangentMode=RCTM_User, time=0.4781, value=0.509}
           {arriveTangent=-0.2378, interpMode=RCIM_Cubic, leaveTangent=-0.2378, tangentMode=RCTM_User, time=1, value=0}
          ]}}}
       CurveIndex=&Divide_Float(A=$Particles.NormalizedAge, B=&RandomRangeFloat(Minimum=0.5, Maximum=0.75))
       Scale Curve=0.5
      )
     )
    )
    Particles.Mesh_04_DynamicParameter:Vector4f=&MakeVector4(
     X=&FloatFromCurve(
      FloatCurve=DataInterface() {propertyValues {Curve {keys [{interpMode=RCIM_Cubic, tangentMode=RCTM_User, time=0.5, value=0}, {interpMode=RCIM_Cubic, tangentMode=RCTM_User, time=1, value=1}]}}}
      CurveIndex=&Divide_Float(A=$Particles.NormalizedAge, B=&RandomRangeFloat(Minimum=0.5, Maximum=0.75))
     )
     Y=&FNTimeOfDayScalar(Night Scale=0.05, Morning Scale=-0.26, Day Scale=0, Evening Scale=-0.35)
     Z=&FNTimeOfDayScalar(Night Scale=0.5, Morning Scale=0.23, Evening Scale=0.2)
     W=&FNTimeOfDayScalar(Night Scale=0.02, Morning Scale=-0.01, Day Scale=0, Evening Scale=-0.01)
    )
    Particles.Mesh_04_Orientation:Quat4f=&MultiplyQuaternion(
     Quaternion A=$Particles.Initial.Mesh_04_Orientation
     Quaternion B=&MakeQuatFromAxisAngle(
      Axis=&RandomRangeVector(Minimum=<0 0 -1>, Maximum=<0 0 1>)
      AngleInDegrees=&RandomRangeFloat(Minimum=-50, Maximum=50)
     )
    )
    Particles.Mesh_04_Size:Vector3f=&Multiply_Vector(
     A=$Particles.Initial.Mesh_04_Size
     B=&Multiply_VectorByFloat(
      Vector=&MakeVector(
       X=&FloatFromCurve(
        FloatCurve=DataInterface() {propertyValues {Curve {keys [
            {arriveTangent=0.8546, interpMode=RCIM_Cubic, leaveTangent=3.1486, tangentMode=RCTM_Break, time=0, value=0}
            {arriveTangent=1.4415, interpMode=RCIM_Cubic, leaveTangent=1.4415, tangentMode=RCTM_User, time=0.3242, value=0.7}
            {arriveTangent=1.2727, interpMode=RCIM_Cubic, leaveTangent=1.2727, tangentMode=RCTM_User, time=1, value=1.5}
           ]}}}
        CurveIndex=&Divide_Float(A=$Particles.NormalizedAge, B=&RandomRangeFloat(Minimum=0.5, Maximum=0.75))
       )
       Y=&FloatFromCurve(
        FloatCurve=DataInterface() {propertyValues {Curve {keys [
            {arriveTangent=0.8546, interpMode=RCIM_Cubic, leaveTangent=3.1486, tangentMode=RCTM_Break, time=0, value=0}
            {arriveTangent=1.4415, interpMode=RCIM_Cubic, leaveTangent=1.4415, tangentMode=RCTM_User, time=0.3242, value=0.7}
            {arriveTangent=1.2727, interpMode=RCIM_Cubic, leaveTangent=1.2727, tangentMode=RCTM_User, time=1, value=1.5}
           ]}}}
        CurveIndex=&Divide_Float(A=$Particles.NormalizedAge, B=&RandomRangeFloat(Minimum=0.5, Maximum=0.75))
       )
       Z=&FloatFromCurve(
        FloatCurve=DataInterface() {propertyValues {Curve {keys [
            {arriveTangent=1.8479, interpMode=RCIM_Cubic, leaveTangent=1.8479, tangentMode=RCTM_User, time=0, value=0}
            {arriveTangent=0.9995, interpMode=RCIM_Cubic, leaveTangent=0.9995, tangentMode=RCTM_User, time=0.4, value=0.4}
            {arriveTangent=0.4855, interpMode=RCIM_Cubic, leaveTangent=0.4855, tangentMode=RCTM_User, time=1, value=0.6}
           ]}}}
        CurveIndex=&Divide_Float(A=$Particles.NormalizedAge, B=&RandomRangeFloat(Minimum=0.5, Maximum=0.75))
       )
      )
      Float=&FloatFromCurve(
       FloatCurve=DataInterface() {propertyValues {Curve {keys [
           {interpMode=RCIM_Cubic, leaveTangent=6.5404, tangentMode=RCTM_Break, time=0, value=1.25}
           {arriveTangent=0.0345, interpMode=RCIM_Cubic, leaveTangent=0.0345, tangentMode=RCTM_User, time=0.0648, value=1.4045}
           {arriveTangent=0.0556, interpMode=RCIM_Cubic, leaveTangent=0.0556, tangentMode=RCTM_User, time=1, value=1.4}
          ]}}}
       CurveIndex=&Divide_Float(A=$Particles.NormalizedAge, B=&RandomRangeFloat(Minimum=0.5, Maximum=0.75))
      )
     )
    )
   ) {isSetParameters=true}
   /SetVariables(Particles.Mesh_Position:NiagaraPosition=&Add_Vector(
     A=Linked() {linkedVariable {name=Particles.Position, type=NiagaraPosition}}
     B=&RandomRangeVector(Minimum=<0 0 25>, Maximum=<0 0 50>)
    )) {isSetParameters=true}
  ]
  rendererClasses [&NiagaraMeshRendererProperties]
  renderers [
   {rendererClass=&NiagaraMeshRendererProperties, rendererIndex=0}
   {rendererClass=&NiagaraMeshRendererProperties, rendererIndex=1}
   {rendererClass=&NiagaraMeshRendererProperties, rendererIndex=2}
   {rendererClass=&NiagaraMeshRendererProperties, rendererIndex=3}
  ]
 }
}
summary {
 systemName=NS_CubeSurfing_PuffLand
 emitters [
  {emitterName=Block_Burst, rendererClasses [&NiagaraMeshRendererProperties]}
  {emitterName=ImpactRing, rendererClasses [&NiagaraSpriteRendererProperties]}
  {emitterName=Dust, rendererClasses [&NiagaraSpriteRendererProperties]}
  {emitterName=AirStreaks, rendererClasses [&NiagaraSpriteRendererProperties]}
  {emitterName=Wind_Radial, rendererClasses [&NiagaraMeshRendererProperties]}
 ]
 userVariables [
  User.CubeColorMax {type=LinearColor, description="", defaultValue {struct=LinearColor, value=<rgba 0 0.6514 0.7379 1>}}
  User.CubeColorMin {type=LinearColor, description="", defaultValue {struct=LinearColor, value=<rgba 0.0262 0.2016 0.5583 1>}}
  User.FX_Scale_Global {type=NiagaraFloat, defaultValue=0, description=""}
 ]
}
