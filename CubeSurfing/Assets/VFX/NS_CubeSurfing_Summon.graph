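# Niagara summon effect for Cube Surfing: a cyan-blue energy build-up of converging blocks that culminates in a flash.
# Blocks are repositioned each update on a shrinking, rotating shape and lerped toward the origin, while Airflow meshes, Corona and glow sprites build.
# A delayed Core burst adds a light and three layered sprites animated through custom per-renderer attributes.
# User colors tint blocks, airflow and corona; time-of-day scalars adjust brightness or opacity.
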NiagaraSystem /CubeSurfing/Assets/VFX/NS_CubeSurfing_Summon
dependencies {
 usedDataInterfaces [
  &NiagaraDataInterfaceCurve
  &NiagaraDataInterfaceVectorCurve
  &NiagaraDataInterfaceColorCurve
  &NiagaraDataInterfacePlatformSet
 ]
 usedDynamicInputs [&SimulationPosition, &ConvertVectorToPosition]
 usedModules [&SystemState]
}
emitters {
 Airflow {
  emitterUpdateScript [
   &EmitterState(Life Cycle Mode=::NewEnumerator1, Loop Behavior=::NewEnumerator1, Loop Duration=1, UseLoopDelay=false)
   &SpawnRate(SpawnRate=15, Use Spawn Probability=false)
  ]
  particleSpawnScript [
   &InitializeParticle(
    Color Maximum=$User.CubeColorMax
    Color Minimum=$User.CubeColorMin
    Color Mode=::NewEnumerator2
    Lifetime=&RandomRangeFloat(Minimum=0.5, Maximum=0.5)
    Mass Mode=::NewEnumerator1
    Mesh Scale=&Multiply_VectorByFloat(Vector=&RandomRangeVector(Minimum=<0.7 0.7 0.7>, Maximum=<1 1 1>), Float=0.9)
    Mesh Scale Mode=::NewEnumerator4
    Position Offset Coordinate Space=::Local
    Sprite Size=<5 5>
    Sprite Size Mode=::NewEnumerator4
    UsePositionOffset=true
    Write Material Random=true
   ) {module=InitializeParticle001}
   &InitialMeshOrientation(
    Mesh Axis To Orient=&RandomRangeVector(Minimum=<-1 0 0>, Maximum=<1 0 0>)
    Mesh Orientation Mode=::NewEnumerator2
    Orientation Vector=&RandomRangeVector(Minimum=<-1 0 0>, Maximum=<1 0 0>)
    Rotation=&RandomRangeVector(Minimum=<-0.1 0 -1>, Maximum=<0.1 0 1>)
    Use Rotation Vector=true
   )
  ]
  particleUpdateScript [
   &ParticleState()
   &SolveForcesAndVelocity(Clamp Velocity=false, Limit Acceleration=false)
   &DynamicMaterialParameters(
    Index 0 Param 1=&FloatFromCurve(FloatCurve=DataInterface() {propertyValues {Curve {keys [<at 0 -0.2>,<at 1 1.2>]}}})
    Index 0 Param 2=1
    Index 0 Param 3=&Multiply_Float(
     A=&RandomRangeFloat(Minimum=0.8, Maximum=1)
     B=&FloatFromCurve(FloatCurve=DataInterface() {propertyValues {Curve {keys [<at 0 1>,<at 1 0.5>]}}}, Scale Curve=5)
    )
    Module.Param0WriteEnabled=true
    Module.Param1WriteEnabled=true
    Module.Param2WriteEnabled=true
    Module.Param3WriteEnabled=true
   )
   &ScaleColor(
    Scale Alpha=&Multiply_Float(
     A=&Multiply_Float(
      A=&Multiply_Float(A=1, B=&FNTimeOfDayScalar(Night Scale=1, Morning Scale=0.3, Evening Scale=0.4))
      B=&FloatFromCurve(FloatCurve=DataInterface() {propertyValues {Curve {keys [
           <at 0 0 RCIM_Cubic>
           <at 0.1912 0.9932>
           {arriveTangent=-1.2279, interpMode=RCIM_Cubic, leaveTangent=-1.2279, time=0.7879, value=1.0115}
           <at 1 0>
          ]}}})
     )
     B=&FloatFromCurve(
      FloatCurve=DataInterface() {propertyValues {Curve {keys [
          {interpMode=RCIM_Cubic, tangentMode=RCTM_User, time=0, value=0}
          {arriveTangent=0.37, interpMode=RCIM_Cubic, leaveTangent=0.37, tangentMode=RCTM_User, time=0.6, value=0.05}
          {arriveTangent=1.1869, interpMode=RCIM_Cubic, leaveTangent=1.1869, tangentMode=RCTM_User, time=0.9, value=0.5258}
          <at 1.2 0.8>
         ]}}}
      CurveIndex=$Emitter.Age
     )
    )
    ScaleA=true
    ScaleRGB=true
   ) {enabled=false}
   &ScaleColor(
    Scale Alpha=&Multiply_Float(
     A=&Multiply_Float(
      A=&Multiply_Float(A=1, B=&FNTimeOfDayScalar(Night Scale=1, Morning Scale=0.3, Evening Scale=0.4))
      B=&FloatFromCurve(
       FloatCurve=DataInterface() {propertyValues {Curve {keys [
           <at 0 0 RCIM_Cubic>
           <at 0.1912 0.9932>
           {arriveTangent=-1.2279, interpMode=RCIM_Cubic, leaveTangent=-1.2279, time=0.7879, value=1.0115}
           <at 1 0>
          ]}}}
       CurveIndex=$System.NormalizedLoopAge
      )
     )
     B=0.5
    )
    ScaleA=true
    ScaleRGB=true
   ) {module=ScaleColor002}
   &ScaleColor(
    Scale Alpha=&FloatFromCurve(
     FloatCurve=DataInterface() {propertyValues {Curve {
        defaultValue=1
        keys [
         {arriveTangent=3.3375, interpMode=RCIM_Cubic, leaveTangent=3.3375, tangentMode=RCTM_User, time=0, value=1}
         {arriveTangent=-2.5249, interpMode=RCIM_Cubic, leaveTangent=-2.5249, tangentMode=RCTM_User, time=0.8, value=1.275}
         {arriveTangent=-0.0405, interpMode=RCIM_Cubic, leaveTangent=-0.0405, tangentMode=RCTM_User, time=1, value=0}
        ]
       }}}
     CurveIndex=$Emitter.Age
     Scale Curve=&FN_PickFloatDependOnPlatform(TargetPlatformActive=0.75, TargetPlatformNoActive=0.5)
    )
    Scale RGB=&Multiply_Vector(
     A=&MakeVectorFromLinearColorRGB(Color=&ColorFromCurve(ColorCurve=DataInterface() {propertyValues {
         AlphaCurve {keys [<at 0 1 RCIM_Cubic>,<at 0.9977 1 RCIM_Cubic>]}
         BlueCurve {keys [
           <at 0.0057 4.7917>
           {arriveTangent=-8.8394, interpMode=RCIM_Cubic, leaveTangent=-8.8394, time=0.1454, value=2}
           <at 0.4346 1>
           <at 0.7175 4.592>
           {arriveTangent=-7.943, interpMode=RCIM_Cubic, leaveTangent=-7.943, time=0.8517, value=5}
           <at 0.9448 2.7865>
          ]}
         GreenCurve {keys [
           <at 0.0057 4.7917>
           {arriveTangent=-8.9149, interpMode=RCIM_Cubic, leaveTangent=-8.9149, time=0.1454, value=1.8197}
           <at 0.4346 0.9676>
           <at 0.7175 5>
           <at 0.8517 5 RCIM_Cubic>
           <at 0.9448 5>
          ]}
         RedCurve {keys [
           <at 0.0057 5>
           {arriveTangent=-10.4664, interpMode=RCIM_Cubic, leaveTangent=-10.4664, time=0.1454, value=0.9183}
           <at 0.4346 0.5104>
           <at 0.7175 2.5521>
           {arriveTangent=1.2995, interpMode=RCIM_Cubic, leaveTangent=1.2995, time=0.8517, value=1.7188}
           <at 0.9448 2.8475>
          ]}
        }}))
     B=<1 1 1>
    )
    ScaleA=true
    ScaleRGB=true
   ) {module=ScaleColor001}
   &ScaleMeshSize(Scale Factor=&Multiply_Vector(
     A=<1 1 1>
     B=&VectorFromFloat(Value=&FloatFromCurve(
       FloatCurve=DataInterface() {propertyValues {Curve {keys [
           {arriveTangent=0.2635, interpMode=RCIM_Cubic, leaveTangent=0.2635, tangentMode=RCTM_User, time=0, value=0.2}
           {arriveTangent=0.4932, interpMode=RCIM_Cubic, leaveTangent=0.4932, tangentMode=RCTM_User, time=0.5, value=0.5}
           {arriveTangent=0.447, interpMode=RCIM_Cubic, leaveTangent=0.447, tangentMode=RCTM_User, time=1, value=1.2}
          ]}}}
       CurveIndex=$Emitter.Age
      ))
    ))
   &CameraOffset()
  ]
  rendererClasses [&NiagaraMeshRendererProperties]
  renderers [{rendererClass=&NiagaraMeshRendererProperties, rendererIndex=0}]
 }
 Blocks {
  emitterUpdateScript [
   &EmitterState(Life Cycle Mode=::NewEnumerator1, Loop Behavior=::NewEnumerator1, Loop Delay=0.1, Loop Duration=1.5, UseLoopDelay=true)
   &SpawnBurst_Instantaneous(Spawn Count=5, Use Spawn Probability=false, UseLoopCountLimit=false)
   &SpawnRate(
    SpawnRate=&FloatFromCurve(
     FloatCurve=DataInterface() {propertyValues {Curve {
        defaultValue=1
        keys [
         {arriveTangent=-3.0102, interpMode=RCIM_Cubic, leaveTangent=-3.0102, tangentMode=RCTM_User, time=0, value=1}
         {interpMode=RCIM_Cubic, tangentMode=RCTM_User, time=0.75, value=0}
        ]
       }}}
     CurveIndex=$Emitter.NormalizedLoopAge
     Scale Curve=30
    )
    Use Spawn Probability=false
   )
  ]
  particleSpawnScript [&InitializeParticle(
    Color Maximum=$User.CubeColorMax
    Color Minimum=$User.CubeColorMin
    Color Mode=::NewEnumerator2
    Lifetime Max=0.6
    Lifetime Min=0.3
    Lifetime Mode=::NewEnumerator1
    Mesh Scale Mode=::NewEnumerator3
    Position Mode=::NewEnumerator0
    Sprite Size Mode=::NewEnumerator1
    Uniform Sprite Size Max=20
    Uniform Sprite Size Min=12
    Write Material Random=true
   )]
  particleUpdateScript [
   &ParticleState()
   &ShapeLocation(
    MaskSpawnGroup=false
    Non Uniform Scale=&RandomRangeVector(Minimum=<0.25 0.25 1>, Maximum=<1 1 1>)
    Offset=&MakeVector(
     X=0
     Y=0
     Z=&FloatFromCurve(FloatCurve=DataInterface() {propertyValues {Curve {keys [<at 0 1>,<at 1 0>]}}}, Scale Curve=250)
    )
    Offset Mode=::NewEnumerator0
    Override Seed=false
    Rotation Mode=::NewEnumerator2
    Sphere Radius=&FloatFromCurve(
     FloatCurve=DataInterface() {propertyValues {Curve {
        defaultValue=0
        keys [
         {arriveTangent=2.1078, interpMode=RCIM_Cubic, leaveTangent=2.1078, tangentMode=RCTM_User, time=0, value=0.5}
         {interpMode=RCIM_Cubic, tangentMode=RCTM_User, time=0.5, value=1}
         {tangentMode=RCTM_User, time=1, value=0}
        ]
       }}}
     Scale Curve=&RandomRangeFloat(Minimum=80, Maximum=120)
    )
    Sphere Surface Distribution=0.75
    UseOwnerScale=true
    Yaw / Pitch / Roll=&VectorFromCurve(
     VectorCurve=DataInterface() {propertyValues {
       XCurve {
        defaultValue=0
        keys [{interpMode=RCIM_Cubic, tangentMode=RCTM_User, time=0, value=1}, {interpMode=RCIM_Cubic, tangentMode=RCTM_User, time=1, value=0}]
       }
       YCurve {
        defaultValue=0
        keys [
         {arriveTangent=-1.3041, interpMode=RCIM_Cubic, leaveTangent=-1.3041, tangentMode=RCTM_User, time=0, value=0.8}
         {interpMode=RCIM_Cubic, tangentMode=RCTM_User, time=1, value=0.25}
        ]
       }
       ZCurve {
        defaultValue=1
        keys [{interpMode=RCIM_Cubic, tangentMode=RCTM_User, time=0, value=0}, {interpMode=RCIM_Cubic, tangentMode=RCTM_User, time=1, value=1}]
       }
      }}
     Scale Curve=&RandomRangeVector(Minimum=<0.25 0.25 0.25>, Maximum=<1 1 1>)
    )
   )
   /SetVariables(Particles.Position=&LerpPosition(A=$Particles.Position, B=<0 0 0>, Alpha=$Particles.NormalizedAge)) {isSetParameters=true}
   &ScaleMeshSize(Scale Factor=&Lerp_Vector(
     A=&RandomRangeVector(Minimum=<0.1 0.1 0.1>, Maximum=<0.25 0.25 0.25>)
     B=<1 1 1>
     Alpha=&FloatFromCurve(FloatCurve=DataInterface() {propertyValues {Curve {
         defaultValue=0
         keys [
          {arriveTangent=4.2267, interpMode=RCIM_Cubic, leaveTangent=4.2267, tangentMode=RCTM_User, time=0, value=0}
          {interpMode=RCIM_Cubic, tangentMode=RCTM_User, time=0.5, value=1}
          {tangentMode=RCTM_User, time=1, value=0}
         ]
        }}})
    ))
   &ScaleSpriteSize(
    Scale Sprite Size Mode=::NewEnumerator0
    Uniform Scale Factor=&FloatFromCurve(FloatCurve=DataInterface() {propertyValues {Curve {keys [
         {arriveTangent=9.1187, interpMode=RCIM_Cubic, time=0, value=0}
         {arriveTangent=1.1429, interpMode=RCIM_Cubic, leaveTangent=1.1429, time=0.2487, value=1.1246}
         {arriveTangent=-1.497, interpMode=RCIM_Cubic, leaveTangent=-1.497, time=0.7, value=0.8}
         {arriveTangent=-3.0554, interpMode=RCIM_Cubic, leaveTangent=-3.0554, tangentMode=RCTM_User, time=1, value=0}
        ]}}})
   )
   &DynamicMaterialParameters(
    Index 0 Param 1=&FloatFromCurve(
     FloatCurve=DataInterface() {propertyValues {Curve {
        defaultValue=0
        keys [
         {arriveTangent=9.0725, interpMode=RCIM_Cubic, leaveTangent=9.0725, tangentMode=RCTM_User, time=0, value=0}
         {arriveTangent=-2.5249, interpMode=RCIM_Cubic, leaveTangent=-2.5249, tangentMode=RCTM_User, time=0.1408, value=0.9847}
         {arriveTangent=-0.0405, interpMode=RCIM_Cubic, leaveTangent=-0.0405, tangentMode=RCTM_User, time=1, value=0}
        ]
       }}}
     Scale Curve=&RandomRangeFloat(Minimum=0, Maximum=2)
    )
    Index 0 Param 2=1
    Index 0 Param 3=0.33
    Module.Index0Float0Write=true
    Module.Index0Float1Write=true
    Module.Index0Float2Write=true
    Module.Index0Float3Write=true
   )
   &Color(Color=&RandomRangeLinearColor(Minimum=$User.CubeColorMin, Maximum=$User.CubeColorMax))
   /SetVariables(Particles.Color_Glow:LinearColor=&MultiplyLinearColor(
     Linear Color=<rgba 0 0.7437 1 0.1>
     Scale Factor=&MakeVector4_FromLinearColor(LinearColor=$Particles.Color)
    )) {isSetParameters=true}
  ]
  rendererClasses [&NiagaraMeshRendererProperties, &NiagaraSpriteRendererProperties]
  renderers [
   {rendererClass=&NiagaraMeshRendererProperties, rendererIndex=0}
   {rendererClass=&NiagaraSpriteRendererProperties, rendererIndex=1}
  ]
 }
 Core {
  emitterUpdateScript [
   &EmitterState(Life Cycle Mode=::NewEnumerator1, Loop Behavior=::NewEnumerator1, Loop Delay=0.95, Loop Duration=1, UseLoopDelay=true)
   &SpawnBurst_Instantaneous(Spawn Count=1, Use Spawn Probability=false)
  ]
  particleSpawnScript [
   &InitializeParticle(Lifetime=0.7, Position Mode=::NewEnumerator0, Write Material Random=true)
   /SetVariables(
    Particles.Sprite_Color_01:LinearColor=<rgba 0 1.4263 2 1>
    Particles.Sprite_DP_01:Vector4f=&MakeVector4(X=&RandomRangeFloat(Minimum=0.4, Maximum=0.6), Y=1, Z=1, W=1)
    Particles.Sprite_Size_01:Vector2f=&Vector2DFromFloat(Value=200)
   ) {isSetParameters=true}
   /SetVariables(
    Particles.Sprite_Color_02:LinearColor=<rgba 0 0.0654 0.276 0.3>
    Particles.Sprite_DP_02:Vector4f=&MakeVector4(X=&RandomRangeFloat(Minimum=0.4, Maximum=0.6), Y=1, Z=1, W=1)
    Particles.Sprite_Size_02:Vector2f=&Vector2DFromFloat(Value=450)
   ) {isSetParameters=true}
   /SetVariables(
    Particles.Sprite_Color_03:LinearColor=<rgba 0 0.5167 1 3.5>
    Particles.Sprite_Size_03:Vector2f=&Vector2DFromFloat(Value=200)
   ) {isSetParameters=true}
  ]
  particleUpdateScript [
   &ParticleState()
   &CameraOffset(Camera Offset Amount=50)
   /SetVariables(
    Particles.Sprite_Color_01:LinearColor=&MultiplyLinearColor(
     Linear Color=$Particles.Initial.Sprite_Color_01
     Scale Factor=&MakeFloat4FromVectorAndFloat(
      "Vector (RGB)"=<1 1 1>
      "Float (Alpha)"=&Multiply_Float(
       A=&FNTimeOfDayScalar(Night Scale=0.5, Morning Scale=0.3, Evening Scale=0.4)
       B=&FloatFromCurve(
        FloatCurve=DataInterface() {propertyValues {Curve {keys [<at 0 0>,<at 0.2 1>,<at 0.8 1>,<at 1 0>]}}}
        CurveIndex=&Multiply_Float(A=$Particles.NormalizedAge, B=1.75)
       )
      )
     )
    )
    Particles.Sprite_Size_01:Vector2f=&Multiply_Vector2D(
     A=$Particles.Initial.Sprite_Size_01
     B=&Vector2DFromFloat(Value=&FloatFromCurve(
       FloatCurve=DataInterface() {propertyValues {Curve {keys [<at 0 1.5 RCIM_Cubic>,<at 1 0>]}}}
       CurveIndex=&Multiply_Float(A=$Particles.NormalizedAge, B=1.75)
      ))
    )
   ) {isSetParameters=true}
   /SetVariables(
    Particles.Sprite_Color_02:LinearColor=&MultiplyLinearColor(
     Linear Color=$Particles.Initial.Sprite_Color_02
     Scale Factor=&MakeFloat4FromVectorAndFloat(
      "Vector (RGB)"=&Multiply_VectorByFloat(Vector=<1 1 1>, Float=&FNTimeOfDayScalar(Night Scale=0.5, Morning Scale=0.3, Evening Scale=0.4))
      "Float (Alpha)"=&FloatFromCurve(
       FloatCurve=DataInterface() {propertyValues {Curve {keys [<at 0 0>,<at 0.2 1>,<at 0.8 1>,<at 1 0>]}}}
       CurveIndex=&Multiply_Float(A=$Particles.NormalizedAge, B=1.75)
      )
     )
    )
    Particles.Sprite_Size_02:Vector2f=&Multiply_Vector2D(
     A=$Particles.Initial.Sprite_Size_02
     B=&Vector2DFromFloat(Value=&FloatFromCurve(
       FloatCurve=DataInterface() {propertyValues {Curve {keys [<at 0 1.5 RCIM_Cubic>,<at 1 0>]}}}
       CurveIndex=&Multiply_Float(A=$Particles.NormalizedAge, B=1.75)
      ))
    )
   ) {isSetParameters=true}
   /SetVariables(
    Particles.Sprite_Color_03:LinearColor=&MultiplyLinearColor(
     Linear Color=$Particles.Initial.Sprite_Color_03
     Scale Factor=&MakeFloat4FromVectorAndFloat(
      "Vector (RGB)"=<1 1 1>
      "Float (Alpha)"=&Multiply_Float(
       A=&FloatFromCurve(
        FloatCurve=DataInterface() {propertyValues {Curve {keys [{interpMode=RCIM_Cubic, leaveTangent=-0.6434, tangentMode=RCTM_Break, time=0, value=1}, <at 1 0>]}}}
        CurveIndex=&Multiply_Float(A=$Particles.NormalizedAge, B=2)
       )
       B=&FNTimeOfDayScalar(Night Scale=1, Morning Scale=0.3, Evening Scale=0.5)
      )
     )
    )
    Particles.Sprite_Size_03:Vector2f=&Multiply_Vector2D(
     A=$Particles.Initial.Sprite_Size_03
     B=&Vector2DFromFloat(Value=&FloatFromCurve(
       FloatCurve=DataInterface() {propertyValues {Curve {keys [{tangentMode=RCTM_User, time=0, value=0}, <at 1 1>]}}}
       CurveIndex=&Multiply_Float(A=$Particles.NormalizedAge, B=2)
       Scale Curve=0.75
      ))
    )
   ) {isSetParameters=true}
   &FN_Light_Attributes(
    Color=&MultiplyLinearColor(
     Linear Color=<rgba 338.541 1814.6604 5000 1>
     Scale Factor=&MakeVector4_FromLinearColor(LinearColor=&ColorFromCurve(ColorCurve=DataInterface() {propertyValues {
         AlphaCurve {keys [<at 0 0>,<at 0.2565 1>,<at 1 0>]}
         BlueCurve {keys [<at 0 1>,<at 0.2541 5>,<at 0.303 1>,<at 1 1>]}
         GreenCurve {keys [<at 0 1>,<at 0.2541 5>,<at 0.303 1>,<at 1 1>]}
         RedCurve {keys [<at 0 1>,<at 0.2541 5>,<at 0.303 1>,<at 1 1>]}
        }}))
    )
    Create Color Attribute=true
   )
  ]
  rendererClasses [&NiagaraLightRendererProperties, &NiagaraSpriteRendererProperties]
  renderers [
   {rendererClass=&NiagaraLightRendererProperties, rendererIndex=0}
   {rendererClass=&NiagaraSpriteRendererProperties, rendererIndex=1}
   {rendererClass=&NiagaraSpriteRendererProperties, rendererIndex=2}
   {rendererClass=&NiagaraSpriteRendererProperties, rendererIndex=3}
  ]
 }
 Corona {
  emitterUpdateScript [
   &EmitterState(Inactive Response=::NewEnumerator1, Life Cycle Mode=::NewEnumerator1, Loop Behavior=::NewEnumerator1, Loop Duration=0.9, UseLoopDelay=false)
   &SpawnRate(SpawnRate=5, Use Spawn Probability=false)
   &SpawnBurst_Instantaneous(Use Spawn Probability=false, UseLoopCountLimit=false) {enabled=false}
  ]
  particleSpawnScript [
   &InitializeParticle(
    Color=$User.CubeColorMax
    Lifetime=0.3
    Position Offset Coordinate Space=::Local
    Sprite Rotation Mode=::NewEnumerator2
    Sprite Size Mode=::NewEnumerator3
    Uniform Sprite Size=&Multiply_Float(
     A=100
     B=/Game/Effects/Niagara/DynamicInputs/Math/FN_Fit(
      Clamp:NiagaraBool=true
      Value:NiagaraFloat=$System.NormalizedLoopAge
      SourceMin:NiagaraFloat=0
      SourceMax:NiagaraFloat=1
      DestMin:NiagaraFloat=0.3
      DestMax:NiagaraFloat=1
     )
    )
    UsePositionOffset=true
    Write Material Random=true
   )
   &CameraOffset(Camera Offset Amount=&Multiply_Float(A=50, B=1))
  ]
  particleUpdateScript [
   &ParticleState()
   &ScaleSpriteSize(
    Scale Sprite Size Mode=::NewEnumerator0
    Uniform Scale Factor=&FloatFromCurve(
     FloatCurve=DataInterface() {propertyValues {Curve {keys [
         {arriveTangent=0.4295, interpMode=RCIM_Cubic, leaveTangent=0.4295, tangentMode=RCTM_User, time=0, value=2.5}
         {arriveTangent=-2.2222, interpMode=RCIM_Cubic, leaveTangent=-2.2222, time=0.5336, value=1.0603}
         <at 0.9 0.5 RCIM_Cubic>
        ]}}}
     Scale Curve=1.5
    )
   )
   &ScaleColor(
    Scale Alpha=&Multiply_Float(
     A=&MakeCustomFloatFromBool(Bool=$User.IsMaskMakerSpireShards, True Float=0.05, False Float=1)
     B=&FloatFromCurve(
      FloatCurve=DataInterface() {propertyValues {Curve {keys [
          {interpMode=RCIM_Cubic, tangentMode=RCTM_User, time=0, value=0}
          {interpMode=RCIM_Cubic, tangentMode=RCTM_User, time=0.1, value=1}
          {tangentMode=RCTM_User, time=1, value=0}
         ]}}}
      Scale Curve=&Multiply_Float(A=0.5, B=$System.NormalizedLoopAge)
     )
    )
    Scale RGB=&MakeVectorFromLinearColorRGB(Color=&ColorFromCurve(ColorCurve=DataInterface() {propertyValues {
        AlphaCurve {keys []}
        BlueCurve {keys [<at 0 1 RCIM_Cubic>,<at 0.9887 5 RCIM_Cubic>]}
        GreenCurve {keys [<at 0 0.968 RCIM_Cubic>,<at 0.9887 3.3507 RCIM_Cubic>]}
        RedCurve {keys [<at 0 0.904 RCIM_Cubic>,<at 0.9887 2.526 RCIM_Cubic>]}
       }}))
    ScaleA=true
    ScaleRGB=true
   )
   &DynamicMaterialParameters(
    Index 0 Param 1=&FloatFromCurve(FloatCurve=DataInterface() {propertyValues {Curve {keys [<at 0 1 RCIM_Cubic>,<at 1 1>]}}})
    Index 0 Param 2=1
    Index 0 Param 3=&FloatFromCurve(FloatCurve=DataInterface() {propertyValues {Curve {keys [<at 0 0.3>,<at 2 1>]}}})
    Index 0 Param 4=1
    Module.Index0Param0Write=true
    Module.Index0Param1Write=true
    Module.Index0Param2Write=true
    Module.Index0Param3Write=true
   )
  ]
  rendererClasses [&NiagaraSpriteRendererProperties]
  renderers [{rendererClass=&NiagaraSpriteRendererProperties}]
 }
 DarkGlow {
  emitterUpdateScript [
   &EmitterState(
    Inactive Response=::NewEnumerator1
    Life Cycle Mode=::NewEnumerator1
    Loop Behavior=::NewEnumerator1
    Loop Delay=0.45
    Loop Duration=1
    UseLoopDelay=true
   )
   &SpawnBurst_Instantaneous(Use Spawn Probability=false, UseLoopCountLimit=false)
  ]
  particleSpawnScript [
   &InitializeParticle(
    Color=<rgba 0 0.3787 0.5267 1>
    Position Mode=::NewEnumerator1
    Sprite Size Mode=::NewEnumerator3
    Uniform Sprite Size=75
    UsePositionOffset=false
    Write Material Random=true
   )
   &CameraOffset(Camera Offset Amount=20)
  ]
  particleUpdateScript [
   &ParticleState()
   &ScaleSpriteSize(Uniform Curve Sprite Scale=DataInterface() {propertyValues {Curve {
       defaultValue=0
       keys [
        {arriveTangent=1.6255, interpMode=RCIM_Cubic, leaveTangent=1.6255, tangentMode=RCTM_User, time=0, value=0}
        {interpMode=RCIM_Cubic, tangentMode=RCTM_User, time=1, value=1}
       ]
      }}})
   &ScaleColor(
    Scale Alpha=&FloatFromCurve(
     FloatCurve=DataInterface() {propertyValues {Curve {keys [
         {arriveTangent=4.3523, interpMode=RCIM_Cubic, leaveTangent=4.3523, tangentMode=RCTM_User, time=0, value=0}
         {arriveTangent=0.6281, interpMode=RCIM_Cubic, leaveTangent=0.6281, tangentMode=RCTM_User, time=0.8, value=0.8}
         {arriveTangent=2.2868, interpMode=RCIM_Cubic, leaveTangent=2.2868, tangentMode=RCTM_User, time=1, value=0}
        ]}}}
     Scale Curve=&Multiply_Float(A=&FNTimeOfDayScalar(Night Scale=0.7, Morning Scale=1, Evening Scale=1), B=0.5)
    )
    ScaleA=true
    ScaleRGB=true
   )
   &DynamicMaterialParameters(
    Index 0 Param 1=2
    Index 0 Param 2=1
    Module.Index0Float0Write=true
    Module.Index0Float1Write=true
    Module.Index0Float2Write=true
    Module.Index0Float3Write=true
   )
  ]
  rendererClasses [&NiagaraSpriteRendererProperties]
  renderers [{rendererClass=&NiagaraSpriteRendererProperties}]
 }
 Glow {
  emitterUpdateScript [
   &EmitterState(Life Cycle Mode=::NewEnumerator1, Loop Behavior=::NewEnumerator1, Loop Delay=0.2, Loop Duration=0.5, UseLoopDelay=true)
   &SpawnBurst_Instantaneous(Spawn Count=1, Use Spawn Probability=false)
  ]
  particleSpawnScript [&InitializeParticle(
    Color=<rgba 0 0.4667 0.7 0.25>
    Lifetime=0.75
    Position Offset Coordinate Space=::Local
    Sprite Size Mode=::NewEnumerator3
    Uniform Sprite Size=5
    UsePositionOffset=true
    Write Material Random=true
   )]
  particleUpdateScript [
   &ParticleState(
    Kill Particles When Lifetime Has Elapsed=false
    Let Infinitely Lived Particles Die When Emitter Deactivates=true
    Loop Particles Lifetime=true
   )
   &ScaleSpriteSize(Scale Factor=&Vector2DFromFloat(Value=&FloatFromCurve(
      FloatCurve=DataInterface() {propertyValues {Curve {keys [
          {arriveTangent=0.4346, interpMode=RCIM_Cubic, leaveTangent=0.4346, tangentMode=RCTM_User, time=0, value=0.7}
          {arriveTangent=0.0473, interpMode=RCIM_Cubic, leaveTangent=0.0473, tangentMode=RCTM_User, time=0.2, value=1}
          {arriveTangent=-0.0078, interpMode=RCIM_Cubic, leaveTangent=-0.0078, tangentMode=RCTM_User, time=0.5, value=1}
          <at 1 0.7 RCIM_Cubic>
         ]}}}
      Scale Curve=&Multiply_Float(A=10, B=$Emitter.Age)
     )))
   &ScaleColor(
    Scale Alpha=&Multiply_Float(
     A=&FloatFromCurve(
      FloatCurve=DataInterface() {propertyValues {Curve {keys [
          {arriveTangent=0.4005, interpMode=RCIM_Cubic, leaveTangent=0.4005, tangentMode=RCTM_User, time=0, value=0.25}
          {arriveTangent=0.0256, interpMode=RCIM_Cubic, leaveTangent=0.0256, tangentMode=RCTM_User, time=0.2, value=1}
          {arriveTangent=0.01, interpMode=RCIM_Cubic, leaveTangent=0.01, tangentMode=RCTM_User, time=0.7, value=1}
          <at 1 0.25 RCIM_Cubic>
         ]}}}
      Scale Curve=&FNTimeOfDayScalar(Night Scale=1, Morning Scale=1, Evening Scale=1)
     )
     B=1
    )
    ScaleA=true
    ScaleRGB=true
   )
   /SetVariables(
    Particles.GlowColor2:LinearColor=&MakeLinearColorFromVectorAndFloat(
     "Vector (RGB)"=&MakeVectorFromLinearColorRGB(Color=<rgba 0 2.5 5 1>)
     "Float (Alpha)"=&Multiply_Float(
      A=&FloatFromCurve(
       FloatCurve=DataInterface() {propertyValues {Curve {keys [
           {interpMode=RCIM_Cubic, leaveTangent=0.3585, tangentMode=RCTM_Break, time=0, value=0}
           {arriveTangent=2.2344, tangentMode=RCTM_User, time=1, value=1}
          ]}}}
       CurveIndex=$Emitter.Age
       Scale Curve=0.2
      )
      B=&FloatFromCurve(
       FloatCurve=DataInterface() {propertyValues {Curve {keys [
           {arriveTangent=0.4005, interpMode=RCIM_Cubic, leaveTangent=0.4005, tangentMode=RCTM_User, time=0, value=0}
           {arriveTangent=0.0256, interpMode=RCIM_Cubic, leaveTangent=0.0256, tangentMode=RCTM_User, time=0.2, value=1}
           {arriveTangent=0.01, interpMode=RCIM_Cubic, leaveTangent=0.01, tangentMode=RCTM_User, time=0.7, value=1}
           <at 1 0 RCIM_Cubic>
          ]}}}
       Scale Curve=0.5
      )
     )
    )
    Particles.GlowSize2:Vector2f=&Vector2DFromFloat(Value=&Multiply_Float(
      A=150
      B=&FloatFromCurve(
       FloatCurve=DataInterface() {propertyValues {Curve {keys [
           {arriveTangent=0.4346, interpMode=RCIM_Cubic, leaveTangent=0.4346, tangentMode=RCTM_User, time=0, value=0}
           {arriveTangent=0.0473, interpMode=RCIM_Cubic, leaveTangent=0.0473, tangentMode=RCTM_User, time=0.2, value=1}
           {arriveTangent=-0.0078, interpMode=RCIM_Cubic, leaveTangent=-0.0078, tangentMode=RCTM_User, time=0.5, value=1}
           <at 1 0 RCIM_Cubic>
          ]}}}
       Scale Curve=&RandomRangeFloat(Minimum=0.25, Maximum=0.5, Evaluation Type=::NewEnumerator1)
      )
     ))
   ) {isSetParameters=true}
   &CameraOffset(Camera Offset Amount=25)
  ]
  rendererClasses [&NiagaraSpriteRendererProperties]
  renderers [
   {rendererClass=&NiagaraSpriteRendererProperties}
   {rendererClass=&NiagaraSpriteRendererProperties, rendererIndex=1}
  ]
 }
}
summary {
 systemName=NS_CubeSurfing_Summon
 emitters [
  {emitterName=Blocks, rendererClasses [&NiagaraMeshRendererProperties, &NiagaraSpriteRendererProperties]}
  {emitterName=Core, rendererClasses [&NiagaraLightRendererProperties, &NiagaraSpriteRendererProperties]}
  {emitterName=Airflow, rendererClasses [&NiagaraMeshRendererProperties]}
  {emitterName=Corona, rendererClasses [&NiagaraSpriteRendererProperties]}
  {emitterName=DarkGlow, rendererClasses [&NiagaraSpriteRendererProperties]}
  {emitterName=Glow, rendererClasses [&NiagaraSpriteRendererProperties]}
 ]
 userVariables [
  User.bls_Medallion_LightWeight {type=NiagaraBool}
  User.CubeColorMax {type=LinearColor, description="", defaultValue {struct=LinearColor, value=<rgba 0 0.6514 0.7379 1>}}
  User.CubeColorMin {type=LinearColor, description="", defaultValue {struct=LinearColor, value=<rgba 0.0262 0.2016 0.5583 1>}}
  User.ImpactLocation {type=NiagaraPosition, description="", defaultValue {}}
  User.IsMaskMakerSpireShards {type=NiagaraBool}
 ]
}
