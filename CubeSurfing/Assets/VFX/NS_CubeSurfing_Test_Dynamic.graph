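# Niagara system for a cube-surfing effect (named as a dynamic test variant): a panning strip of colored cubes, with glows on the character's feet and hands.
# Blocks derives a mask from each grid cube's position to shape scale, alpha, offsets and rotation, bending with a direction vector and system timelines.
# Deploy_Block bursts cubes that converge on the origin.
# Skeletal-mesh sampling places glows, while velocity, timelines and time of day modulate hand energy and smoke.
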NiagaraSystem /CubeSurfing/Assets/VFX/NS_CubeSurfing_Test_Dynamic
dependencies {
 usedDataInterfaces [
  &NiagaraDataInterfaceCurve
  &NiagaraDataInterfaceVectorCurve
  &NiagaraDataInterfaceSkeletalMesh
  &NiagaraDataInterfaceColorCurve
  &NiagaraDataInterfaceArrayFloat3
  &NiagaraDataInterfaceSpriteRendererInfo
  &NiagaraDataInterfaceVector2DCurve
 ]
 usedDynamicInputs [
  &Vector2DFromFloat
  &MakeFloatFromVector2d
  &Multiply_Vector2
  &ClampFloat
  &VectorLength
 ]
 usedModules [
  &self:SystemSpawnScript.SystemScriptSource.SystemScriptGraph.NiagaraNodeAssignment_1.SetVariables_#1
  &self:System_State_-_Intro_Loop_Outro
  &SystemState
  &Timeline
  &self:SystemSpawnScript.SystemScriptSource.SystemScriptGraph.NiagaraNodeAssignment_0.SetVariables_#2
  &self:SystemSpawnScript.SystemScriptSource.SystemScriptGraph.NiagaraNodeAssignment_2.SetVariables_#3
  &self:SystemSpawnScript.SystemScriptSource.SystemScriptGraph.NiagaraNodeAssignment_3.SetVariables_#4
  &/TimeOfDayGlobal
 ]
}
emitters {
 Blocks {
  emitterSpawnScript [
   /SetVariables(Emitter.CubeSize=8) {isSetParameters=true}
   /SetVariables(Emitter.GridCountLength=28, Emitter.GridCountWidth=7) {isSetParameters=true}
   /SetVariables(Emitter.CubePanningSpeed=1) {isSetParameters=true}
  ]
  emitterUpdateScript [
   &EmitterState(Life Cycle Mode=::NewEnumerator1, Loop Behavior=::NewEnumerator1, Loop Duration Mode=::NewEnumerator1, UseLoopDelay=false)
   /Niagara/Modules/Spawn/Location/SpawnParticlesInGrid(X Count:NiagaraInt32=$Emitter.GridCountLength, Y Count:NiagaraInt32=$Emitter.GridCountWidth, Z Count=0)
  ]
  particleSpawnScript [
   &InitializeParticle(
    Color Maximum=$User.CubeColorMax
    Color Minimum=$User.CubeColorMin
    Color Mode=::NewEnumerator2
    Mesh Scale Mode=::NewEnumerator3
    Mesh Uniform Scale=&Multiply_Float(A=$Emitter.CubeSize, B=0.5)
    Position Offset=<-100 0 0>
    Sprite Size Mode=::NewEnumerator3
    Uniform Sprite Size=1
    UsePositionOffset=false
    Write Material Random=true
   )
   /Niagara/Modules/Spawn/Location/V2/GridLocation(
    Module.Jitter Inside Grid Cells=false
    Module.Override Random Seed=false
    Normalize Offsets=false
    Offset:Vector3f=<0 6.275 0>
    XYZ Dimensions:Vector3f=&MakeVector(
     X=&Multiply_Float(A=$Emitter.CubeSize, B=1.275)
     Y=&Multiply_Float(A=$Emitter.CubeSize, B=1.275)
     Z=&Multiply_Float(A=$Emitter.CubeSize, B=1.275)
    )
   )
   /SetVariables(Particles.GridDimensions:Vector3f=&Multiply_Vector(A=$Output.GridLocation.XYZCountAsVector, B=$Output.GridLocation.XYZ Spacing)) {isSetParameters=true}
  ]
  particleUpdateScript [
   &ParticleState(Kill Particles When Lifetime Has Elapsed=false)
   /SetVariables() {isSetParameters=true}
   /SetVariables(Particles.Position=&AddVectorToPosition(
     Position=$Particles.Initial.Position
     Vector=&Multiply_VectorByFloat(Vector=<-0.1 0 0>, Float=&Multiply_Float(A=$Emitter.DistanceTraveled, B=$Emitter.CubePanningSpeed))
    )) {isSetParameters=true}
   /SetVariables(Particles.GridDimensions:Vector3f=&Multiply_Vector(
     A=$Particles.Initial.GridDimensions
     B=&Lerp_Vector(A=<0 0.75 1>, B=<1 1 1>, Alpha=$System.Timeline.NormalizedTime)
    )) {isSetParameters=true}
   &self:ScratchModule(InputPosition:NiagaraPosition=$Particles.PositionCenter, Vector:Vector3f=$Particles.GridDimensions) {module=ScratchModule}
   /SetVariables(Particles.Mask:NiagaraFloat=&Add_Float(
     A=&Divide_Float(
      A=&MakeFloatFromVector(Vector=Linked() {linkedVariable {name=Particles.Position, type=NiagaraPosition}}, Channel=::NewEnumerator0)
      B=&MakeFloatFromVector(Vector=$Particles.GridDimensions, Channel=::NewEnumerator0)
     )
     B=0.5
    )) {isSetParameters=true}
   /SetVariables(Particles.Mask:NiagaraFloat=&Multiply_Float(
     A=$Particles.Mask
     B=&MakeCustomFloatFromBool(
      Bool=$System.Timeline.IsPlaying
      True Float=&FloatFromCurve(
       FloatCurve=DataInterface() {propertyValues {Curve {
          defaultValue=0
          keys [{arriveTangent=7.5427, interpMode=RCIM_Cubic, leaveTangent=7.5427, tangentMode=RCTM_User, time=0, value=0}, <at 0.33 1 RCIM_Cubic>]
         }}}
       CurveIndex=$System.Timeline.NormalizedTime
      )
      False Float=&FloatFromCurve(
       FloatCurve=DataInterface() {propertyValues {Curve {
          defaultValue=1
          keys [
           {tangentMode=RCTM_User, time=0, value=1}
           {arriveTangent=-0.0405, interpMode=RCIM_Cubic, leaveTangent=-0.0405, tangentMode=RCTM_User, time=0.7009, value=1}
           {arriveTangent=-8.0089, interpMode=RCIM_Cubic, leaveTangent=-8.0089, tangentMode=RCTM_User, time=1, value=0}
          ]
         }}}
       CurveIndex=$System.Timeline.InverseNormalizedTime
      )
     )
    )) {isSetParameters=true}
   &ScaleMeshSize(Scale Factor=&Lerp_Vector(
     A=<0 0 0>
     B=<1 1 1>
     Alpha=&FloatFromCurve(
      FloatCurve=DataInterface() {propertyValues {Curve {
         defaultValue=1
         keys [
          {arriveTangent=1.4146, time=0.15, value=0}
          {arriveTangent=1.5291, time=0.4, value=1}
          {arriveTangent=-2.2222, leaveTangent=-2.2222, time=0.854, value=1}
          {leaveTangent=-2.4758, time=0.95, value=0}
         ]
        }}}
      CurveIndex=$Particles.Mask
      Scale Curve=&MakeCustomFloatFromBool(
       Bool=$System.Timeline.IsPlaying
       True Float=&FloatFromCurve(
        FloatCurve=DataInterface() {propertyValues {Curve {keys [
            {arriveTangent=1.8077, interpMode=RCIM_Cubic, leaveTangent=1.8077, tangentMode=RCTM_User, time=0, value=0}
            {tangentMode=RCTM_User, time=1, value=1}
           ]}}}
        CurveIndex=$System.Timeline.NormalizedTime
       )
       False Float=&FloatFromCurve(
        FloatCurve=DataInterface() {propertyValues {Curve {keys [{interpMode=RCIM_Cubic, tangentMode=RCTM_User, time=0, value=1}, {interpMode=RCIM_Cubic, tangentMode=RCTM_User, time=1, value=0}]}}}
        CurveIndex=$System.Timeline.InverseNormalizedTime
       )
      )
     )
    ))
   &ScaleColor(
    Scale Alpha=&FloatFromCurve(
     FloatCurve=DataInterface() {propertyValues {Curve {defaultValue=1, keys [<at 0 0>, <at 0.2 1>, {arriveTangent=-1.3333, leaveTangent=-1.3333, time=0.95, value=1}, <at 1 0>]}}}
     CurveIndex=$Particles.Mask
    )
    ScaleA=true
    ScaleRGB=true
   )
   /SetVariables(Particles.Position=&AddVectorToPosition(Position=$Particles.Position, Vector=<-40 0 0>)) {isSetParameters=true}
   /SetVariables(Particles.Position=&AddVectorToPosition(
     Position=$Particles.Position
     Vector=&VectorFromCurve(
      VectorCurve=DataInterface() {propertyValues {
        XCurve {keys [
          {arriveTangent=-2.7307, interpMode=RCIM_Cubic, leaveTangent=-2.7307, tangentMode=RCTM_User, time=0, value=1}
          {arriveTangent=-0.0254, interpMode=RCIM_Cubic, leaveTangent=-0.0254, tangentMode=RCTM_User, time=0.5, value=0}
          {arriveTangent=-0.0154, interpMode=RCIM_Cubic, leaveTangent=-0.0154, tangentMode=RCTM_User, time=0.7, value=0}
          {arriveTangent=0.8018, interpMode=RCIM_Cubic, leaveTangent=0.8018, tangentMode=RCTM_User, time=1, value=0.25}
         ]}
        YCurve {keys [
          {arriveTangent=-2.7307, interpMode=RCIM_Cubic, leaveTangent=-2.7307, tangentMode=RCTM_User, time=0, value=1}
          {arriveTangent=-0.0254, interpMode=RCIM_Cubic, leaveTangent=-0.0254, tangentMode=RCTM_User, time=0.5, value=0}
          {arriveTangent=0.0182, interpMode=RCIM_Cubic, leaveTangent=0.0182, tangentMode=RCTM_User, time=0.7, value=0}
          {arriveTangent=0.8018, interpMode=RCIM_Cubic, leaveTangent=0.8018, tangentMode=RCTM_User, time=1, value=0.25}
         ]}
        ZCurve {keys [
          {arriveTangent=3.3995, interpMode=RCIM_Cubic, leaveTangent=3.3995, tangentMode=RCTM_User, time=0, value=0.2}
          {arriveTangent=-0.0629, interpMode=RCIM_Cubic, leaveTangent=-0.0629, tangentMode=RCTM_User, time=0.2, value=0.75}
          {arriveTangent=-1.3636, interpMode=RCIM_Cubic, leaveTangent=-1.3636, time=0.5, value=0}
          {arriveTangent=0.0394, interpMode=RCIM_Cubic, leaveTangent=0.0394, tangentMode=RCTM_User, time=0.75, value=0}
          {arriveTangent=-0.0989, interpMode=RCIM_Cubic, leaveTangent=-0.0989, tangentMode=RCTM_User, time=0.9, value=1}
          {arriveTangent=-31.2161, interpMode=RCIM_Cubic, leaveTangent=-31.2161, tangentMode=RCTM_User, time=1, value=-1}
         ]}
       }}
      CurveIndex=$Particles.Mask
      Scale Curve=&RandomRangeVector(Minimum=<-60 -60 15>, Maximum=<60 60 30>)
     )
    )) {isSetParameters=true}
   /SetVariables(Particles.Position=&AddVectorToPosition(
     Position=$Particles.Position
     Vector=&Multiply_VectorByFloat(
      Vector=&Multiply_Vector(A=<0 -150 -150>, B=$System.Direction Vector)
      Float=&RemapRange(
       Input Value To Remap=&Sine(Normalized Angle=$Particles.Mask, Period=2, Scale=1, Bias=0)
       Input Min=-1
       Input Max=1
       Output Min=1
       Output Max=0
      )
     )
    )) {enabled=false, isSetParameters=true}
   /SetVariables(Particles.Position=&AddVectorToPosition(
     Position=$Particles.Position
     Vector=&Multiply_VectorByFloat(
      Vector=&Multiply_Vector(A=<0 -150 -150>, B=$System.Direction Vector)
      Float=&FloatFromCurve(
       FloatCurve=DataInterface() {propertyValues {Curve {
          defaultValue=1
          keys [
           {arriveTangent=0.0061, interpMode=RCIM_Cubic, time=0, value=1}
           <at 0.66 0 RCIM_Cubic>
           {arriveTangent=6.8177, tangentMode=RCTM_User, time=1, value=1}
          ]
         }}}
       CurveIndex=$Particles.Mask
       Scale Curve=0.5
      )
     )
    )) {isSetParameters=true}
   /SetVariables(Particles.Position=&AddVectorToPosition(
     Position=$Particles.Position
     Vector=&Multiply_VectorByFloat(
      Vector=&Multiply_Vector(
       A=&Multiply_Vector(A=$System.Direction Vector, B=<-300 -1200 -1200>)
       B=&VectorFromFloat(Value=&FloatFromCurve(
         FloatCurve=DataInterface() {propertyValues {Curve {
            defaultValue=1
            keys [
             {arriveTangent=-4.2164, interpMode=RCIM_Cubic, leaveTangent=-4.2164, tangentMode=RCTM_User, time=0, value=1}
             {arriveTangent=-1.7051, interpMode=RCIM_Cubic, leaveTangent=-1.7051, tangentMode=RCTM_User, time=0.2, value=0.15}
             {arriveTangent=-0.0092, interpMode=RCIM_Cubic, leaveTangent=-0.0092, tangentMode=RCTM_User, time=0.5, value=0}
             {interpMode=RCIM_Cubic, leaveTangent=-1.3333, time=1, value=0}
            ]
           }}}
         CurveIndex=$Particles.Mask
        ))
      )
      Float=1
     )
    )) {isSetParameters=true}
   &InterpolateOverTime(
    Initial Vector Value:Vector3f=$Engine.Owner.SystemZAxis
    Rate Of Change:NiagaraFloat=&Lerp_Float(A=1, B=30, Alpha=&Power(a=$Particles.Mask, b=1.5))
    Target Vector Value:Vector3f=$Engine.Owner.SystemZAxis
    Variable Type=::NewEnumerator2
   )
   /SetVariables(Particles.RotationMask:NiagaraFloat=&DotProduct(Vector A=$Engine.Owner.SystemYAxis, Vector B=$Particles.InterpolateOverTime.InterpolateOverTime_V3_.Moving Average)) {isSetParameters=true}
   /SetVariables(Particles.Position=/Niagara/DynamicInputs/Vectors/RotateVector(
     Vector To Rotate:Vector3f=Linked() {linkedVariable {name=Particles.Position, type=NiagaraPosition}}
     Yaw:NiagaraFloat=0
     Pitch:NiagaraFloat=0
     Roll:NiagaraFloat=90
     Rotation Rate:NiagaraFloat=&MakeFloatFromVector(
      Vector=&VectorFromFloat(Value=&FloatFromCurve(
        FloatCurve=DataInterface() {propertyValues {Curve {defaultValue=1, keys [{tangentMode=RCTM_User, time=0, value=1}, {tangentMode=RCTM_User, time=1, value=-1}]}}}
        CurveIndex=$Particles.Mask
        Scale Curve=$Particles.RotationMask
       ))
      Channel=::NewEnumerator0
     )
    )) {isSetParameters=true}
   &UpdateMeshOrientation(
    Facing Coordinate Space:ENiagaraCoordinateSpace=::Local
    Facing Direction=&NormalizeVector(
     Vector To Normalize=&Add_Vector(
      A=<1 0 0>
      B=&Multiply_VectorByFloat(
       Vector=&Multiply_Vector(A=<0 1 1>, B=$System.Direction Vector)
       Float=&FloatFromCurve(
        FloatCurve=DataInterface() {propertyValues {Curve {
           defaultValue=1
           keys [
            {arriveTangent=3.8932, interpMode=RCIM_Cubic, time=0, value=-1}
            {arriveTangent=0.0477, interpMode=RCIM_Cubic, leaveTangent=0.0477, tangentMode=RCTM_Break, time=0.66, value=0}
            {arriveTangent=6.4528, tangentMode=RCTM_User, time=1, value=1}
           ]
          }}}
        CurveIndex=$Particles.Mask
        Scale Curve=-1
       )
      )
     )
     Safe Normalize=true
     Safe Fallback Threshold=0
     Safe Fallback Vector=<0 0 0.001>
    )
    Facing Interpolation=false
    Orientation Method=::NewEnumerator1
    Override Quat=false
    TransformFacing=true
    TransformUp=true
    Up Coordinate Space:ENiagaraCoordinateSpace=::World
    Up Direction:Vector3f=$Particles.InterpolateOverTime.InterpolateOverTime_V3_.Moving Average
   )
   &DynamicMaterialParameters(
    Index 0 Param 1=&FloatFromCurve(
     FloatCurve=DataInterface() {propertyValues {Curve {keys [
         {arriveTangent=-2.1994, time=0, value=1}
         {arriveTangent=-0.0235, leaveTangent=-0.0235, time=0.4, value=0}
         {arriveTangent=0.0194, leaveTangent=0.0194, time=0.825, value=0}
         <at 1 1>
        ]}}}
     CurveIndex=$Particles.Mask
    )
    Index 0 Param 2=1
    Index 0 Param 3=0.33
    Module.Index0Float0Write=true
    Module.Index0Float1Write=true
    Module.Index0Float2Write=true
    Module.Index0Float3Write=true
   )
   &KillParticles(Kill Particles=&Boolean_AndOperation(
     Bool1=&Boolean_NotOperation(Bool=/Game/Effects/Niagara/DynamicInputs/State/FN_IsExecutionStateActive())
     Bool2=&Boolean_NotOperation(Bool=$System.Timeline.IsRewinding)
    ))
   &Color(Color=&RandomRangeLinearColor(Minimum=$User.CubeColorMin, Maximum=$User.CubeColorMax))
   &ScaleMeshSize(Scale Factor=&VectorFromFloat(Value=&Waveform(
      "Waveform [1]"=::NewEnumerator1
      "[1] Frequency"=1
      "[1] Amplitude Min/Max"=&MakeVector2D(
       X=/Game/Effects/Niagara/DynamicInputs/Math/FN_Fit01(Clamp:NiagaraBool=true, Value:NiagaraFloat=$User.DeRezz_Deplete, DestMin:NiagaraFloat=0.35, DestMax:NiagaraFloat=1)
       Y=1
      )
     ))) {enabled=false, module=ScaleMeshSize001}
   &ScaleColor(Scale Alpha=&Lerp_Float(A=0, B=2, Alpha=$Particles.Mask), ScaleA=true, ScaleRGB=true) {module=ScaleColor001}
  ]
  rendererClasses [&NiagaraMeshRendererProperties, &NiagaraSpriteRendererProperties]
  renderers [
   {rendererClass=&NiagaraMeshRendererProperties, rendererIndex=0}
   {rendererClass=&NiagaraSpriteRendererProperties, rendererIndex=1}
  ]
 }
 Deploy_Block {
  emitterUpdateScript [
   &EmitterState(Inactive Response=::NewEnumerator1, Life Cycle Mode=::NewEnumerator1, Loop Behavior=::NewEnumerator1, Loop Duration=1, UseLoopDelay=false)
   &SpawnBurst_Instantaneous(Spawn Count=25, Use Spawn Probability=false, UseLoopCountLimit=false)
   &SpawnRate(
    SpawnRate=&FloatFromCurve(
     FloatCurve=DataInterface() {propertyValues {Curve {
        defaultValue=1
        keys [
         {arriveTangent=-3.0102, interpMode=RCIM_Cubic, leaveTangent=-3.0102, tangentMode=RCTM_User, time=0, value=1}
         {interpMode=RCIM_Cubic, tangentMode=RCTM_User, time=0.75, value=0}
        ]
       }}}
     CurveIndex=$Emitter.NormalizedLoopAge
     Scale Curve=30
    )
    Use Spawn Probability=false
   )
  ]
  particleSpawnScript [&InitializeParticle(
    Color Maximum=$User.CubeColorMax
    Color Minimum=$User.CubeColorMin
    Color Mode=::NewEnumerator2
    Lifetime Max=1
    Lifetime Min=0.5
    Lifetime Mode=::NewEnumerator1
    Mesh Scale Mode=::NewEnumerator1
    Position Mode=::NewEnumerator0
    Sprite Size Mode=::NewEnumerator1
    Uniform Sprite Size Max=20
    Uniform Sprite Size Min=12
    Write Material Random=true
   )]
  particleUpdateScript [
   &ParticleState()
   &ShapeLocation(
    MaskSpawnGroup=false
    Non Uniform Scale=&RandomRangeVector(Minimum=<0.25 0.25 1>, Maximum=<1 1 1>)
    Offset=&MakeVector(
     X=0
     Y=0
     Z=&FloatFromCurve(FloatCurve=DataInterface() {propertyValues {Curve {keys [<at 0 1>,<at 1 0>]}}}, Scale Curve=0)
    )
    Offset Mode=::NewEnumerator0
    Override Seed=false
    Rotation Mode=::NewEnumerator2
    Sphere Radius=&FloatFromCurve(
     FloatCurve=DataInterface() {propertyValues {Curve {
        defaultValue=0
        keys [
         {arriveTangent=2.1078, interpMode=RCIM_Cubic, leaveTangent=2.1078, tangentMode=RCTM_User, time=0, value=0.5}
         {interpMode=RCIM_Cubic, tangentMode=RCTM_User, time=0.5, value=1}
         {tangentMode=RCTM_User, time=1, value=0.5}
        ]
       }}}
     Scale Curve=&RandomRangeFloat(Minimum=150, Maximum=200)
    )
    Sphere Surface Distribution=0.75
    UseOwnerScale=true
    Yaw / Pitch / Roll=&VectorFromCurve(
     VectorCurve=DataInterface() {propertyValues {
       XCurve {
        defaultValue=0
        keys [{interpMode=RCIM_Cubic, tangentMode=RCTM_User, time=0, value=1}, {interpMode=RCIM_Cubic, tangentMode=RCTM_User, time=1, value=0}]
       }
       YCurve {
        defaultValue=0
        keys [
         {arriveTangent=-1.3041, interpMode=RCIM_Cubic, leaveTangent=-1.3041, tangentMode=RCTM_User, time=0, value=0.8}
         {interpMode=RCIM_Cubic, tangentMode=RCTM_User, time=1, value=0.25}
        ]
       }
       ZCurve {
        defaultValue=1
        keys [{interpMode=RCIM_Cubic, tangentMode=RCTM_User, time=0, value=0}, {interpMode=RCIM_Cubic, tangentMode=RCTM_User, time=1, value=1}]
       }
      }}
     Scale Curve=&RandomRangeVector(Minimum=<0.25 0.25 0.25>, Maximum=<1 1 1>)
    )
   )
   /SetVariables(Particles.Position=&LerpPosition(A=$Particles.Position, B=/Niagara/DynamicInputs/Helpers/SimulationPosition(), Alpha=$Particles.NormalizedAge)) {isSetParameters=true}
   &ScaleMeshSize(Scale Factor=&Lerp_Vector(
     A=&RandomRangeVector(Minimum=<0.1 0.1 0.1>, Maximum=<0.25 0.25 0.25>)
     B=<1 1 1>
     Alpha=&FloatFromCurve(FloatCurve=DataInterface() {propertyValues {Curve {
         defaultValue=0
         keys [
          {arriveTangent=4.2267, interpMode=RCIM_Cubic, leaveTangent=4.2267, tangentMode=RCTM_User, time=0, value=0}
          {interpMode=RCIM_Cubic, tangentMode=RCTM_User, time=0.5, value=1}
          {tangentMode=RCTM_User, time=1, value=0}
         ]
        }}})
    ))
   &ScaleSpriteSize(
    Scale Sprite Size Mode=::NewEnumerator0
    Uniform Scale Factor=&FloatFromCurve(FloatCurve=DataInterface() {propertyValues {Curve {keys [
         {arriveTangent=9.1187, interpMode=RCIM_Cubic, time=0, value=0}
         {arriveTangent=1.1429, interpMode=RCIM_Cubic, leaveTangent=1.1429, time=0.2487, value=1.1246}
         {arriveTangent=-1.497, interpMode=RCIM_Cubic, leaveTangent=-1.497, time=0.7, value=0.8}
         {arriveTangent=-3.0554, interpMode=RCIM_Cubic, leaveTangent=-3.0554, tangentMode=RCTM_User, time=1, value=0}
        ]}}})
   )
   &DynamicMaterialParameters(
    Index 0 Param 1=&FloatFromCurve(
     FloatCurve=DataInterface() {propertyValues {Curve {
        defaultValue=0
        keys [
         {arriveTangent=9.0725, interpMode=RCIM_Cubic, leaveTangent=9.0725, tangentMode=RCTM_User, time=0, value=0}
         {arriveTangent=-2.5249, interpMode=RCIM_Cubic, leaveTangent=-2.5249, tangentMode=RCTM_User, time=0.1408, value=0.9847}
         {arriveTangent=-0.0405, interpMode=RCIM_Cubic, leaveTangent=-0.0405, tangentMode=RCTM_User, time=1, value=0}
        ]
       }}}
     Scale Curve=&RandomRangeFloat(Minimum=0, Maximum=2)
    )
    Index 0 Param 2=1
    Index 0 Param 3=0.33
    Module.Index0Float0Write=true
    Module.Index0Float1Write=true
    Module.Index0Float2Write=true
    Module.Index0Float3Write=true
   )
   &Color(Color=&RandomRangeLinearColor(Minimum=$User.CubeColorMin, Maximum=$User.CubeColorMax))
   /SetVariables(Particles.Color_Glow:LinearColor=&MultiplyLinearColor(
     Linear Color=<rgba 0 0.7437 1 0.1>
     Scale Factor=&MakeVector4_FromLinearColor(LinearColor=$Particles.Color)
    )) {isSetParameters=true}
  ]
  rendererClasses [&NiagaraMeshRendererProperties, &NiagaraSpriteRendererProperties]
  renderers [
   {rendererClass=&NiagaraMeshRendererProperties, rendererIndex=0}
   {rendererClass=&NiagaraSpriteRendererProperties, rendererIndex=1}
  ]
 }
 Glow {
  emitterUpdateScript [
   &EmitterState(
    Inactive Response=::NewEnumerator1
    Life Cycle Mode=::NewEnumerator1
    Loop Behavior=::NewEnumerator1
    Loop Delay=0.1
    Loop Duration Mode=::NewEnumerator1
    UseLoopDelay=true
   )
   &SpawnBurst_Instantaneous(Spawn Group=1, Use Spawn Probability=false, UseLoopCountLimit=false) {module=SpawnBurst_Instantaneous001}
   &SpawnBurst_Instantaneous(Spawn Count=2, Use Spawn Probability=false, UseLoopCountLimit=false)
  ]
  particleSpawnScript [
   /SetVariables(Particles.VisibilityTag=$Emitter.SpawnGroup) {isSetParameters=true}
   &InitializeParticle(
    Color=<rgba 0.2893 1.1517 2.5 0.5>
    Lifetime=2
    Position Mode=::NewEnumerator0
    Sprite Size Mode=::NewEnumerator3
    Uniform Sprite Size=75
    Write Material Random=true
   )
   &PartitionParticles()
   &CameraOffset(Camera Offset Amount=20)
   /SetVariables(Particles.Color_FakeLight_Foot:LinearColor=<rgba 0.2042 2.2834 3 0.1>, Particles.SpriteSize_FakeLight_Foot:Vector2f=<50 50>) {isSetParameters=true}
  ]
  particleUpdateScript [
   &ParticleState(
    Kill Particles When Lifetime Has Elapsed=false
    Let Infinitely Lived Particles Die When Emitter Deactivates=true
    Loop Particles Lifetime=true
   )
   &SkeletalMeshLocation(
    Bone / Socket Index=$Particles.PartitionParticles.Partition
    Bone Sampling Mode=::NewEnumerator4
    MaskSpawnGroup=false
    Sampled Position Offset=<0 0 -15>
    Sampled Position Offset Coordinate Space:ENiagaraSkelMeshTransforms=::NewEnumerator1
    Skeletal Mesh=DataInterface() {propertyValues {
      DefaultMesh=null
      ExcludeBoneName=root
      OptionalComponentClass=null
      PreviewMesh=&M_Med_Soldier_04_Preview
      SkinningMode=SkinOnTheFly
      SoftSourceActor=null
      SourceMode=Default
      UvSetIndex=0
      WholeMeshLOD=-1
      bExcludeBone=false
      bReadDeformedGeometry=true
      bRequireCurrentFrameData=true
      ComponentTags []
      FilteredBones [foot_l,foot_r]
      FilteredSockets []
      MeshUserParameter {parameter {name=None, typeDefHandle {registeredTypeIndex=1}, varData []}}
      SamplingRegions []
     }}
   )
   /SetVariables(
    Particles.Color_FakeLight:LinearColor=<rgba 0 1.8977 2.5 0.1>
    Particles.Position_FakeLight:NiagaraPosition=&AddVectorToPosition(Position=&SimulationPosition(), Vector=<0 0 -100>)
    Particles.SpriteSize_FakeLight:Vector2f=<250 250>
   ) {isSetParameters=true}
   &ScaleColor(
    Scale Alpha=&FloatFromCurve(
     FloatCurve=DataInterface() {propertyValues {Curve {
        defaultValue=0
        keys [{interpMode=RCIM_Cubic, tangentMode=RCTM_User, time=0, value=0}, {interpMode=RCIM_Cubic, tangentMode=RCTM_User, time=1, value=1}]
       }}}
     CurveIndex=$Engine.System.Age
    )
    ScaleA=true
    ScaleRGB=false
   )
   &SolveForcesAndVelocity(Clamp Velocity=false, Limit Acceleration=false)
  ]
  rendererClasses [&NiagaraSpriteRendererProperties]
  renderers [
   {rendererClass=&NiagaraSpriteRendererProperties}
   {rendererClass=&NiagaraSpriteRendererProperties, rendererIndex=1}
   {rendererClass=&NiagaraSpriteRendererProperties, rendererIndex=2}
  ]
 }
 Hands {
  emitterUpdateScript [
   &EmitterState(
    Inactive Response=::NewEnumerator1
    Life Cycle Mode=::NewEnumerator1
    Loop Behavior=::NewEnumerator1
    Loop Duration Mode=::NewEnumerator1
    UseLoopDelay=false
   )
   &SpawnBurst_Instantaneous(Spawn Count=2, Use Spawn Probability=false, UseLoopCountLimit=false)
  ]
  particleSpawnScript [
   &PartitionParticles()
   /SetVariables(Particles.HandFootAlpha:NiagaraFloat=&MakeCustomFloatFromBool(
     Bool=&SetBoolByIntComparison(A=$Particles.PartitionParticles.Partition, B=1, Comparison Type=::NewEnumerator0)
     True Float=1
     False Float=0
    )) {isSetParameters=true}
   &InitializeParticle(
    Lifetime=0.25
    Mass Mode=::NewEnumerator1
    Mesh Scale Mode=::NewEnumerator3
    Mesh Uniform Scale=0.125
    Sprite Rotation Mode=::NewEnumerator2
    Sprite Size Mode=::NewEnumerator3
    Uniform Sprite Size=50
    UsePositionOffset=true
    Write Material Random=true
   ) {module=InitializeParticle001}
   &CameraOffset(Camera Offset Amount=10)
   /SetVariables(Particles.GlowCameraOffset=10) {isSetParameters=true}
   &InitialMeshOrientation(Orientation Vector=<0 0 1>, Rotation=<0 0.25 0>, Use Orientation Vector=true, Use Rotation Vector=true) {module=InitialMeshOrientation001}
  ]
  particleUpdateScript [
   &ParticleState(Kill Particles When Lifetime Has Elapsed=false, Let Infinitely Lived Particles Die When Emitter Deactivates=true)
   /Niagara/Noise/Waveform(
    EnableExponent1=false
    GlobalClampMax=false
    GlobalClampMin=false
    UseInput1=true
    UsePhase1=false
    "Waveform [1]":ENiagara_Waveforms=::NewEnumerator12
    "[1] Amplitude Min/Max":Vector2f=<0.5 1.2>
    "[1] Frequency":NiagaraFloat=&RandomRangeFloat(Minimum=5, Maximum=10)
    "[1] Input Position":NiagaraFloat=&RandomRangeFloat(Minimum=0, Maximum=1)
   )
   &SolveForcesAndVelocity(Clamp Velocity=false, Limit Acceleration=false)
   &SkeletalMeshLocation(
    Bone / Socket Index=$Particles.PartitionParticles.Partition
    MaskSpawnGroup=false
    Mesh Sampling Type=::NewEnumerator5
    Orientation Sampling=::NewEnumerator1
    Sampled Position Offset=<0 5 0>
    Skeletal Mesh=DataInterface() {propertyValues {
      DefaultMesh=null
      ExcludeBoneName=root
      OptionalComponentClass=null
      PreviewMesh=&M_Med_Soldier_04_Preview
      SkinningMode=SkinOnTheFly
      SoftSourceActor=null
      SourceMode=Default
      UvSetIndex=0
      WholeMeshLOD=-1
      bExcludeBone=false
      bReadDeformedGeometry=true
      bRequireCurrentFrameData=true
      ComponentTags []
      FilteredBones []
      FilteredSockets [hand_pickaxe_l, hand_pickaxe_r]
      MeshUserParameter {parameter {name=None, typeDefHandle {registeredTypeIndex=1}, varData []}}
      SamplingRegions []
     }}
    Socket Sampling Mode=::NewEnumerator5
   )
   &ScaleSpriteSize(Scale Factor=&Multiply_Vector2D_ByFloat(
     Float=&FloatFromCurve(FloatCurve=DataInterface() {propertyValues {Curve {keys [
          {arriveTangent=0.7585, interpMode=RCIM_Cubic, leaveTangent=0.7585, tangentMode=RCTM_User, time=0, value=0.4}
          {interpMode=RCIM_Cubic, tangentMode=RCTM_User, time=1, value=1}
         ]}}})
     Vector2D=<1 1>
    ))
   &SpriteRotationRate(Rotation Rate=&UniformRangedFloat(Minimum=-50, Maximum=50))
   &SpriteFacingAndAlignment(
    Sprite Alignment=&TransformVector(Vector=<1 0 0>, Source Space=::World, Destination Space=::Simulation)
    WriteAlignment=true
    WriteFacing=false
   )
   /SetVariables(Particles.DarkGlowSize:Vector2f=&Multiply_Vector2D_ByFloat(Float=1.3, Vector2D=$Particles.SpriteSize)) {isSetParameters=true}
   /SetVariables(Particles.DargGlowColor:LinearColor=&MakeLinearColorFromVectorAndFloat(
     "Vector (RGB)"=&MakeVectorFromLinearColorRGB(Color=<rgba 0 0.2859 0.4144 1>)
     "Float (Alpha)"=&FloatFromCurve(
      FloatCurve=DataInterface() {propertyValues {Curve {keys [
          {arriveTangent=6.4019, interpMode=RCIM_Cubic, leaveTangent=6.4019, tangentMode=RCTM_User, time=0, value=0}
          {interpMode=RCIM_Cubic, leaveTangent=0.002, time=0.2, value=0.7}
         ]}}}
      Scale Curve=&Multiply_Float(
       A=&Multiply_Float(A=1, B=&Lerp_Float(A=$System.Idle_Opacity_NoBoost, B=1, Alpha=$System.Timeline001.NormalizedTime))
       B=$Particles.Waveform.Waveform
      )
     )
    )) {isSetParameters=true}
   /SetVariables(Particles.CentralFlareSize:Vector2f=&Multiply_Vector2D_ByFloat(Float=1, Vector2D=$Particles.SpriteSize)) {isSetParameters=true}
   /SetVariables(Particles.CentralFlareColor:LinearColor=&MakeLinearColorFromVectorAndFloat(
     "Vector (RGB)"=&MakeVectorFromLinearColorRGB(Color=&ColorFromCurve(ColorCurve=DataInterface() {propertyValues {
         AlphaCurve {keys [<at 0 1 RCIM_Cubic>]}
         BlueCurve {keys [<at 0 0.5 RCIM_Cubic>,<at 1 2 RCIM_Cubic>]}
         GreenCurve {keys [<at 0 0.3303 RCIM_Cubic>,<at 1 1.2751 RCIM_Cubic>]}
         RedCurve {keys [<at 0 0 RCIM_Cubic>,<at 1 0 RCIM_Cubic>]}
        }}))
     "Float (Alpha)"=&Multiply_Float(
      A=&FloatFromCurve(
       FloatCurve=DataInterface() {propertyValues {Curve {keys [
           {arriveTangent=6.4019, interpMode=RCIM_Cubic, leaveTangent=6.4019, tangentMode=RCTM_User, time=0, value=0}
           {interpMode=RCIM_Cubic, leaveTangent=0.002, time=0.2, value=0.7}
          ]}}}
       Scale Curve=&Multiply_Float(
        A=&Lerp_Float(A=$System.Idle_Opacity_NoBoost, B=1, Alpha=&Multiply_Float(A=$System.Timeline001.NormalizedTime, B=0.5))
        B=$Particles.Waveform.Waveform
       )
      )
      B=&FNTimeOfDayScalar(Night Scale=0.5, Morning Scale=0.5, Evening Scale=0.5)
     )
    )) {isSetParameters=true}
   &FN_Light_Attributes(
    Color=$Particles.CentralFlareColor
    Create Color Attribute=true
    Create Exponent Attribute=true
    Create Radius Attribute=true
    Exponent=1
    Radius=30
    Scale Color=&Multiply_Float(
     A=&Multiply_Float(A=100, B=&Lerp_Float(A=$System.Idle_Opacity_NoBoost, B=1, Alpha=$System.Timeline001.NormalizedTime))
     B=$Particles.Waveform.Waveform
    )
   )
   &DynamicMaterialParameters(
    Index 1 Param 1=0.3
    Index 1 Param 3=1
    Module.Index1Float0Write=true
    Module.Index1Float1Write=true
    Module.Index1Float2Write=true
    Module.Index1Float3Write=true
    Module.Index3Float0Write=true
    Module.Index3Float1Write=true
    Module.Index3Float2Write=true
    Module.Index3Float3Write=true
    Write Parameter Index 0=::NewEnumerator0
    Write Parameter Index 1=::NewEnumerator1
    Write Parameter Index 3=::NewEnumerator1
   )
   &FN_Sprite_Attributes(
    Alignment:Vector3f=<1 0 0>
    Create Alignment Attribute=true
    Create Sprite Rotation Attribute=true
    Create SpriteSize Attribute=true
    SpriteSize=&Multiply_Vector2D(A=$Particles.CentralFlareSize, B=<0.5 0.5>)
    Write to Module Namespace=true
   )
   &ScaleMeshSize(Scale Factor=&ScaleFloatByCurve(
     InitialValue:Vector3f=<1 1 1>
     ScaleCurve:NiagaraDataInterfaceVectorCurve=DataInterface() {propertyValues {XCurve {keys [<at 0 0.25>,<at 1 1>]}, YCurve {keys [<at 0 1.25>,<at 1 2.25>]}, ZCurve {keys [<at 0 1.25>,<at 1 2.25>]}}}
     CurveIndex:NiagaraFloat=$Particles.NormalizedAge
    ))
  ]
  rendererClasses [
   &NiagaraSpriteRendererProperties
   &NiagaraLightRendererProperties
   &NiagaraMeshRendererProperties
  ]
  renderers [
   {rendererClass=&NiagaraSpriteRendererProperties}
   {rendererClass=&NiagaraSpriteRendererProperties, rendererIndex=1}
   {rendererClass=&NiagaraLightRendererProperties, rendererIndex=2}
   {rendererClass=&NiagaraSpriteRendererProperties, rendererIndex=3}
   {rendererClass=&NiagaraMeshRendererProperties, rendererIndex=4}
  ]
 }
 Hands_Energy {
  emitterUpdateScript [
   &EmitterState(Life Cycle Mode=::NewEnumerator1, Loop Behavior=::NewEnumerator1, Loop Duration Mode=::NewEnumerator1, UseLoopDelay=false)
   &SpawnBurst_Instantaneous(Spawn Count=2, Use Spawn Probability=false)
  ]
  particleSpawnScript [
   &InitializeParticle(
    Color=<rgba 1 1 1 3>
    Lifetime Max=1
    Lifetime Mode=::NewEnumerator1
    Mesh Scale=<0.55 0.55 0.5>
    Mesh Scale Mode=::NewEnumerator4
    UsePositionOffset=true
    Write Material Random=true
   )
   &PartitionParticles()
   &InitialMeshOrientation(Rotation=<0 0.5 0>, Use Orientation Vector=true, Use Rotation Vector=true)
   &DynamicMaterialParameters(
    Index 0 Param 1=1
    Index 0 Param 2=1
    Index 0 Param 3=1
    Index 0 Param 4=1
    Module.Param0WriteEnabled=true
    Module.Param1WriteEnabled=true
    Module.Param2WriteEnabled=true
    Module.Param3WriteEnabled=true
   )
   &CameraOffset(Camera Offset Amount=10)
  ]
  particleUpdateScript [
   &SolveForcesAndVelocity(Clamp Velocity=false, Limit Acceleration=false)
   &SkeletalMeshLocation(
    Bone / Socket Index=$Particles.PartitionParticles.Partition
    MaskSpawnGroup=false
    Mesh Sampling Type=::NewEnumerator5
    Orientation Sampling=::NewEnumerator1
    Skeletal Mesh=DataInterface() {propertyValues {
      DefaultMesh=null
      ExcludeBoneName=root
      OptionalComponentClass=null
      PreviewMesh=&M_Med_Soldier_04_Preview
      SkinningMode=SkinOnTheFly
      SoftSourceActor=null
      SourceMode=Default
      UvSetIndex=0
      WholeMeshLOD=-1
      bExcludeBone=false
      bReadDeformedGeometry=true
      bRequireCurrentFrameData=true
      ComponentTags []
      FilteredBones []
      FilteredSockets [hand_pickaxe_l, hand_pickaxe_r]
      MeshUserParameter {parameter {name=None, typeDefHandle {registeredTypeIndex=1}, varData []}}
      SamplingRegions []
     }}
    Socket Sampling Mode=::NewEnumerator5
   )
   &ParticleState(Kill Particles When Lifetime Has Elapsed=false, Let Infinitely Lived Particles Die When Emitter Deactivates=true)
   &ScaleColor(
    Scale Alpha=&Multiply_Float(
     A=&FNTimeOfDayScalar(Night Scale=1, Morning Scale=0.3, Evening Scale=0.3)
     B=&FloatFromCurve(
      FloatCurve=DataInterface() {propertyValues {Curve {keys [
          {arriveTangent=3.944, interpMode=RCIM_Cubic, leaveTangent=3.944, tangentMode=RCTM_User, time=0, value=0}
          {arriveTangent=-0.0149, interpMode=RCIM_Cubic, leaveTangent=-0.0149, tangentMode=RCTM_User, time=0.75, value=1}
         ]}}}
      CurveIndex=$System.NormalizedVelocityRange
      Scale Curve=0.5
     )
    )
    Scale RGB=&VectorFromFloat(Value=0.5)
    ScaleA=true
    ScaleRGB=true
   )
   &ScaleMeshSize()
  ]
  rendererClasses [&NiagaraMeshRendererProperties]
  renderers [{rendererClass=&NiagaraMeshRendererProperties, rendererIndex=0}]
 }
 Smoke {
  emitterUpdateScript [
   &EmitterState(Life Cycle Mode=::NewEnumerator1, Loop Behavior=::NewEnumerator1, Loop Duration Mode=::NewEnumerator1, UseLoopDelay=false)
   &SpawnRate(
    Spawn Probability=&Multiply_Float(
     A=&Multiply_Float(A=$System.Timeline002.InverseNormalizedTime, B=$System.Timeline001.InverseNormalizedTime)
     B=$System.Timeline003.NormalizedTime
    )
    SpawnRate=&FloatFromCurve(
     FloatCurve=DataInterface() {propertyValues {Curve {keys [
         {arriveTangent=-0.0514, interpMode=RCIM_Cubic, leaveTangent=-0.0514, tangentMode=RCTM_User, time=0, value=0}
         {arriveTangent=0.114, interpMode=RCIM_Cubic, leaveTangent=0.114, tangentMode=RCTM_User, time=0.5, value=1}
         {arriveTangent=-0.0149, interpMode=RCIM_Cubic, leaveTangent=-0.0149, tangentMode=RCTM_User, time=0.75, value=1}
        ]}}}
     CurveIndex=$System.NormalizedVelocityRange
     Scale Curve=10
    )
    Use Spawn Probability=true
   )
  ]
  particleSpawnScript [
   &PartitionParticles()
   &InitializeParticle(
    Color Channel Mode:ENiagaraChannelCorrelation=::NewEnumerator0
    Color Maximum=<rgba 0.35 0.3477 0.336 0.15>
    Color Minimum=<rgba 0.35 0.3404 0.3251 0.1>
    Color Mode=::NewEnumerator2
    Lifetime Max=0.8
    Lifetime Min=0.5
    Lifetime Mode=::NewEnumerator1
    Sprite Rotation Mode=::NewEnumerator2
    Sprite Size Mode=::NewEnumerator1
    Uniform Sprite Size Max=50
    Uniform Sprite Size Min=25
    UsePositionOffset=false
    Write Material Random=true
   ) {module=InitializeParticle001}
   &ShapeLocation(
    MaskSpawnGroup=false
    Offset=&SelectVectorFromArray(
     Direct Array Index=$Particles.PartitionParticles.Partition
     Vector Selection Array=DataInterface() {propertyValues {FloatData [<-50 35 -25>, <-50 -10 -25>]}}
    )
    Offset Coordinate Space=::Local
    Offset Mode=::NewEnumerator0
    Override Seed=false
    Sphere Radius=5
    UseOwnerScale=true
   )
   &AddVelocity(Velocity=<-5 0 0>, Velocity Speed Scale=&RandomRangeFloat(Minimum=0, Maximum=100)) {module=AddVelocity001}
   &AddVelocity(
    Module.Use Falloff Distance=false
    Origin Offset=<222 0 0>
    Override Seed=false
    Velocity Mode=::NewEnumerator1
    Velocity Origin=$Particles.ShapeLocation.ShapePosition
    Velocity Speed=250
   )
   &AccelerationForce(Acceleration=&UniformRangedVector(Minimum=<-33 0 0>, Maximum=<-111 0 0>), Coordinate Space=::Local)
   &ApplyInitialForces()
  ]
  particleUpdateScript [
   /SetVariables(Particles.Jitter:NiagaraFloat=/Niagara/DynamicInputs/Helpers/Waveform("Waveform [1]"=::NewEnumerator12, "[1] Frequency"=25, "[1] Amplitude Min/Max"=<0.5 1>)) {isSetParameters=true}
   &ParticleState()
   &Drag(Drag=10, Use Linear Drag=true, Use Rotational Drag=true)
   &InheritSourceMovement(
    Applied Position Scale=&Multiply_VectorByFloat(
     Vector=<1 1 1>
     Float=&FloatFromCurve(FloatCurve=DataInterface() {propertyValues {Curve {
         defaultValue=1
         keys [{interpMode=RCIM_Cubic, tangentMode=RCTM_User, time=0, value=1}, {interpMode=RCIM_Cubic, tangentMode=RCTM_User, time=1, value=0}]
        }}})
    )
    ApplyForce=false
    ApplyPosition=true
    ApplyPositionFalloff=false
   )
   &SolveForcesAndVelocity(Clamp Velocity=false, Limit Acceleration=false)
   &ScaleSpriteSize(
    Uniform Curve Scale=&Lerp_Float(A=1.33, B=2, Alpha=$System.Timeline.Float)
    Uniform Curve Sprite Scale=DataInterface() {propertyValues {Curve {keys [
        {arriveTangent=1.491, interpMode=RCIM_Cubic, leaveTangent=1.491, tangentMode=RCTM_User, time=0, value=0.5}
        {arriveTangent=0.5, interpMode=RCIM_Cubic, leaveTangent=0.5, time=0.95, value=3}
        <at 1 1 RCIM_Cubic>
       ]}}}
   )
   &DynamicMaterialParameters(
    Index 0 Param 1=&Add_Float(
     A=&RandomRangeFloat(Minimum=0, Maximum=1)
     B=&FloatFromCurve(
      FloatCurve=DataInterface() {propertyValues {Curve {keys [
          <at 0 0 RCIM_Cubic>
          {arriveTangent=1.5947, interpMode=RCIM_Cubic, leaveTangent=1.5947, tangentMode=RCTM_User, time=0.2557, value=0.5257}
          <at 1 1 RCIM_Cubic>
         ]}}}
      CurveIndex=$Particles.Age
      Scale Curve=0.25
     )
    )
    Index 0 Param 3=1
    Index 0 Param 4=1
    Index 2 Param 3=2
    Index 3 Param 1=&Lerp_Float(A=0, B=&Multiply_Float(A=1, B=$Particles.Jitter), Alpha=$System.Timeline.Float)
    Index 3 Param 2=125
    Module.Param0WriteEnabled=true
    Module.Param0WriteEnabled1=true
    Module.Param0WriteEnabled2=true
    Module.Param0WriteEnabled3=true
    Module.Param1WriteEnabled=true
    Module.Param1WriteEnabled1=true
    Module.Param1WriteEnabled2=true
    Module.Param1WriteEnabled3=true
    Module.Param2WriteEnabled=true
    Module.Param2WriteEnabled1=true
    Module.Param2WriteEnabled2=true
    Module.Param2WriteEnabled3=true
    Module.Param3WriteEnabled=true
    Module.Param3WriteEnabled1=true
    Module.Param3WriteEnabled2=true
    Module.Param3WriteEnabled3=true
    Write Parameter Index 1=true
    Write Parameter Index 2=true
    Write Parameter Index 3=true
   ) {module=DynamicMaterialParameters001}
   &CameraOffset(Camera Offset Amount=5)
   &ScaleColor(
    Scale Alpha=&Multiply_Float(
     A=&FloatFromCurve(
      FloatCurve=DataInterface() {propertyValues {Curve {keys [{interpMode=RCIM_Cubic, tangentMode=RCTM_User, time=0, value=0}, <at 0.25 1 RCIM_Cubic>, <at 0.9 0>]}}}
      Scale Curve=0.7
     )
     B=&FloatFromCurve(
      FloatCurve=DataInterface() {propertyValues {Curve {keys [
          {interpMode=RCIM_Cubic, tangentMode=RCTM_User, time=0, value=0.8}
          {interpMode=RCIM_Cubic, tangentMode=RCTM_User, time=0.5, value=1}
          {tangentMode=RCTM_User, time=1, value=0.8}
         ]}}}
      CurveIndex=Linked() {linkedVariable {name=NPC.FortniteNPC.TimeOfDay_0-1, type=NiagaraFloat}}
      Scale Curve=&Multiply_Float(
       A=&Multiply_Float(A=$System.Timeline002.InverseNormalizedTime, B=$System.Timeline001.InverseNormalizedTime)
       B=$System.Timeline003.NormalizedTime
      )
     )
    )
    ScaleA=true
    ScaleRGB=true
   )
  ]
  rendererClasses [&NiagaraSpriteRendererProperties]
  renderers [
   {rendererClass=&NiagaraSpriteRendererProperties}
   {rendererClass=&NiagaraSpriteRendererProperties, rendererIndex=1}
  ]
 }
 Wind_Streaks {
  emitterUpdateScript [
   &EmitterState(Life Cycle Mode=::NewEnumerator1, Loop Behavior=::NewEnumerator1, Loop Duration Mode=::NewEnumerator1, UseLoopDelay=false)
   &SpawnRate(
    Spawn Probability=&Multiply_Float(
     A=&Multiply_Float(A=$System.Timeline002.InverseNormalizedTime, B=$System.Timeline001.InverseNormalizedTime)
     B=$System.Timeline003.NormalizedTime
    )
    SpawnRate=&FloatFromCurve(
     FloatCurve=DataInterface() {propertyValues {Curve {keys [
         {arriveTangent=-0.0514, interpMode=RCIM_Cubic, leaveTangent=-0.0514, tangentMode=RCTM_User, time=0, value=0}
         {arriveTangent=0.114, interpMode=RCIM_Cubic, leaveTangent=0.114, tangentMode=RCTM_User, time=0.5, value=1}
         {arriveTangent=-0.0149, interpMode=RCIM_Cubic, leaveTangent=-0.0149, tangentMode=RCTM_User, time=0.75, value=1}
        ]}}}
     CurveIndex=$System.NormalizedVelocityRange
     Scale Curve=10
    )
    Use Spawn Probability=true
   )
  ]
  particleSpawnScript [
   &InitializeParticle(
    Color Maximum=<rgba 0.4 0.3068 0.26 0.15>
    Color Minimum=<rgba 0.4 0.3617 0.3 0.2>
    Color Mode=::NewEnumerator2
    Lifetime=&UniformRangedFloat(Minimum=0.3, Maximum=0.5)
    Sprite Size=&UniformRangedVector2D(Minimum=<35 50>, Maximum=<45 75>)
    Sprite Size Mode=::NewEnumerator4
    UsePositionOffset=false
    Write Material Random=true
   )
   &PartitionParticles()
   &ShapeLocation(
    MaskSpawnGroup=false
    Offset=&SelectVectorFromArray(
     Direct Array Index=$Particles.PartitionParticles.Partition
     Vector Selection Array=DataInterface() {propertyValues {FloatData [<-50 35 -25>, <-50 -10 -25>]}}
     Random Seed=0
    )
    Offset Coordinate Space=::Local
    Offset Mode=::NewEnumerator0
    Override Seed=false
    Sphere Radius=5
    UseOwnerScale=true
   )
   &AddVelocity(
    Velocity=&MergeHelper_VectorSwitch(0=<-5 5 2.5>, 1=<-5 -5 2.5>, 2=?, 3=?, Index=$Particles.PartitionParticles.Partition, Num=::NewEnumerator0)
    Velocity Speed Scale=&RandomRangeFloat(Minimum=0, Maximum=100)
   ) {module=AddVelocity001}
   &AddVelocity(
    Module.Use Falloff Distance=false
    Origin Offset=<222 0 0>
    Override Seed=false
    Velocity Mode=::NewEnumerator1
    Velocity Origin=$Particles.ShapeLocation.ShapePosition
    Velocity Speed=250
   )
   &AccelerationForce(Acceleration=&UniformRangedVector(Minimum=<-33 0 0>, Maximum=<-111 0 0>), Coordinate Space=::Local)
   &ApplyInitialForces()
   &SubUVAnimation(
    Evaluation Type=::NewEnumerator0
    "Sprite Renderer "=DataInterface() {propertyValues {SpriteRenderer=&self:Wind_Streaks_0.NiagaraSpriteRendererProperties_0}}
    SubUV Animation Mode:ENiagara_SubUVLookupModeV2=::NewEnumerator2
    UseEndFrame=false
    UseStartFrame=false
   )
  ]
  particleUpdateScript [
   &ParticleState()
   &ScaleSpriteSize(Scale Factor=&Vector2DFromCurve(
     Vector2Curve=DataInterface() {propertyValues {
       XCurve {keys [
         {arriveTangent=2.5428, interpMode=RCIM_Cubic, leaveTangent=2.5428, tangentMode=RCTM_User, time=0, value=1}
         {arriveTangent=0.1281, interpMode=RCIM_Cubic, leaveTangent=0.1281, tangentMode=RCTM_User, time=1, value=2}
        ]}
       YCurve {keys [
         {arriveTangent=2.8823, interpMode=RCIM_Cubic, leaveTangent=2.8823, tangentMode=RCTM_User, time=0, value=1}
         {arriveTangent=2.0297, interpMode=RCIM_Cubic, leaveTangent=2.0297, tangentMode=RCTM_User, time=1, value=2}
        ]}
      }}
     ScaleCurve=&Multiply_Vector2D_ByFloat(
      Float=&RandomRangeFloat(Minimum=1, Maximum=2)
      Vector2D=&RandomRangeVector2D(Minimum=<0.6 0.75>, Maximum=<0.8 1>)
     )
    ))
   &DynamicMaterialParameters(
    Index 0 Param 1=20
    Index 0 Param 2=1
    Index 0 Param 3=1
    Index 0 Param 4=&FloatFromCurve(FloatCurve=DataInterface() {propertyValues {Curve {keys [{tangentMode=RCTM_User, time=0.1, value=0}, {tangentMode=RCTM_User, time=1, value=0.5}]}}})
    Module.Param0WriteEnabled=true
    Module.Param1WriteEnabled=true
    Module.Param2WriteEnabled=true
    Module.Param3WriteEnabled=true
   ) {module=DynamicMaterialParameters001}
   &Drag(Drag=5, Use Linear Drag=true, Use Rotational Drag=true)
   &InheritSourceMovement(
    Applied Position Scale=&Multiply_VectorByFloat(
     Vector=<1 1 1>
     Float=&FloatFromCurve(FloatCurve=DataInterface() {propertyValues {Curve {
         defaultValue=1
         keys [
          {interpMode=RCIM_Cubic, tangentMode=RCTM_User, time=0, value=1}
          {arriveTangent=-3.2686, interpMode=RCIM_Cubic, leaveTangent=-3.2686, tangentMode=RCTM_User, time=1, value=0}
         ]
        }}})
    )
    ApplyForce=false
    ApplyPosition=true
    ApplyPositionFalloff=false
   )
   &SolveForcesAndVelocity(Clamp Velocity=false, Limit Acceleration=false)
   &ScaleColor(
    Scale Alpha=&FloatFromCurve(
     FloatCurve=DataInterface() {propertyValues {Curve {keys [
         {arriveTangent=9.0725, interpMode=RCIM_Cubic, leaveTangent=9.0725, tangentMode=RCTM_User, time=0, value=0}
         {arriveTangent=-2.5249, interpMode=RCIM_Cubic, leaveTangent=-2.5249, tangentMode=RCTM_User, time=0.1408, value=0.9847}
         {arriveTangent=-0.0405, interpMode=RCIM_Cubic, leaveTangent=-0.0405, tangentMode=RCTM_User, time=1, value=0}
        ]}}}
     Scale Curve=&Multiply_Float(
      A=$System.TODAlpha
      B=&Multiply_Float(
       A=1.5
       B=&Multiply_Float(
        A=&Multiply_Float(A=$System.Timeline002.InverseNormalizedTime, B=$System.Timeline001.InverseNormalizedTime)
        B=$System.Timeline003.NormalizedTime
       )
      )
     )
    )
    Scale RGB=&MakeVectorFromLinearColorRGB(Color=$System.TODColor)
    ScaleA=true
    ScaleRGB=true
   )
  ]
  rendererClasses [&NiagaraSpriteRendererProperties]
  renderers [{rendererClass=&NiagaraSpriteRendererProperties}]
 }
}
summary {
 systemName=NS_CubeSurfing_Test_Dynamic
 emitters [
  {emitterName=Blocks, rendererClasses [&NiagaraMeshRendererProperties, &NiagaraSpriteRendererProperties]}
  {emitterName=Glow, rendererClasses [&NiagaraSpriteRendererProperties]}
  {
   emitterName=Hands
   rendererClasses [
    &NiagaraSpriteRendererProperties
    &NiagaraLightRendererProperties
    &NiagaraMeshRendererProperties
   ]
  }
  {emitterName=Wind_Streaks, rendererClasses [&NiagaraSpriteRendererProperties]}
  {emitterName=Smoke, rendererClasses [&NiagaraSpriteRendererProperties]}
  {emitterName=Hands_Energy, rendererClasses [&NiagaraMeshRendererProperties]}
  {emitterName=Deploy_Block, rendererClasses [&NiagaraMeshRendererProperties, &NiagaraSpriteRendererProperties]}
 ]
 userVariables [
  User.bIsBoosted {type=NiagaraBool}
  User.bIsOnWater {type=NiagaraBool}
  User.blsOnLand {type=NiagaraBool, defaultValue=true}
  User.CubeColorMax {type=LinearColor, description="", defaultValue {struct=LinearColor, value=<rgba 0 0.6514 0.7379 1>}}
  User.CubeColorMin {type=LinearColor, description="", defaultValue {struct=LinearColor, value=<rgba 0.0262 0.2016 0.5583 1>}}
 ]
}
