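# Niagara one-shot effect for the spawn of a CubeSurfing trail, combining a burst of small cube-colored meshes with a glowing core.
# Blocks002 spawns short-lived meshes at a decaying rate, popping and shrinking them, fading alpha, and tinting between User.CubeColorMin and CubeColorMax.
# Core bursts one camera-offset particle carrying three sprite color/size parameter sets for its two sprite renderers.
# Jagged curves make those layers flicker, while FNTimeOfDayScalar scales their brightness or alpha by time of day.
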NiagaraSystem /CubeSurfing/Assets/VFX/NS_CubeSurfing_Trail_Spawn
dependencies {
 usedDataInterfaces [&NiagaraDataInterfaceCurve]
 usedDynamicInputs [
  &MakeLinearColorFromVectorAndFloat
  &MakeVectorFromLinearColorRGB
  &MergeHelper_VectorSwitch
  &SpawnGroup
 ]
 usedModules [&SystemState]
}
emitters {
 Blocks002 {
  emitterUpdateScript [
   &EmitterState(Life Cycle Mode=::NewEnumerator1, Loop Behavior=::NewEnumerator1, Loop Delay=0.05, Loop Duration=1, UseLoopDelay=false)
   &SpawnRate(
    SpawnRate=&FloatFromCurve(FloatCurve=DataInterface() {propertyValues {Curve {keys [<at 0 1>,<at 1 0>]}}}, CurveIndex=$Emitter.NormalizedLoopAge, Scale Curve=25)
    Use Spawn Probability=false
   )
  ]
  particleSpawnScript [
   &InitializeParticle(
    Color Maximum=$User.CubeColorMax
    Color Minimum=$User.CubeColorMin
    Color Mode=::NewEnumerator2
    Lifetime Max=0.5
    Lifetime Min=0.25
    Lifetime Mode=::NewEnumerator1
    Mesh Scale Mode=::NewEnumerator3
    Mesh Uniform Scale=1.5
    Sprite Size Mode=::NewEnumerator3
    Uniform Sprite Size=7.5
    UsePositionOffset=false
    Write Material Random=true
   )
   &ShapeLocation(MaskSpawnGroup=false, Override Seed=false, Sphere Radius=10, UseOwnerScale=true)
  ]
  particleUpdateScript [
   &ParticleState()
   &ScaleMeshSize(Scale Factor=&Lerp_Vector(
     A=&RandomRangeVector(Minimum=<0.1 0.1 0.1>, Maximum=<0.25 0.25 0.25>)
     B=<1 1 1>
     Alpha=&FloatFromCurve(FloatCurve=DataInterface() {propertyValues {Curve {
         defaultValue=1
         keys [
          {arriveTangent=9.0725, interpMode=RCIM_Cubic, leaveTangent=9.0725, tangentMode=RCTM_User, time=0, value=0}
          {arriveTangent=-2.5249, interpMode=RCIM_Cubic, leaveTangent=-2.5249, tangentMode=RCTM_User, time=0.1408, value=0.9847}
          {arriveTangent=-0.0405, interpMode=RCIM_Cubic, leaveTangent=-0.0405, tangentMode=RCTM_User, time=1, value=0}
         ]
        }}})
    ))
   &ScaleColor(
    Scale Alpha=&FloatFromCurve(
     FloatCurve=DataInterface() {propertyValues {Curve {defaultValue=1, keys [<at 0 0>, <at 0.2 1>, {arriveTangent=-1.3333, leaveTangent=-1.3333, time=0.95, value=1}, <at 1 0>]}}}
     CurveIndex=$Particles.Mask
    )
    ScaleA=true
    ScaleRGB=true
   )
   &DynamicMaterialParameters(
    Index 0 Param 1=&FloatFromCurve(FloatCurve=DataInterface() {propertyValues {Curve {
        defaultValue=1
        keys [
         {arriveTangent=-10.4479, interpMode=RCIM_Cubic, leaveTangent=-10.4479, tangentMode=RCTM_User, time=0, value=1}
         {interpMode=RCIM_Cubic, tangentMode=RCTM_User, time=0.25, value=0}
        ]
       }}})
    Index 0 Param 2=1
    Index 0 Param 3=0.33
    Module.Index0Float0Write=true
    Module.Index0Float1Write=true
    Module.Index0Float2Write=true
    Module.Index0Float3Write=true
   )
   &Color(Color=&RandomRangeLinearColor(Minimum=$User.CubeColorMin, Maximum=$User.CubeColorMax))
   &Drag(Drag=7.5, Use Linear Drag=true, Use Rotational Drag=true)
   &SolveForcesAndVelocity(Clamp Velocity=false, Limit Acceleration=false)
  ]
  rendererClasses [&NiagaraMeshRendererProperties]
  renderers [{rendererClass=&NiagaraMeshRendererProperties, rendererIndex=0}]
 }
 Core {
  emitterUpdateScript [
   &EmitterState(Life Cycle Mode=::NewEnumerator1, Loop Behavior=::NewEnumerator1, Loop Duration=1, UseLoopDelay=false)
   &SpawnBurst_Instantaneous(Spawn Count=1, Use Spawn Probability=false)
  ]
  particleSpawnScript [
   &InitializeParticle(Lifetime=0.8, Position Mode=::NewEnumerator0, Write Material Random=true)
   /SetVariables(
    Particles.Sprite_Color_01:LinearColor=<rgba 0 1.4263 2 1>
    Particles.Sprite_DP_01:Vector4f=&MakeVector4(X=&RandomRangeFloat(Minimum=0.4, Maximum=0.6), Y=1, Z=1, W=1)
    Particles.Sprite_Size_01:Vector2f=&Vector2DFromFloat(Value=250)
   ) {isSetParameters=true}
   /SetVariables(
    Particles.Sprite_Color_02:LinearColor=<rgba 0 0.0654 0.276 0.3>
    Particles.Sprite_DP_02:Vector4f=&MakeVector4(X=&RandomRangeFloat(Minimum=0.4, Maximum=0.6), Y=1, Z=1, W=1)
    Particles.Sprite_Size_02:Vector2f=&Vector2DFromFloat(Value=400)
   ) {isSetParameters=true}
   /SetVariables(
    Particles.Sprite_Color_03:LinearColor=<rgba 0 0.5167 1 3.5>
    Particles.Sprite_Size_03:Vector2f=&Vector2DFromFloat(Value=200)
   ) {isSetParameters=true}
  ]
  particleUpdateScript [
   &ParticleState()
   &CameraOffset(Camera Offset Amount=50)
   /SetVariables(
    Particles.Sprite_Color_01:LinearColor=&MultiplyLinearColor(
     Linear Color=$Particles.Initial.Sprite_Color_01
     Scale Factor=&MakeFloat4FromVectorAndFloat(
      "Vector (RGB)"=<1 1 1>
      "Float (Alpha)"=&Multiply_Float(
       A=&FNTimeOfDayScalar(Night Scale=0.5, Morning Scale=0.3, Evening Scale=0.4)
       B=&FloatFromCurve(
        FloatCurve=DataInterface() {propertyValues {Curve {keys [<at 0 0>,<at 0.2 1>,<at 0.8 1>,<at 1 0>]}}}
        CurveIndex=&Multiply_Float(A=$Particles.NormalizedAge, B=1.75)
       )
      )
     )
    )
    Particles.Sprite_Size_01:Vector2f=&Multiply_Vector2D(
     A=$Particles.Initial.Sprite_Size_01
     B=&Vector2DFromFloat(Value=&FloatFromCurve(
       FloatCurve=DataInterface() {propertyValues {Curve {keys [<at 0 1.5 RCIM_Cubic>,<at 1 0>]}}}
       CurveIndex=&Multiply_Float(A=$Particles.NormalizedAge, B=1.75)
       Scale Curve=&FloatFromCurve(
        FloatCurve=DataInterface() {propertyValues {Curve {keys [<at 0 0>,<at 0.0757 0.4305>,<at 0.139 0.0649>,<at 0.2293 1.0605>,<at 0.3366 0.6372>,<at 0.4376 1.5821>,<at 0.6165 0.9227>,<at 0.7495 2.2107>,<at 1 0>]}}}
        CurveIndex=&Multiply_Float(A=$Particles.NormalizedAge, B=1)
       )
      ))
    )
   ) {isSetParameters=true}
   /SetVariables(
    Particles.Sprite_Color_02:LinearColor=&MultiplyLinearColor(
     Linear Color=$Particles.Initial.Sprite_Color_02
     Scale Factor=&MakeFloat4FromVectorAndFloat(
      "Vector (RGB)"=&Multiply_VectorByFloat(Vector=<1 1 1>, Float=&FNTimeOfDayScalar(Night Scale=0.5, Morning Scale=0.3, Evening Scale=0.4))
      "Float (Alpha)"=&FloatFromCurve(
       FloatCurve=DataInterface() {propertyValues {Curve {keys [<at 0 0>,<at 0.0757 0.4305>,<at 0.139 0.0649>,<at 0.2293 1.0605>,<at 0.3366 0.6372>,<at 0.4376 1.5821>,<at 0.6165 0.9227>,<at 0.9 2.375>,<at 1 0>]}}}
       CurveIndex=&Multiply_Float(A=$Particles.NormalizedAge, B=1.75)
      )
     )
    )
    Particles.Sprite_Size_02:Vector2f=&Multiply_Vector2D(
     A=$Particles.Initial.Sprite_Size_02
     B=&Vector2DFromFloat(Value=&FloatFromCurve(
       FloatCurve=DataInterface() {propertyValues {Curve {keys [<at 0 1.5 RCIM_Cubic>,<at 1 0>]}}}
       CurveIndex=&Multiply_Float(A=$Particles.NormalizedAge, B=1.75)
       Scale Curve=&FloatFromCurve(
        FloatCurve=DataInterface() {propertyValues {Curve {keys [<at 0 0>,<at 0.0757 0.4305>,<at 0.139 0.0649>,<at 0.2293 1.0605>,<at 0.3366 0.6372>,<at 0.4376 1.5821>,<at 0.6165 0.9227>,<at 0.7495 2.2107>,<at 1 0>]}}}
        CurveIndex=&Multiply_Float(A=$Particles.NormalizedAge, B=1)
       )
      ))
    )
   ) {isSetParameters=true}
   /SetVariables(
    Particles.Sprite_Color_03:LinearColor=&MultiplyLinearColor(
     Linear Color=$Particles.Initial.Sprite_Color_03
     Scale Factor=&MakeFloat4FromVectorAndFloat(
      "Vector (RGB)"=<1 1 1>
      "Float (Alpha)"=&Multiply_Float(
       A=&FloatFromCurve(
        FloatCurve=DataInterface() {propertyValues {Curve {keys [{interpMode=RCIM_Cubic, leaveTangent=-0.6434, tangentMode=RCTM_Break, time=0, value=1}, <at 1 0>]}}}
        CurveIndex=&Multiply_Float(A=$Particles.NormalizedAge, B=2)
        Scale Curve=10
       )
       B=&FNTimeOfDayScalar(Night Scale=1, Morning Scale=0.3, Evening Scale=0.5)
      )
     )
    )
    Particles.Sprite_Size_03:Vector2f=&Multiply_Vector2D(
     A=$Particles.Initial.Sprite_Size_03
     B=&Vector2DFromFloat(Value=&FloatFromCurve(
       FloatCurve=DataInterface() {propertyValues {Curve {keys [{tangentMode=RCTM_User, time=0, value=0}, <at 1 1>]}}}
       CurveIndex=&Multiply_Float(A=$Particles.NormalizedAge, B=2)
       Scale Curve=&FloatFromCurve(
        FloatCurve=DataInterface() {propertyValues {Curve {keys [<at 0 0>,<at 0.0757 0.4305>,<at 0.139 0.0649>,<at 0.2293 1.0605>,<at 0.3366 0.6372>,<at 0.4376 1.5821>,<at 0.6165 0.9227>,<at 0.7495 2.2107>,<at 1 0>]}}}
        CurveIndex=&Multiply_Float(A=$Particles.NormalizedAge, B=1)
        Scale Curve=1.5
       )
      ))
    )
   ) {isSetParameters=true}
  ]
  rendererClasses [&NiagaraSpriteRendererProperties]
  renderers [
   {rendererClass=&NiagaraSpriteRendererProperties}
   {rendererClass=&NiagaraSpriteRendererProperties, rendererIndex=1}
  ]
 }
}
summary {
 systemName=NS_CubeSurfing_Trail_Spawn
 emitters [
  {emitterName=Blocks002, rendererClasses [&NiagaraMeshRendererProperties]}
  {emitterName=Core, rendererClasses [&NiagaraSpriteRendererProperties]}
 ]
 userVariables [
  User.CubeColorMax {type=LinearColor, description="", defaultValue {struct=LinearColor, value=<rgba 0 0.6514 0.7379 1>}}
  User.CubeColorMin {type=LinearColor, description="", defaultValue {struct=LinearColor, value=<rgba 0.0262 0.2016 0.5583 1>}}
 ]
}
