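# Niagara system for Cube Surfing water interaction, spraying a twin wake from two offset points, angled outward along local -Y.
# Four emitters layer sprite drops, splashes, wispy smoke and motor-burble meshes, which inherit source movement.
# A shared product of three system timelines, one inverted, gates every spawn probability and scales alpha.
# Normalized velocity sizes the burble meshes and their lifetime; user bools expose boost, water and land states.
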NiagaraSystem /CubeSurfing/Assets/VFX/NS_CubeSurfing_WaterInteractions
dependencies {
 usedDataInterfaces [
  &NiagaraDataInterfaceCurve
  &NiagaraDataInterfaceVectorCurve
  &NiagaraDataInterfaceVector2DCurve
  &NiagaraDataInterfaceColorCurve
 ]
 usedDynamicInputs [
  &SimulationPosition
  &ClampFloat
  &RemapRange
  &VectorLength
  /Game/Effects/Niagara/DynamicInputs/State/FN_IsExecutionStateActive
 ]
 usedModules [
  &self:SystemSpawnScript.SystemScriptSource.SystemScriptGraph.NiagaraNodeAssignment_2.SetVariables_#1
  &SystemState
  &InterpolateOverTime
  &self:SystemSpawnScript.SystemScriptSource.SystemScriptGraph.NiagaraNodeAssignment_0.SetVariables_#2
  &self:SystemSpawnScript.SystemScriptSource.SystemScriptGraph.NiagaraNodeAssignment_1.SetVariables_#3
  &Timeline
 ]
}
emitters {
 FN_Water_Drops_ {
  emitterUpdateScript [
   &EmitterState()
   &SpawnRate(
    Spawn Probability=&Multiply_Float(
     A=&Multiply_Float(A=$System.Timeline002.NormalizedTime, B=$System.Timeline003.InverseNormalizedTime)
     B=$System.Timeline001.NormalizedTime
    )
    SpawnRate=30
    Use Spawn Probability=true
   )
  ]
  particleSpawnScript [
   /SetVariables(Particles.Random=&UniformRangedFloat(Minimum=0, Maximum=1)) {isSetParameters=true}
   &PartitionParticles()
   &InitializeParticle(
    Color=<rgba 1 1 1 0.8>
    Lifetime=&UniformRangedFloat(Minimum=1, Maximum=2)
    Mass=&UniformRangedFloat(Minimum=0.6, Maximum=1.2)
    Mass Mode=::NewEnumerator1
    Sprite Rotation Mode=::NewEnumerator0
    Sprite Size=&Vector2DFromCurve(
     Vector2Curve=DataInterface() {propertyValues {
       XCurve {keys [<at -0.0011 2.0768 RCIM_Cubic>,<at 0.9989 0.3489 RCIM_Cubic>]}
       YCurve {keys [
         {arriveTangent=-30.8979, interpMode=RCIM_Cubic, leaveTangent=-30.8979, tangentMode=RCTM_User, time=0, value=10.3803}
         {arriveTangent=-4.6899, interpMode=RCIM_Cubic, leaveTangent=-4.6899, tangentMode=RCTM_User, time=0.2382, value=3.8026}
         <at 0.9989 1.9328 RCIM_Cubic>
        ]}
      }}
     CurveIndex=$Particles.Random
     ScaleCurve=<1 1>
    )
    Sprite Size Mode=::NewEnumerator4
    UsePositionOffset=false
    Write Material Random=true
   ) {module=InitializeParticle001}
   &DynamicMaterialParameters(
    Index 0 Param 1=&UniformRangedFloat(Minimum=4, Maximum=10)
    Module.Param0WriteEnabled=true
    Module.Param1WriteEnabled=false
    Module.Param2WriteEnabled=false
    Module.Param3WriteEnabled=false
   )
   &ShapeLocation(
    MaskSpawnGroup=false
    Non Uniform Scale=<1 0.2 0.2>
    Offset=&Add_Vector(
     A=Linked() {linkedVariable {name=System.Position_Offset, type=NiagaraPosition}}
     B=&MergeHelper_VectorSwitch(0=<25 0 0>, 1=<-25 0 0>, 2=?, 3=?, Index=$Particles.PartitionParticles.Partition, Num=::NewEnumerator0)
    )
    Offset Coordinate Space=::Local
    Offset Mode=::NewEnumerator0
    Override Seed=false
    Sphere Radius=2.5
    UseOwnerScale=true
   )
   &AddVelocity(
    Cone Angle=20
    Cone Axis=&MergeHelper_VectorSwitch(0=<0.5 -1 0>, 1=<-0.5 -1 0>, 2=?, 3=?, Index=$Particles.PartitionParticles.Partition, Num=::NewEnumerator0)
    Module.UseConeFalloff=false
    Module.UseVelDistribution=false
    Override Seed=false
    Velocity Mode=::NewEnumerator2
    Velocity Speed=&RandomRangeFloat(Minimum=250, Maximum=500)
   ) {module=AddVelocity001}
  ]
  particleUpdateScript [
   &ParticleState()
   /Game/Effects/Niagara/Emitters/Water/FN_DistanceKillParticles(AgeReduiceAmount=0.01)
   &ScaleColor(
    Scale Alpha=&FloatFromCurve(
     FloatCurve=DataInterface() {propertyValues {Curve {keys [
         <at 0 0 RCIM_Cubic>
         {arriveTangent=0.002, interpMode=RCIM_Cubic, leaveTangent=0.002, time=0.1183, value=1.0113}
         <at 1.0011 0.002 RCIM_Cubic>
        ]}}}
     Scale Curve=&Multiply_Float(
      A=&Multiply_Float(A=$System.Timeline002.NormalizedTime, B=$System.Timeline003.InverseNormalizedTime)
      B=$System.Timeline001.NormalizedTime
     )
    )
    ScaleA=true
    ScaleRGB=true
   )
   &InheritSourceMovement(
    Applied Position Scale=&Multiply_VectorByFloat(
     Vector=<1 1 1>
     Float=&FloatFromCurve(FloatCurve=DataInterface() {propertyValues {Curve {
         defaultValue=1
         keys [
          {interpMode=RCIM_Cubic, tangentMode=RCTM_User, time=0.5, value=1}
          {arriveTangent=-4.0056, interpMode=RCIM_Cubic, leaveTangent=-4.0056, tangentMode=RCTM_User, time=1, value=0.25}
         ]
        }}})
    )
    ApplyForce=false
    ApplyPosition=true
    ApplyPositionFalloff=false
   )
   /Niagara/Modules/Update/Forces/CurlNoiseForce(
    Noise Frequency=1
    Noise Strength=&FloatFromCurve(FloatCurve=DataInterface() {propertyValues {Curve {keys [
         {arriveTangent=-1159.6794, interpMode=RCIM_Cubic, leaveTangent=-1159.6792, tangentMode=RCTM_User, time=-0.0046, value=500}
         {arriveTangent=-447.4286, interpMode=RCIM_Cubic, leaveTangent=-447.4286, time=0.2223, value=189.6793}
         <at 1.0011 50 RCIM_Cubic>
        ]}}})
    Pan Noise=false
   )
   &Drag(Use Linear Drag=true, Use Rotational Drag=false)
   &GravityForce()
   &SolveForcesAndVelocity(Clamp Velocity=false, Limit Acceleration=false)
   &ScaleSpriteSize(Scale Factor=&Vector2DFromCurve(
     Vector2Curve=DataInterface() {propertyValues {
       XCurve {keys [<at -0.0011 1.3381 RCIM_Cubic>, {interpMode=RCIM_Cubic, leaveTangent=-1.0325, time=0.3279, value=0.7734}]}
       YCurve {keys [<at -0.0011 3.4478 RCIM_Cubic>,<at 0.3279 0.7734 RCIM_Cubic>]}
      }}
     ScaleCurve=<1 1>
    ))
  ]
  rendererClasses [&NiagaraSpriteRendererProperties]
  renderers [{rendererClass=&NiagaraSpriteRendererProperties}]
 }
 Motor_Burble_Mesh_WS_NoSkelRef {
  emitterUpdateScript [
   &EmitterState()
   &SpawnRate(
    Spawn Probability=&Multiply_Float(
     A=&Multiply_Float(A=$System.Timeline002.NormalizedTime, B=$System.Timeline003.InverseNormalizedTime)
     B=$System.Timeline001.NormalizedTime
    )
    SpawnRate=20
    Use Spawn Probability=true
   )
  ]
  particleSpawnScript [
   &InitializeParticle(
    Color=<rgba 1 1.5 2 1>
    Lifetime=&Lerp_Float(A=0.25, B=&UniformRangedFloat(Minimum=0.75, Maximum=1), Alpha=$System.NormalizedVelocityRange)
    Mesh Scale=&Multiply_VectorByFloat(Vector=&UniformRangedVector(Minimum=<0.3125 0.2 0.2>, Maximum=<0.5 0.35 0.2875>), Float=1)
    Mesh Scale Mode=::NewEnumerator4
    Position Mode=::NewEnumerator0
    Write Material Random=true
   ) {module=InitializeParticle001}
   &PartitionParticles()
   /SetVariables(Particles.Scale=&Multiply_Vector(A=$Particles.Scale, B=&Lerp_Vector(A=<0.5 0.5 0.5>, B=<0.75 0.75 0.8>, Alpha=$System.NormalizedVelocityRange))) {isSetParameters=true}
   &CameraOffset(Camera Offset Amount=10)
   &ShapeLocation(
    MaskSpawnGroup=false
    Non Uniform Scale=<1 0.2 0.2>
    Offset=&Add_Vector(
     A=Linked() {linkedVariable {name=System.Position_Offset, type=NiagaraPosition}}
     B=&MergeHelper_VectorSwitch(0=<25 0 0>, 1=<-25 0 0>, 2=?, 3=?, Index=$Particles.PartitionParticles.Partition, Num=::NewEnumerator0)
    )
    Offset Coordinate Space=::Local
    Offset Mode=::NewEnumerator0
    Override Seed=false
    Sphere Radius=2.5
    UseOwnerScale=true
   )
   &AddVelocity(
    Cone Angle=20
    Cone Axis=&MergeHelper_VectorSwitch(0=<0.5 -1 0>, 1=<-0.5 -1 0>, 2=?, 3=?, Index=$Particles.PartitionParticles.Partition, Num=::NewEnumerator0)
    Module.UseConeFalloff=false
    Module.UseVelDistribution=false
    Override Seed=false
    Velocity Mode=::NewEnumerator2
    Velocity Speed=&RandomRangeFloat(Minimum=150, Maximum=300)
   ) {module=AddVelocity001}
  ]
  particleUpdateScript [
   &ParticleState()
   &InheritSourceMovement(
    Applied Position Scale=&Multiply_VectorByFloat(
     Vector=<1 1 1>
     Float=&FloatFromCurve(FloatCurve=DataInterface() {propertyValues {Curve {
         defaultValue=1
         keys [
          {interpMode=RCIM_Cubic, tangentMode=RCTM_User, time=0.5, value=1}
          {arriveTangent=-2.6233, interpMode=RCIM_Cubic, leaveTangent=-2.6233, tangentMode=RCTM_User, time=1, value=0.5}
         ]
        }}})
    )
    ApplyForce=false
    ApplyPosition=true
    ApplyPositionFalloff=false
   )
   &SolveForcesAndVelocity(Clamp Velocity=false, Limit Acceleration=false)
   /SetVariables(Particles.Scale=&Multiply_Vector(
     A=$Particles.Initial.Scale
     B=&VectorFromCurve(
      VectorCurve=DataInterface() {propertyValues {XCurve {keys [<at 0 1>,<at 0.9989 1>]}, YCurve {keys [<at 0 1>,<at 0.9989 1>]}, ZCurve {keys [<at 0 0>,<at 0.2345 0.7983>,<at 0.9989 1>]}}}
      CurveIndex=$Emitter.Scalability.DistanceFraction
     )
    )) {isSetParameters=true}
   &ScaleColor(
    Scale Alpha=&FloatFromCurve(
     FloatCurve=DataInterface() {propertyValues {Curve {keys [
         {arriveTangent=9.9926, interpMode=RCIM_Cubic, leaveTangent=9.9927, tangentMode=RCTM_User, time=0, value=0}
         {arriveTangent=-1.1119, interpMode=RCIM_Cubic, leaveTangent=-1.1119, tangentMode=RCTM_User, time=0.0216, value=0.35}
         <at 1.0011 0.002 RCIM_Cubic>
        ]}}}
     Scale Curve=&Multiply_Float(
      A=&Multiply_Float(A=$System.Timeline002.NormalizedTime, B=$System.Timeline003.InverseNormalizedTime)
      B=$System.Timeline001.NormalizedTime
     )
    )
    ScaleA=true
    ScaleRGB=true
   )
   /SetVariables(Particles.Scale=&Multiply_Vector(
     A=$Particles.Initial.Scale
     B=&VectorFromCurve(
      VectorCurve=DataInterface() {propertyValues {
        XCurve {keys [
          {arriveTangent=4.6994, interpMode=RCIM_Cubic, leaveTangent=4.6994, tangentMode=RCTM_User, time=0, value=0.0795}
          {arriveTangent=0.4589, interpMode=RCIM_Cubic, leaveTangent=0.4589, tangentMode=RCTM_User, time=0.1447, value=0.7609}
          <at 1 1>
         ]}
        YCurve {keys [{arriveTangent=2.5595, interpMode=RCIM_Cubic, leaveTangent=2.5595, tangentMode=RCTM_User, time=0, value=0.0767}, <at 1 1>]}
        ZCurve {keys [
          <at 0 0 RCIM_Cubic>
          {arriveTangent=-2.3619, interpMode=RCIM_Cubic, leaveTangent=-2.3619, tangentMode=RCTM_User, time=0.336, value=0.9567}
          <at 0.9989 0>
         ]}
       }}
      Scale Curve=<1.5 2.5 1>
     )
    )) {isSetParameters=true}
   &DynamicMaterialParameters(
    Index 0 Param 1=&FloatFromCurve(
     FloatCurve=DataInterface() {propertyValues {Curve {keys [
         {arriveTangent=0.2615, interpMode=RCIM_Cubic, leaveTangent=0.2615, tangentMode=RCTM_User, time=0, value=-1.7984}
         {arriveTangent=3.2109, interpMode=RCIM_Cubic, leaveTangent=3.2109, tangentMode=RCTM_User, time=0.9985, value=0}
        ]}}}
     Scale Curve=0.8
    )
    Index 0 Param 2=&Add_Float(
     A=&FloatFromCurve(
      FloatCurve=DataInterface() {propertyValues {Curve {keys [
          {arriveTangent=-9.4845, interpMode=RCIM_Cubic, leaveTangent=-9.4845, tangentMode=RCTM_User, time=0, value=4}
          {arriveTangent=0.2383, interpMode=RCIM_Cubic, leaveTangent=0.2383, tangentMode=RCTM_User, time=1, value=0}
         ]}}}
      Scale Curve=0.6
     )
     B=0
    )
    Index 1 Param 1=&FloatFromCurve(FloatCurve=DataInterface() {propertyValues {Curve {keys [
         <at 0 0 RCIM_Cubic>
         {arriveTangent=1, interpMode=RCIM_Cubic, leaveTangent=1, time=0.4, value=0.15}
         {arriveTangent=0.5518, interpMode=RCIM_Cubic, leaveTangent=0.5518, tangentMode=RCTM_User, time=1, value=1}
        ]}}})
    Module.Param0WriteEnabled=true
    Module.Param0WriteEnabled1=true
    Module.Param1WriteEnabled=true
    Module.Param1WriteEnabled1=false
    Module.Param2WriteEnabled=true
    Module.Param2WriteEnabled1=false
    Module.Param3WriteEnabled=true
    Module.Param3WriteEnabled1=false
    Write Parameter Index 1=true
   )
  ]
  rendererClasses [&NiagaraMeshRendererProperties]
  renderers [
   {rendererClass=&NiagaraMeshRendererProperties, rendererIndex=0}
   {rendererClass=&NiagaraMeshRendererProperties, rendererIndex=1}
  ]
 }
 SlurpWispySmoke {
  emitterUpdateScript [
   &EmitterState()
   &SpawnRate(
    Spawn Probability=&Multiply_Float(
     A=&Multiply_Float(A=$System.Timeline002.NormalizedTime, B=$System.Timeline003.InverseNormalizedTime)
     B=$System.Timeline001.NormalizedTime
    )
    SpawnRate=30
    Use Spawn Probability=true
   )
  ]
  particleSpawnScript [
   &PartitionParticles()
   &InitializeParticle(
    Color Maximum=<rgba 1 1 1 0.25>
    Color Minimum=<rgba 1 1 1 0.8>
    Color Mode=::NewEnumerator2
    Lifetime Max=1
    Lifetime Min=0.5
    Lifetime Mode=::NewEnumerator1
    Sprite Rotation Mode=::NewEnumerator2
    Sprite Size Mode=::NewEnumerator1
    Uniform Sprite Size Max=80
    Uniform Sprite Size Min=45
    UsePositionOffset=false
    Write Material Random=true
   ) {module=InitializeParticle001}
   &Color(Color=&MultiplyLinearColor(
     Linear Color=$Particles.Color
     Scale Factor=&MakeVector4_FromLinearColor(LinearColor=&RandomRangeLinearColor(Minimum=<rgba 0 0.8333 1 1>, Maximum=<rgba 0 0.4471 1 1>))
    ))
   /SetVariables(Particles.UVScale:Vector2f=&UniformRangedVector2D(Minimum=<-1 1>, Maximum=<-1 1>)) {isSetParameters=true}
   &ShapeLocation(
    MaskSpawnGroup=false
    Non Uniform Scale=<1 0.2 0.2>
    Offset=&Add_Vector(
     A=Linked() {linkedVariable {name=System.Position_Offset, type=NiagaraPosition}}
     B=&MergeHelper_VectorSwitch(0=<25 0 0>, 1=<-25 0 0>, 2=?, 3=?, Index=$Particles.PartitionParticles.Partition, Num=::NewEnumerator0)
    )
    Offset Coordinate Space=::Local
    Offset Mode=::NewEnumerator0
    Override Seed=false
    Sphere Radius=2.5
    UseOwnerScale=true
   )
   &AddVelocity(
    Cone Angle=20
    Cone Axis=&MergeHelper_VectorSwitch(0=<0.5 -1 0>, 1=<-0.5 -1 0>, 2=?, 3=?, Index=$Particles.PartitionParticles.Partition, Num=::NewEnumerator0)
    Module.UseConeFalloff=false
    Module.UseVelDistribution=false
    Override Seed=false
    Velocity Mode=::NewEnumerator2
    Velocity Speed=&RandomRangeFloat(Minimum=150, Maximum=300)
   ) {module=AddVelocity001}
  ]
  particleUpdateScript [
   &ParticleState()
   &InheritSourceMovement(
    Applied Position Scale=&Multiply_VectorByFloat(
     Vector=<1 1 1>
     Float=&FloatFromCurve(FloatCurve=DataInterface() {propertyValues {Curve {
         defaultValue=1
         keys [
          {interpMode=RCIM_Cubic, tangentMode=RCTM_User, time=0.5, value=1}
          {arriveTangent=-2.946, interpMode=RCIM_Cubic, leaveTangent=-2.946, tangentMode=RCTM_User, time=1, value=0.5}
         ]
        }}})
    )
    ApplyForce=false
    ApplyPosition=true
    ApplyPositionFalloff=false
   )
   &GravityForce(Gravity=&RandomRangeVector(Minimum=<0 0 -732>, Maximum=<1 1 -1500>))
   &SolveForcesAndVelocity(Clamp Velocity=false, Limit Acceleration=false)
   &SpriteRotationRate(Rotation Rate=&Multiply_Float(
     A=&Lerp_Float(
      A=-400
      B=400
      Alpha=/Niagara/DynamicInputs/Math/Floor_Float(Float:NiagaraFloat=&Add_Float(A=0.2, B=$Particles.Initial.MaterialRandom))
     )
     B=&FloatFromCurve(
      FloatCurve=DataInterface() {propertyValues {Curve {keys [{arriveTangent=-4.4309, interpMode=RCIM_Cubic, leaveTangent=-4.4309, tangentMode=RCTM_User, time=0, value=1}, <at 0.1544 0.0948>, <at 1 0>]}}}
      Scale Curve=2
     )
    ))
   &ScaleColor(
    Scale Alpha=&Multiply_Float(
     A=&FloatFromCurve(FloatCurve=DataInterface() {propertyValues {Curve {
         defaultValue=0
         keys [
          {arriveTangent=4.1785, interpMode=RCIM_Cubic, leaveTangent=4.1785, tangentMode=RCTM_User, time=0, value=0}
          {arriveTangent=1.1179, interpMode=RCIM_Cubic, leaveTangent=1.1179, tangentMode=RCTM_User, time=0.2042, value=0.6228}
          {arriveTangent=0.0224, interpMode=RCIM_Cubic, leaveTangent=0.0224, tangentMode=RCTM_User, time=1, value=1}
         ]
        }}})
     B=&Multiply_Float(
      A=&Lerp_Float(A=1.25, B=0, Alpha=&Power(a=$Particles.NormalizedAge, b=1))
      B=&Multiply_Float(
       A=&Multiply_Float(A=$System.Timeline002.NormalizedTime, B=$System.Timeline003.InverseNormalizedTime)
       B=$System.Timeline001.NormalizedTime
      )
     )
    )
    ScaleA=true
    ScaleRGB=true
   )
   &ScaleSpriteSize(Scale Factor=&Vector2DFromFloat(Value=&FloatFromCurve(FloatCurve=DataInterface() {propertyValues {Curve {keys [
          {arriveTangent=6.0672, interpMode=RCIM_Cubic, leaveTangent=6.0672, tangentMode=RCTM_User, time=0, value=0.25}
          {arriveTangent=0.75, interpMode=RCIM_Cubic, leaveTangent=0.75, time=0.0965, value=0.8098}
          <at 1 1>
         ]}}})))
   &SubUVAnimation() {module=SubUVAnimation001}
   &CameraOffset(Camera Offset Amount=99)
  ]
  rendererClasses [&NiagaraSpriteRendererProperties]
  renderers [{rendererClass=&NiagaraSpriteRendererProperties}]
 }
 WaterSplash {
  emitterUpdateScript [
   &EmitterState()
   &SpawnRate(
    Spawn Probability=&Multiply_Float(
     A=&Multiply_Float(A=$System.Timeline002.NormalizedTime, B=$System.Timeline003.InverseNormalizedTime)
     B=$System.Timeline001.NormalizedTime
    )
    SpawnRate=30
    Use Spawn Probability=true
   )
  ]
  particleSpawnScript [
   &InitializeParticle(
    Color Mode=::NewEnumerator0
    Lifetime Max=0.75
    Lifetime Min=0.5
    Lifetime Mode=::NewEnumerator1
    Sprite Rotation Mode=::NewEnumerator2
    Sprite Size Mode=::NewEnumerator1
    Uniform Sprite Size Max=100
    Uniform Sprite Size Min=75
    UsePositionOffset=false
    Write Material Random=true
   )
   &PartitionParticles()
   &ShapeLocation(
    MaskSpawnGroup=false
    Non Uniform Scale=<1 0.2 0.2>
    Offset=&Add_Vector(
     A=Linked() {linkedVariable {name=System.Position_Offset, type=NiagaraPosition}}
     B=&MergeHelper_VectorSwitch(0=<25 0 0>, 1=<-25 0 0>, 2=?, 3=?, Index=$Particles.PartitionParticles.Partition, Num=::NewEnumerator0)
    )
    Offset Coordinate Space=::Local
    Offset Mode=::NewEnumerator0
    Override Seed=false
    Sphere Radius=2.5
    UseOwnerScale=true
   )
   &AddVelocity(
    Cone Angle=20
    Cone Axis=&MergeHelper_VectorSwitch(0=<0.5 -1 0>, 1=<-0.5 -1 0>, 2=?, 3=?, Index=$Particles.PartitionParticles.Partition, Num=::NewEnumerator0)
    Module.UseConeFalloff=false
    Module.UseVelDistribution=false
    Override Seed=false
    Velocity Mode=::NewEnumerator2
    Velocity Speed=&RandomRangeFloat(Minimum=300, Maximum=600)
   ) {module=AddVelocity001}
   /SetVariables(Particles.SubImageIndex=&RandomRangeFloat(Minimum=0, Maximum=3)) {isSetParameters=true}
  ]
  particleUpdateScript [
   &ParticleState()
   &InheritSourceMovement(
    Applied Position Scale=&Multiply_VectorByFloat(
     Vector=<1 1 1>
     Float=&FloatFromCurve(FloatCurve=DataInterface() {propertyValues {Curve {
         defaultValue=1
         keys [
          {interpMode=RCIM_Cubic, tangentMode=RCTM_User, time=0, value=1}
          {arriveTangent=-2.4996, interpMode=RCIM_Cubic, leaveTangent=-2.4996, tangentMode=RCTM_User, time=1, value=0.25}
         ]
        }}})
    )
    ApplyForce=false
    ApplyPosition=true
    ApplyPositionFalloff=false
   )
   &GravityForce(Gravity=<0 0 -680>)
   &SolveForcesAndVelocity(Clamp Velocity=false, Limit Acceleration=false)
   &ScaleSpriteSize(Uniform Curve Sprite Scale=DataInterface() {propertyValues {Curve {keys [
        {arriveTangent=4.1683, interpMode=RCIM_Cubic, leaveTangent=4.1683, tangentMode=RCTM_User, time=0, value=0.2}
        {interpMode=RCIM_Cubic, tangentMode=RCTM_User, time=1, value=2}
       ]}}}) {module=ScaleSpriteSize001}
   &CameraOffset(Camera Offset Amount=&Multiply_Float(
     A=&RandomRangeFloat(Minimum=0, Maximum=9)
     B=&FloatFromCurve(FloatCurve=DataInterface() {propertyValues {Curve {keys [
          {interpMode=RCIM_Cubic, leaveTangent=-42.7533, tangentMode=RCTM_Break, time=0, value=1}
          {interpMode=RCIM_Cubic, tangentMode=RCTM_Break, time=1, value=-9}
         ]}}})
    ))
   &ScaleColor(
    Scale Alpha=&FloatFromCurve(
     FloatCurve=DataInterface() {propertyValues {Curve {keys [{interpMode=RCIM_Cubic, tangentMode=RCTM_User, time=0, value=1}, {interpMode=RCIM_Cubic, tangentMode=RCTM_User, time=1, value=0}]}}}
     Scale Curve=&Multiply_Float(
      A=0.1
      B=&Multiply_Float(
       A=&Multiply_Float(A=$System.Timeline002.NormalizedTime, B=$System.Timeline003.InverseNormalizedTime)
       B=$System.Timeline001.NormalizedTime
      )
     )
    )
    Scale RGB=&MakeVectorFromLinearColorRGB(Color=&ColorFromCurve(ColorCurve=DataInterface() {propertyValues {
        AlphaCurve {keys [<at 0 1 RCIM_Cubic>,<at 1 1 RCIM_Cubic>]}
        BlueCurve {keys [<at 0.2219 0.9792>,<at 0.6021 0.2245>,<at 1 0.2969>]}
        GreenCurve {keys [<at 0.2219 0.5638>,<at 0.6021 0.164>,<at 1 0.0648>]}
        RedCurve {keys [<at 0.2219 0.0804>,<at 0.6021 0>,<at 1 0.0321>]}
       }}))
    ScaleA=true
    ScaleRGB=true
   ) {module=ScaleColor002}
   &DynamicMaterialParameters(
    Index 0 Param 2=&FloatFromCurve(
     FloatCurve=DataInterface() {propertyValues {Curve {keys [<at 0.1 1 RCIM_Cubic>, {interpMode=RCIM_Cubic, tangentMode=RCTM_User, time=0.3, value=0}]}}}
     Scale Curve=0
    )
    Index 0 Param 3=&Lerp_Float(
     A=1
     B=4
     Alpha=&FloatFromCurve(FloatCurve=DataInterface() {propertyValues {Curve {keys [<at 0 1 RCIM_Cubic>,<at 0.5 0 RCIM_Cubic>]}}})
    )
    Index 0 Param 4=&FloatFromCurve(FloatCurve=DataInterface() {propertyValues {Curve {keys [{interpMode=RCIM_Cubic, leaveTangent=4.137, tangentMode=RCTM_Break, time=0.1, value=0}, <at 1 1>]}}})
    Module.Index0Param0Write=true
    Module.Index0Param1Write=true
    Module.Index0Param2Write=true
    Module.Index0Param3Write=true
   ) {module=DynamicMaterialParameters001}
  ]
  rendererClasses [&NiagaraSpriteRendererProperties]
  renderers [{rendererClass=&NiagaraSpriteRendererProperties}]
 }
}
summary {
 systemName=NS_CubeSurfing_WaterInteractions
 emitters [
  {emitterName=Motor_Burble_Mesh_WS_NoSkelRef, rendererClasses [&NiagaraMeshRendererProperties]}
  {emitterName=FN_Water_Drops_, rendererClasses [&NiagaraSpriteRendererProperties]}
  {emitterName=WaterSplash, rendererClasses [&NiagaraSpriteRendererProperties]}
  {emitterName=SlurpWispySmoke, rendererClasses [&NiagaraSpriteRendererProperties]}
 ]
 userVariables [
  User.bIsBoosted {type=NiagaraBool}
  User.bIsOnWater {type=NiagaraBool}
  User.blsOnLand {type=NiagaraBool, defaultValue=true}
 ]
}
